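# Stereo one-shot MetaSound source that plays the Radient_MIX track alongside the BeamTest MIDI file and exposes its musical data as outputs.
# On play, a transport drives a metronome whose MIDI clock feeds both the MIDI player and the wave player's transport controller.
# MIDI notes and CC1 become Note On, normalized Pitch and cc01 outputs, alongside the MIDI clock.
# Audio routes to the Fortnite_AudioAnalysis_Music submix.
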MetaSoundSource /Game/NeverCook/FM/Radiant/Audio/MSS_Radiant_MusicMidi
inputs [UE.Source.OnPlay {type=Trigger}]
interfaces [UE.OutputFormat.Stereo:1.0, UE.Source.OneShot:1.0, UE.Source:1.1]
outputs [
 cc01 {type=Float}
 MIDI Clock {type=MIDIClock}
 Note On {type=Trigger}
 Pitch {type=Float}
 "UE.OutputFormat.Stereo.Audio:0" {type=Audio}
 "UE.OutputFormat.Stereo.Audio:1" {type=Audio}
 UE.Source.OneShot.OnFinished {type=Trigger}
]
pages [Default {
  input_defaults {UE.Source.OnPlay=null}
  nodes [
   n1 = HarmonixNodes.TriggerToTransport(Play=^UE.Source.OnPlay)
   n2 = UE.Reroute(Value=@n1.Transport)
   n3 = HarmonixNodes.Metronome(Transport=@n2.Value) {also {defaults {"Loop Length (Bars)"=54, Pre-roll Bars=0}}}
   HarmonixNodesMIDIPlayer_0 = HarmonixNodes.MIDIPlayer(MIDI Clock=@n3.MIDI Clock, Transport=@n1.Transport) {also {defaults {MIDI File=&/BeamTest, Pre-roll Bars=0}}}
   n4 = HarmonixNodes.MidiNoteTrigger(MIDI Stream=@HarmonixNodesMIDIPlayer_0.MIDI Stream)
   MapRangeMapRange_0 = MapRange.MapRange:Float(In=Convert.Float:Int32(Int32=@n4."MIDI Note ##").Float) {also {defaults {In Range B=127}}}
   MapRangeMapRange_1 = MapRange.MapRange:Float(In=Convert.Float:Int32(Int32=HarmonixNodes.MidiCCTrigger(MIDI Stream=@HarmonixNodesMIDIPlayer_0.MIDI Stream) {also {defaults {Control Number=1}}}.Value).Float) {also {defaults {In Range B=127}}}
   Reroute_1 = UE.Reroute(Value=@n4.Note On)
   n5 = HarmonixNodes.TransportWavePlayerController(MIDI Clock=@n3.MIDI Clock, Transport=UE.Reroute(Value=@n2.Value).Value) {also {version [0,1]}}
   WavePlayer_0 = UE.Wave Player:Stereo(Play=@n5.Play, Stop=@n5.Stop) {also {defaults {Loop="<MetaSoundBuilderResult.SUCCEEDED: 0>", Wave Asset=&/Radient_MIX}}}
  ]
  links [
   {dst="@MIDI Clock", src=HarmonixNodesMIDIPlayer_0.MIDI Clock}
   {dst="@Pitch", src=MapRangeMapRange_0.Out Value}
   {dst="@cc01", src=MapRangeMapRange_1.Out Value}
   {dst="@Note On", src=Reroute_1.Value}
   {dst="@UE.Source.OneShot.OnFinished", src=WavePlayer_0.On Finished}
   {dst="@UE.OutputFormat.Stereo.Audio:0", src=WavePlayer_0.Out Left}
   {dst="@UE.OutputFormat.Stereo.Audio:1", src=WavePlayer_0.Out Right}
  ]
 }]
properties {
 assetImportData=&self:AssetImportData
 numChannels=2
 outputFormat=Stereo
 sampleRate=48000
 soundSubmixObject=&Fortnite_AudioAnalysis_Music
}
