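# Sound cue for the non-reactive Radiant mix, part of the FM Radiant audio content.
# A single wave player plays the Radient_MIX sound wave directly as the root, with no modulation, mixing or attenuation nodes.
# A very large max distance keeps it audible across the level at unit volume.
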SoundCue /Game/NeverCook/FM/Radiant/Audio/NonReactive_Radient_MIX_Cue
root=WavePlayer_0
nodes [{id=WavePlayer_0, properties {soundWaveAssetPtr=&/Radient_MIX}}]
properties {maxDistance=2097152, volumeMultiplier=1}
