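# Sound cue that plays the Radient_1BASS wave, apparently a bass layer of music for the NeverCook Radiant content.
# A single wave player is the root, with no mixing, modulation or randomisation nodes.
# It uses the events music sound class, has a very large max distance, and has a submix send to Fortnite's music audio analysis submix.
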SoundCue /Game/NeverCook/FM/Radiant/Audio/Radient_1BASS_Cue
root=WavePlayer_0
nodes [{id=WavePlayer_0, properties {soundWaveAssetPtr=&/Radient_1BASS}}]
properties {
 maxDistance=2097152
 soundClassObject=/Game/Sounds/FortSoundClassesAndModes/Fort_Events_Music
 volumeMultiplier=1
 soundSubmixSends [{minSendLevel=0, soundSubmix=&Fortnite_AudioAnalysis_Music}]
}
