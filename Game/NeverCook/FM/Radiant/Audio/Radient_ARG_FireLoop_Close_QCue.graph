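# Quiet weapon fire-loop cue for the Radient ARG, built from AlienLaser waves and limited by gunfire concurrency.
# A root mixer blends close, medium and distant branches, each with its own attenuation and the third-person ranged-fire sound class.
# Each branch layers a random build, a random shot and an enveloped sustain, with looping enabled only on medium and distant sustains.
# The distant branch reuses the medium branch's waves at lower pitch.
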SoundCue /Game/NeverCook/FM/Radiant/Audio/Radient_ARG_FireLoop_Close_QCue
root=Mixer_17
nodes [
 {
  id=Attenuation_0
  children [Mixer_1]
  properties {attenuationSettings=&Fort_Weapon_Projectile_Medium_Mono_Athena_Att}
 }
 {id=Attenuation_1, children [Mixer_0], properties {attenuationSettings=&Fort_Weapon_Projectile_Close_Athena_Att}}
 {
  id=Attenuation_2
  children [Mixer_2]
  properties {attenuationSettings=&Fort_Weapon_Projectile_Distant_Mono_Athena_Att}
 }
 {
  id=Enveloper_0
  children [Random_3]
  properties {pitchCurve {editorCurveData {keys [{time=0}]}}, volumeCurve {editorCurveData {keys [<at -0.0001 -0.0176>,<at 0.0038 1>]}}}
 }
 {
  id=Enveloper_1
  children [Random_7]
  properties {pitchCurve {editorCurveData {keys [{time=0}]}}, volumeCurve {editorCurveData {keys [<at -0.0001 -0.0176>,<at 0.0038 1>]}}}
 }
 {
  id=Enveloper_2
  children [Random_8]
  properties {pitchCurve {editorCurveData {keys [{time=0}]}}, volumeCurve {editorCurveData {keys [<at -0.0001 -0.0176>,<at 0.0038 1>]}}}
 }
 {id=Mixer_0, children [Random_0,Random_2,Enveloper_0], properties {inputVolume [0.6,0.6,0.5]}}
 {id=Mixer_1, children [Random_5,Random_6,Enveloper_1], properties {inputVolume [0.8,0.6,0.5]}}
 {id=Mixer_17, children [SoundClass_1,SoundClass_0,SoundClass_2], properties {inputVolume [1,0.7,0.4]}}
 {id=Mixer_2, children [Random_1,Random_4,Enveloper_2], properties {inputVolume [0.8,0.6,0.5]}}
 {id=Modulator_0, children [Attenuation_2], properties {pitchMax=0.9, pitchMin=0.9}}
 {id=Random_0, children [WavePlayer_0,WavePlayer_1], properties {weights [1,1]}}
 {id=Random_1, children [WavePlayer_2,WavePlayer_3], properties {weights [1,1]}}
 {id=Random_2, children [WavePlayer_9, WavePlayer_10, WavePlayer_11, WavePlayer_12], properties {weights [1,1,1,1]}}
 {id=Random_3, children [WavePlayer_16, WavePlayer_25], properties {weights [1,1]}}
 {id=Random_4, children [WavePlayer_4,WavePlayer_5,WavePlayer_6,WavePlayer_7], properties {weights [1,1,1,1]}}
 {id=Random_5, children [WavePlayer_17, WavePlayer_18], properties {weights [1,1]}}
 {id=Random_6, children [WavePlayer_19, WavePlayer_20, WavePlayer_21, WavePlayer_22], properties {weights [1,1,1,1]}}
 {id=Random_7, children [WavePlayer_23], properties {weights [1]}}
 {id=Random_8, children [WavePlayer_8], properties {weights [1]}}
 {
  id=SoundClass_0
  children [Attenuation_0]
  properties {soundClassOverride=&Fort_WeaponRanged_Fire_3P_PSM}
 }
 {
  id=SoundClass_1
  children [Attenuation_1]
  properties {soundClassOverride=&Fort_WeaponRanged_Fire_3P_PSM}
 }
 {id=SoundClass_2, children [Modulator_0], properties {soundClassOverride=&Fort_WeaponRanged_Fire_3P_PSM}}
 {id=WavePlayer_0, properties {soundWaveAssetPtr=/UndefeatedWeaponGameplay/Sounds/AlienLaser/AlienLaser_Fire_Close_Build_01}}
 {id=WavePlayer_1, properties {soundWaveAssetPtr=/UndefeatedWeaponGameplay/Sounds/AlienLaser/AlienLaser_Fire_Close_Build_02}}
 {id=WavePlayer_10, properties {soundWaveAssetPtr=/UndefeatedWeaponGameplay/Sounds/AlienLaser/AlienLaser_Fire_Close_02}}
 {id=WavePlayer_11, properties {soundWaveAssetPtr=/UndefeatedWeaponGameplay/Sounds/AlienLaser/AlienLaser_Fire_Close_03}}
 {id=WavePlayer_12, properties {soundWaveAssetPtr=/UndefeatedWeaponGameplay/Sounds/AlienLaser/AlienLaser_Fire_Close_04}}
 {id=WavePlayer_16, properties {soundWaveAssetPtr=/UndefeatedWeaponGameplay/Sounds/AlienLaser/AlienLaser_FireLoop_Close_New_02}}
 {id=WavePlayer_17, properties {soundWaveAssetPtr=/UndefeatedWeaponGameplay/Sounds/AlienLaser/AlienLaser_Fire_Close_Build_Distant_B01}}
 {id=WavePlayer_18, properties {soundWaveAssetPtr=/UndefeatedWeaponGameplay/Sounds/AlienLaser/AlienLaser_Fire_Close_Build_Distant_B02}}
 {id=WavePlayer_19, properties {soundWaveAssetPtr=/UndefeatedWeaponGameplay/Sounds/AlienLaser/AlienLaser_Fire_Medium_01}}
 {id=WavePlayer_2, properties {soundWaveAssetPtr=&AlienLaser_Fire_Close_Build_Distant_B01}}
 {id=WavePlayer_20, properties {soundWaveAssetPtr=/UndefeatedWeaponGameplay/Sounds/AlienLaser/AlienLaser_Fire_Medium_02}}
 {id=WavePlayer_21, properties {soundWaveAssetPtr=/UndefeatedWeaponGameplay/Sounds/AlienLaser/AlienLaser_Fire_Medium_03}}
 {id=WavePlayer_22, properties {soundWaveAssetPtr=/UndefeatedWeaponGameplay/Sounds/AlienLaser/AlienLaser_Fire_Medium_04}}
 {id=WavePlayer_23, properties {bLooping=true, soundWaveAssetPtr=/UndefeatedWeaponGameplay/Sounds/AlienLaser/AlienLaser_FireLoop_Distant_01}}
 {id=WavePlayer_25, properties {soundWaveAssetPtr=/UndefeatedWeaponGameplay/Sounds/AlienLaser/AlienLaser_FireLoop_Close_New_03}}
 {id=WavePlayer_3, properties {soundWaveAssetPtr=&AlienLaser_Fire_Close_Build_Distant_B02}}
 {id=WavePlayer_4, properties {soundWaveAssetPtr=&AlienLaser_Fire_Medium_01}}
 {id=WavePlayer_5, properties {soundWaveAssetPtr=&AlienLaser_Fire_Medium_02}}
 {id=WavePlayer_6, properties {soundWaveAssetPtr=&AlienLaser_Fire_Medium_03}}
 {id=WavePlayer_7, properties {soundWaveAssetPtr=&AlienLaser_Fire_Medium_04}}
 {id=WavePlayer_8, properties {bLooping=true, soundWaveAssetPtr=&AlienLaser_FireLoop_Distant_01}}
 {id=WavePlayer_9, properties {soundWaveAssetPtr=/UndefeatedWeaponGameplay/Sounds/AlienLaser/AlienLaser_Fire_Close_01}}
]
properties {
 bDebug=true
 maxDistance=27500
 pitchMultiplier=1.05
 volumeMultiplier=0.05
 concurrencySet [&Weapon_Gunfire_Local_Conc, &Weapon_Gunfire_Global_Conc]
}
