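; BP_TriggerSequence is a trigger actor that plays a Level Sequence when another actor overlaps it.
; On begin overlap it creates a Level Sequence Player from its Sequence variable and starts playback.
; It does not filter the overlapping actor or guard against repeats, so each overlap creates a new player.
; BeginPlay, Tick and the construction script are empty.
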(event EventBeginPlay)

(event Collision|EventActorBeginOverlap
  (params (OtherActor :type Actor%20Object%20Reference))
  (bind (_outactor _returnvalue) (Sequencer|Player|CreateLevelSequencePlayer (Variables|Default|GetSequence)))
  (Sequencer|Player|Play _returnvalue :!class /Script/MovieScene.MovieSceneSequencePlayer))

(event EventTick
  (params (DeltaSeconds :type Float%20%28single-precision%29)))

(fn ConstructionScript)
