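# Stereo one-shot MetaSound source that plays the Radient_MIX music track for Radiant effects.
# On Play becomes a Harmonix transport that drives a metronome, whose MIDI clock feeds a transport wave player controller to start and stop the wave player.
# The metronome's MIDI clock is also exposed as an output, and the wave player's finish signals OnFinished.
# The sound references the Fortnite_AudioAnalysis_Music submix.
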MetaSoundSource /Game/NeverCook/FM/Radiant/Effects/MSS_Radiant_Music
inputs [UE.Source.OnPlay {type=Trigger}]
interfaces [UE.OutputFormat.Stereo:1.0, UE.Source.OneShot:1.0, UE.Source:1.1]
outputs [
 MIDI Clock {type=MIDIClock}
 "UE.OutputFormat.Stereo.Audio:0" {type=Audio}
 "UE.OutputFormat.Stereo.Audio:1" {type=Audio}
 UE.Source.OneShot.OnFinished {type=Trigger}
]
pages [Default {
  input_defaults {UE.Source.OnPlay=null}
  nodes [
   n1 = HarmonixNodes.TriggerToTransport(Play=^UE.Source.OnPlay)
   HarmonixNodesMetronome_0 = HarmonixNodes.Metronome(Transport=@n1.Transport)
   n2 = HarmonixNodes.TransportWavePlayerController(MIDI Clock=@HarmonixNodesMetronome_0.MIDI Clock, Transport=@n1.Transport) {also {version [0,1]}}
   WavePlayer_0 = UE.Wave Player:Stereo(Play=@n2.Play, Stop=@n2.Stop) {also {defaults {Loop="<MetaSoundBuilderResult.SUCCEEDED: 0>", Wave Asset=/Game/NeverCook/FM/Radiant/Audio/Radient_MIX}}}
  ]
  links [
   {dst="@MIDI Clock", src=HarmonixNodesMetronome_0.MIDI Clock}
   {dst="@UE.Source.OneShot.OnFinished", src=WavePlayer_0.On Finished}
   {dst="@UE.OutputFormat.Stereo.Audio:0", src=WavePlayer_0.Out Left}
   {dst="@UE.OutputFormat.Stereo.Audio:1", src=WavePlayer_0.Out Right}
  ]
 }]
properties {
 assetImportData=&self:AssetImportData
 numChannels=2
 outputFormat=Stereo
 sampleRate=48000
 soundSubmixObject=&Fortnite_AudioAnalysis_Music
}
