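# Translucent, unlit, two-sided test effect material for Niagara mesh particles, drawing glowing edged surfaces with scrolling scanlines.
# An SDF rectangle border, normal-masked scanlines and a view-angle term shape brightness and opacity.
# Animated colors blend toward hue-shifted curve-atlas and particle color via dynamic parameter, with switchable eye-adaptation compensation.
# A WPO branch oscillates upper-bound vertices using music amplitude; many oscillator, wipe and debug experiments remain unconnected.
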Material /Game/NeverCook/FM/Radiant/Effects/M_Barycentric_UV_Testing_A
graph {
 properties {
  bAutomaticallySetUsageInEditor=false
  bUsedWithNiagaraMeshParticles=true
  blendMode=BLEND_Translucent
  shadingModel=MSM_Unlit
  twoSided=true
  usedShadingModels=Unlit
 }
 nodes [
  n1 = Constant()
  n2 = Constant()
  n3 = Multiply(A=Sign(None=VertexNormalWS()), B=Constant3Vector() {constant=<rgba -1 1 1 0>})
  n4 = Add(A=Multiply(A=WorldPosition().XYZ, B=Constant() {r=0.001}), B=Constant())
  n5 = Constant3Vector() {constant=<rgba 1 1 -1 0>}
  n6 = Multiply(A=AppendVector(A=ComponentMask(None=@n3) {b=true}, B=Constant() {r=1}), B=ComponentMask(None=Multiply(A=@n4, B=@n5)) {g=true, r=true})
  n7 = Saturate(None=TextureSample(
    UVs=MaterialFunctionCall!(Stepped Scroll On/Off=@n1, Vertical / Horizontal Scroll, Speed Mult=@n2, UV In=@n6) {materialFunction=/Game/Effects/Niagara/MasterMaterials/Functions/Mojo_MF/MF_FX_Screen_Scroll_Options}.UV Out
    Tex=TextureObject() {texture=/Game/TimeOfDay/LightFunction/T_Cloud2dNoise}
   ).R)
  Add(A=Multiply(A=@n7) {constB=2})
  n8 = TextureCoordinate()
  n9 = Constant() {r=4}
  n10 = Multiply(A=Time(), B=Constant() {desc=Speed Mult, r=0.5}) {constB=5}
  n11 = Subtract(A=Fmod(A=Divide(A=@n10, B=@n9) {constB=4}, B=Constant() {r=1.9}), B=Constant() {r=0.4}) {desc=Y}
  n12 = Subtract(A=ComponentMask(None=@n8) {g=true}, B=@n11)
  n13 = Multiply(A=@n12, B=Constant() {r=110})
  n14 = Constant() {r=2}
  n15 = Add(A=Multiply(A=Power(Base=@n13, Exp=@n14), B=Constant() {r=-1}), B=Constant() {r=0.8}) {desc=str}
  n16 = Constant() {r=0.01}
  n17 = Subtract(A=ComponentMask(None=@n8) {r=true}, B=Clamp(None=Multiply(A=@n15, B=@n16)))
  AppendVector(A=@n17, B=ComponentMask(None=@n8) {g=true})
  ConstantBiasScale()
  Constant() {r=0.5}
  Constant() {r=1}
  Constant() {desc=W, r=0.15}
  Divide(A=WorldPosition().XYZ, B=Constant() {r=50})
  Divide(A=ComponentMask(None=ParticlePositionWS()) {g=true, r=true}) {constB=0.0001}
  n18 = ScalarParameter() {
   defaultValue=1.68
   desc=Eye Adaptation in the old scene was 1.68 - scale to match old appearance
   group=FX Emissive EyeAdaptation
   parameterName=FX Emissive EyeAdaptation Prescale
   sliderMax=4
  }
  n19 = NamedRerouteUsage()
  n20 = Multiply(A=@n19) {constB=2}
  n21 = VectorParameter() {parameterName=Primary Color, defaultValue=<rgba 0 9.9461 25 1>}
  n22 = VectorParameter() {parameterName=Secondary Color, defaultValue=<rgba 0.4295 0 10 1>}
  n23 = Constant3Vector() {constant=<rgba 0.1126 1 0 1>}
  n24 = LinearInterpolate(A=LinearInterpolate(A=@n21.RGB, B=@n22.RGB, Alpha=Saturate(None=@n20)), B=@n23, Alpha=Saturate(None=Subtract(A=@n20)))
  n25 = Multiply(A=@n24, B=Constant3Vector() {constant=<rgba 0.1406 0.1406 0.1406 1>})
  n26 = VertexNormalWS()
  n27 = Constant2Vector() {g=2, r=2}
  n28 = MaterialFunctionCall!(st=Reroute() {desc=st}, s=@n27) {materialFunction=/Game/Effects/Niagara/MasterMaterials/Functions/Mojo_MF/MF_FX_Rect_SDF}
  n29 = Constant() {r=0.05}
  n30 = Multiply(A=@n29, B=Constant() {r=0.5})
  n31 = Constant() {desc=S, r=0.5}
  n32 = Subtract(A=Step(Y=@n31, X=Add(A=@n28.Result, B=@n30)), B=Step(Y=@n31, X=Subtract(A=@n28.Result, B=@n30)))
  n33 = Power(Base=Abs(None=VertexNormalWS()), Exp=Constant() {r=8})
  n34 = Constant3Vector() {constant=<rgba 1 1 1 0>}
  n35 = MaterialFunctionCall(Color=Divide(A=@n33, B=DotProduct(A=@n33, B=@n34))) {materialFunction=&SplitComponents}
  n36 = Multiply(A=Time(), B=ScalarParameter() {defaultValue=0.2, parameterName=Scan Line Speed}) {constB=-0.5, desc=Speed}
  n37 = MaterialFunctionCall(Float2=TextureCoordinate()) {materialFunction=&BreakOutFloat2Components}
  n38 = MaterialFunctionCall(X=@n37.G, Y=OneMinus(None=@n37.R)) {materialFunction=&MakeFloat2}
  n39 = ComponentMask(None=Frac(None=Add(A=@n38.Result, B=@n36))) {g=true}
  n40 = ScalarParameter() {defaultValue=7, parameterName=Scanline Detail Y Tile}
  n41 = ConstantBiasScale(None=Sine(None=Multiply(A=@n39, B=@n40) {desc=Y tile}))
  n42 = Multiply(A=OneMinus(None=@n35.B), B=Multiply(A=Subtract(A=@n41) {constB=0.99}, B=Constant() {r=40}))
  n43 = Multiply(A=@n35.B) {constB=0.5}
  n44 = Max(A=Saturate(None=Add(A=@n42, B=@n43)), B=ComponentMask(None=Clamp(None=@n32)) {r=true})
  n45 = Multiply(A=DotProduct(A=Multiply(A=@n26) {constB=-1}, B=CameraVectorWS()), B=OneMinus(None=@n44) {desc=Ignore Solid Edge/Line Mask})
  n46 = TextureCoordinate()
  n47 = Constant2Vector() {g=1, r=0.1}
  n48 = CurveAtlasRowParameter(CurveTime=ComponentMask(None=Multiply(A=@n46, B=@n47)) {r=true}) {
   atlas=/Game/Creative/Sets/PropSets/Primitives/Customization/Textures/ColorCurve/CC_CM_Primitives_01
   curve=/Game/Creative/Sets/PropSets/Primitives/Customization/Textures/ColorCurve/C_CM_SynthWave_01
   defaultValue=2
   parameterName=ColorCurve
  }
  n49 = Constant() {r=1}
  n50 = Multiply(A=Time(), B=Constant() {desc=Speed Mult, r=0.1}) {constB=5}
  n51 = MaterialFunctionCall(Hue Shift Percentage=Saturate(None=Fmod(A=@n50, B=@n49)), Texture=@n48) {materialFunction=&HueShift}
  n52 = ParticleColor()
  n53 = DynamicParameter() {paramNames [Use Particle Color, Param2, Param3, Param4]}
  n54 = LinearInterpolate(A=LinearInterpolate(A=Constant() {r=0.35}, B=Constant() {r=0.65}, Alpha=@n45), B=Constant() {r=1}, Alpha=@n44)
  n55 = Multiply(
   A=LinearInterpolate(A=LinearInterpolate(A=@n24, B=@n25, Alpha=@n45), B=Multiply(A=@n51.Result, B=@n52.RGB), Alpha=@n53.Use Particle Color)
   B=@n54
  )
  n56 = ComponentMask(None=Multiply(A=@n55, B=@n18) {constB=1.3}) {b=true, g=true, r=true}
  n57 = ScalarParameter() {defaultValue=1, desc=Mix to Original, group=FX Emissive EyeAdaptation, parameterName=FX Emissive EyeAdaptation Alpha}
  EyeAdaptationInverse(LightValueInput=@n56, AlphaInput=@n57) {desc=disabling due to fails on Juno}
  n58 = DepthFade(Opacity=Constant() {r=1}, FadeDistance=Constant() {r=100})
  n59 = VertexColor()
  n60 = Multiply(A=LinearInterpolate(A=Constant() {r=1}, B=@n53.Param4, Alpha=@n52.A), B=Multiply(A=@n54, B=@n59.R))
  LinearInterpolate(A=@n60, B=Constant(), Alpha=Multiply(A=OneMinus(None=@n58)))
  n61 = Subtract(A=ComponentMask(None=@n8) {g=true}, B=@n11)
  n62 = Abs(None=@n61)
  n63 = Constant() {r=0.3}
  n64 = Power(Base=Subtract(A=Constant() {r=1}, B=Power(Base=@n62, Exp=@n63)), Exp=Constant() {r=3}) {desc=colorAdd}
  n65 = Constant() {r=0.5}
  n66 = Constant() {r=10}
  MaterialFunctionCall(Y=Add(A=Multiply(A=@n64, B=@n65)), Z=Add(A=Multiply(A=@n64, B=@n66))) {materialFunction=&MakeFloat3}
  MaterialFunctionCall!(X=@n28.Result, S, W=@n29) {materialFunction=/Game/Effects/Niagara/MasterMaterials/Functions/Mojo_MF/MF_FX_Stroke}
  MaterialFunctionCall(A=@n21.RGB, B=@n22.RGB, C=@n23, Alpha=@n19) {materialFunction=&Lerp_3Color}
  n67 = Saturate(None=Fmod(A=@n10, B=Constant() {r=1}))
  MaterialFunctionCall(Number=@n67) {materialFunction=&DebugScalarValues}
  n68 = TextureCoordinate()
  n69 = MaterialFunctionCall(Input=@n59.R, Input Low=Constant(), Input High=Constant() {r=1}, Target Low=Constant() {r=1}, Target High=Constant() {r=0.7}) {materialFunction=&RemapValueRange}
  Multiply(A=@n68, B=AppendVector(A=ComponentMask(None=@n69.Result) {r=true}, B=Constant() {r=1}))
  n70 = Frac(None=@n10)
  Multiply(A=Add(A=@n70) {constB=-0.5}) {constB=2}
  Multiply(A=@n7) {constB=1000}
  Multiply(A=@n67)
  Multiply(A=@n8, B=Constant2Vector() {g=1, r=0.1})
  n71 = Sine(None=@n10)
  NamedRerouteDeclaration(None=ConstantBiasScale(None=@n71)) {name=Color, nodeColor=<rgba 0.1719 1 0 1>}
  ParticlePositionWS()
  ParticlePositionWS()
  Reroute(None=@n17)
  Reroute(None=@n15)
  n72 = ConstantBiasScale(None=Frac(None=@n10)) {bias=-0.5, scale=2}
  Reroute(None=@n72) {desc=output -1 to 1 loop}
  Reroute(None=Abs(None=@n72)) {desc="Oscillating 0 - 1 "}
  Reroute(None=@n71) {desc=Sin Oscillating -1 - 1}
  n73 = Sine(None=@n10) {period=0.8095}
  n74 = Sine(None=@n10) {period=0.1807}
  Reroute(None=Saturate(None=LinearInterpolate(A=@n71, B=@n73, Alpha=@n74))) {desc=Random Flicker Clamp 0 - 1}
  Reroute(None=Saturate(None=@n11)) {desc=0-1 Loop Clamped Delaypauses for a short time on 0 and 1 bc values extend above and below 0-1 range but are clamped}
  n75 = Add(A=@n8, B=@n10)
  n76 = Multiply(A=ComponentMask(None=Frac(None=@n75)) {g=true}, B=Constant() {desc=Scanline Y Tile, r=1}) {desc=Y tile}
  Reroute(None=ConstantBiasScale(None=Sine(None=@n76))) {desc=Scanline 0-1 Tiling Loop}
  Reroute(None=Saturate(None=@n64)) {desc=Scanline clamp 0-1 loop no Tile}
  Reroute(None=Saturate(None=@n61)) {desc=1-0 Wipe Loop}
  Reroute(None=MaterialFunctionCall(Emissive=Reroute()) {materialFunction=&MF_FX_EmissiveAdjustments}.Result)
  n77 = Subtract(A=Multiply(A=@n68, B=Constant() {r=2}), B=Constant() {r=1})
  n78 = Divide(A=ComponentMask(None=@n77) {r=true}, B=ComponentMask(None=@n27) {r=true})
  n79 = Divide(A=ComponentMask(None=@n77) {g=true}, B=ComponentMask(None=@n27) {g=true})
  Reroute(None=Max(A=Abs(None=@n78), B=Abs(None=@n79)))
  n80 = EyeAdaptation()
  n81 = StaticSwitchParameter(True=LinearInterpolate(A=@n56, B=Divide(A=@n56, B=@n80.EyeAdaptation), Alpha=@n57), False=@n55) {
   defaultValue=true
   desc="Enable/Disable EyeAdaptation (auto exposure) compensation"
   group=FX Emissive EyeAdaptation
   parameterName=FX Emissive EyeAdaptation
  }
  Reroute(None=@n81) {desc=emissive color}
  Reroute(None=@n60) {desc=opacity}
  n82 = LinearInterpolate(Alpha=CollectionParameter() {collection=/Game/Athena/Events/AudioAnalysis/MPC_Audio_Reactive_Parameters, parameterName=Amplitude_Music}) {constA=1, constB=0}
  n83 = ScalarParameter() {defaultValue=0.2, parameterName=TimeMult}
  n84 = Time()
  n85 = Multiply(A=Constant3Vector() {constant=<rgba 0 0 100 0>}, B=Sine(None=Multiply(A=@n83, B=@n84) {constB=5}) {desc=Sin oscilating -1 to 1}) {constB=100}
  n86 = Divide(A=WorldPosition().XYZ, B=ObjectBounds())
  n87 = Power(Base=ComponentMask(None=ConstantBiasScale(None=@n86)) {b=true}, Exp=Constant() {r=100})
  Reroute(None=Multiply(A=Multiply(A=@n82, B=@n85), B=Saturate(None=@n87))) {desc=WPO}
  SmoothStep(Value=@n12) {constMax=0.55, constMin=0.45}
  Step(Y=@n70) {constX=0.5}
  Time()
  WorldPosition()
 ]
}
