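# Material function that builds animated tunnel UVs for FX materials, outputting centered uv, Tunnel UV and Hole Mask.
# Offset texture coordinates sway on a time-driven sine/cosine orbit, then map to scaled Arctangent2 for U and inverse radius plus scrolling time for V.
# Arctangent2 takes G for both X and Y, so R never reaches the angle.
# Clamped squared radius forms the mask; an unconnected circuitboard debug preview samples with pre-warp derivatives.
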MaterialFunction /Game/Effects/Niagara/MasterMaterials/Functions/Mojo_MF/MF_FX_Tunnel_UV_Moving_V1
graph {nodes [
  n1 = Time()
  Frac(None=@n1)
  n2 = Constant() {r=1}
  n3 = TextureCoordinate()
  n4 = FunctionInput(Preview=@n1) {bUsePreviewValueAsDefault=true, inputName=In Time, inputType=FunctionInput_Scalar, sortPriority=2}
  n5 = Multiply(
   A=@n4
   B=FunctionInput(Preview=Constant() {desc=speed, r=0.25}) {bUsePreviewValueAsDefault=true, inputName=Motion Speed, inputType=FunctionInput_Scalar, sortPriority=3}
  )
  n6 = Cosine(None=@n5) {desc=period 9.42, period=9.42}
  n7 = Sine(None=@n5) {desc=period 9.42, period=9.42}
  n8 = Constant() {r=0.3}
  n9 = Add(
   A=Add(
    A=FunctionInput(Preview=@n3) {bUsePreviewValueAsDefault=true, inputName=In TexCoords, inputType=FunctionInput_Vector2, sortPriority=1}
    B=Constant2Vector() {g=-0.75, r=-0.5}
   )
   B=Multiply(
    A=AppendVector(A=@n6, B=@n7)
    B=FunctionInput(Preview=@n8) {bUsePreviewValueAsDefault=true, inputName=Motion Intensity, inputType=FunctionInput_Scalar, sortPriority=4}
   )
  )
  n10 = MaterialFunctionCall(Vector 2=@n9) {materialFunction=/Engine/Functions/Engine_MaterialFunctions02/Utility/VectorLength}
  n11 = FunctionInput(Preview=Constant() {r=0.5}) {bUsePreviewValueAsDefault=true, inputName=Speed Down Tunnel, inputType=FunctionInput_Scalar, sortPriority=7}
  n12 = MaterialFunctionCall(Float2=@n9) {materialFunction=&BreakOutFloat2Components}
  n13 = Arctangent2!(Y=@n12.G, X=@n12.G)
  n14 = FunctionInput(Preview=Constant() {r=3.1416}) {bUsePreviewValueAsDefault=true, inputName=Divide Value, inputType=FunctionInput_Scalar, sortPriority=5}
  n15 = Constant() {r=4}
  n16 = MaterialFunctionCall(
   X=Multiply(
    A=Divide(A=@n13, B=@n14)
    B=FunctionInput(Preview=@n15) {bUsePreviewValueAsDefault=true, inputName=Tile Mult, inputType=FunctionInput_Scalar, sortPriority=6}
   )
   Y=Add(A=Divide(A=@n2, B=@n10.V2 Length), B=Multiply(A=@n4, B=@n11))
  ) {materialFunction=&MakeFloat2}
  FunctionOutput(None=@n16.Result) {outputName=Tunnel UV, sortPriority=2}
  n17 = Min(A=Constant() {r=1}, B=Power(Base=@n10.V2 Length, Exp=Constant() {r=2}))
  FunctionOutput(None=@n17) {outputName=Hole Mask, sortPriority=3}
  FunctionOutput(None=@n9) {outputName=uv, sortPriority=1}
  Multiply(
   A=TextureSample!(
    UVs=@n16.Result
    Tex=TextureObject() {texture=/Game/Effects/Fort_Effects/Textures/T_FX_Circuitboard}
    "DDX(UVs)"=DDX(Value=@n9)
    "DDY(UVs)"=DDY(Value=@n9)
    Apply View MipBias
   ) {mipValueMode=TMVM_Derivative}.RGBA
   B=@n17
  ) {desc=Debug view}
  Saturate(None=@n16.Result)
 ]}
