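# Translucent, unlit, two-sided particle material for sprites and meshes, named as a box impact effect, drawing a music-reactive equalizer.
# Fake FFT noise, driven by audio parameter collections, feeds radial EQ bar boxes on tunnel UVs.
# Color is either hue-cycling tinted by particle color or a three-color gradient from stage colors or presets.
# Switches add sine-wave and edge dots; tunnel pattern, fresnel and depth-fade branches are also present.
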Material /Game/NeverCook/FM/Radiant/Effects/M_EQ_Box_Impact
graph {
 properties {
  bAutomaticallySetUsageInEditor=false
  bUsedWithMeshParticles=true
  bUsedWithNiagaraMeshParticles=true
  bUsedWithNiagaraSprites=true
  bUsedWithParticleSprites=true
  blendMode=BLEND_Translucent
  previewMesh=/Game/Effects/Fort_Effects/Gameplay/Tutorial_Target/Particle_CurvedPlane_Downward
  shadingModel=MSM_Unlit
  twoSided=true
  usedShadingModels=Unlit
 }
 nodes [
  n1 = Constant()
  n2 = MaterialFunctionCall!(In TexCoords, Time, Speed X=@n1, Speed Y=@n1, Size=Constant() {r=0.1}, Divide Control=Constant() {r=0.45}) {materialFunction=/Game/Effects/Niagara/MasterMaterials/Functions/Mojo_MF/MF_FX_Tunnel_UV}
  n3 = Constant() {r=1}
  n4 = Constant2Vector() {g=1, r=0.15}
  n5 = Constant()
  n6 = TextureCoordinate()
  n7 = NamedRerouteDeclaration() {name=FFTTexObject, nodeColor=<rgba 0 1 0.2734 1>}
  n8 = Saturate(None=TextureSample(
    UVs=MaterialFunctionCall!(Center, Maintain Aspect Ratio=@n3, Tiling=@n4, Offset=@n5, In Tex Coords=@n6) {materialFunction=/Game/Effects/Fort_Effects/MaterialFunctions/MF_Uniform_Radial_UV}.Out UV
    Tex=@n7
   ) {texture=/Game/Effects/Fort_Effects/Textures/Fire/T_Fire_Mask}.R)
  n9 = Constant() {r=2}
  n10 = MaterialFunctionCall!(Time Mult=Constant() {r=0.2}, Time Add) {materialFunction=/Game/Effects/Niagara/MasterMaterials/Functions/Mojo_MF/MF_FX_Flicker}
  n11 = Saturate(None=Multiply(
    A=MaterialFunctionCall(In=@n8, Contrast=@n9) {materialFunction=&CheapContrast}.Result
    B=LinearInterpolate(Alpha=@n10.Raw Flicker) {constA=0.65, constB=0.8}
   ))
  Add(A=@n11, B=LinearInterpolate(Alpha=OneMinus(None=@n2.Tunnel Saturated Mask)) {constA=0.01, constB=0.1}) {constB=0, desc=Prevent 0 when no music is playing}
  n12 = MaterialFunctionCall(
   Invert Fresnel=StaticBoolParameter() {defaultValue=true, parameterName=Invert Fresnel}
   Power=ScalarParameter() {defaultValue=2, parameterName=Fresnel Power}
  ) {materialFunction=&Fresnel_Function}
  n13 = NamedRerouteUsage()
  n14 = NamedRerouteUsage()
  n15 = Saturate(None=StaticSwitchParameter(True=NamedRerouteUsage() {desc="New Fancy EQ Mask!"}, False=Multiply(A=@n13, B=@n14) {desc=OLD Radial Beat Circles}) {defaultValue=true, parameterName=Use EQ Bar Boxes?})
  Add(A=@n15, B=Multiply(A=ParticleColor().RGB, B=@n12.Result))
  n16 = Desaturation(
   None=MaterialFunctionCall(Hue Shift Percentage=Constant() {desc=hue shift percent, r=0.575}, Texture=NamedRerouteUsage()) {materialFunction=&HueShift}.Result
   Fraction=Constant() {desc=optional saturate color a bit, r=-0.25}
  )
  n17 = Constant() {r=1}
  n18 = NamedRerouteUsage()
  n19 = Constant() {r=1}
  n20 = NamedRerouteUsage()
  n21 = Constant() {r=1}
  n22 = MaterialFunctionCall!(
   Color 1=StaticSwitch(True=AppendVector(A=@n16, B=@n17), False=Constant4Vector() {constant=<rgba 0 0.9647 0.0529 1>}.RGBA, Value=NamedRerouteUsage())
   Color 2=StaticSwitch(True=AppendVector(A=@n18, B=@n19), False=Constant4Vector() {constant=<rgba 0.0006 0.3325 1 1>}.RGBA, Value=NamedRerouteUsage())
   Color 3=StaticSwitch(True=AppendVector(A=@n20, B=@n21), False=Constant4Vector() {constant=<rgba 0.7305 0.0009 1 1>}.RGBA, Value=NamedRerouteUsage())
   middle
   t
  ) {materialFunction=/Game/Effects/Niagara/MasterMaterials/Functions/Mojo_MF/MF_FX_Lerp3_Color}
  ComponentMask(None=@n22.Result V4) {b=true, g=true, r=true}
  Constant() {r=0.15}
  Constant() {r=0.35}
  Constant() {desc=Sine_Height, r=0.2}
  Constant() {r=0.25}
  Constant()
  Constant() {r=0.5}
  n23 = NamedRerouteUsage()
  DepthFade(Opacity=@n23, FadeDistance=ScalarParameter() {defaultValue=100, parameterName=FadeDistance})
  n24 = ParticleColor()
  n25 = Add(A=Multiply(A=@n2.Tunnel Size Mask, B=NamedRerouteUsage()), B=NamedRerouteUsage())
  n26 = Frac(None=Multiply(A=Time()) {constB=0.15})
  n27 = Multiply(
   A=@n24.RGB
   B=MaterialFunctionCall(Hue Shift Percentage=Add(A=@n25, B=@n26), Texture=Constant3Vector() {constant=<rgba 0.0003 0 0.1458 0>}) {materialFunction=&HueShift}.Result
  )
  n28 = Multiply(A=NamedRerouteUsage()) {constB=0.09, desc=darken}
  n29 = Constant() {r=1}
  n30 = NamedRerouteUsage()
  n31 = Constant() {r=1}
  n32 = NamedRerouteUsage()
  n33 = Constant() {r=1}
  n34 = MaterialFunctionCall(
   Color 1=StaticSwitch(True=AppendVector(A=@n32, B=@n33), False=Constant4Vector() {constant=<rgba 1.3101 0 2 1>}.RGBA, Value=NamedRerouteUsage())
   Color 2=StaticSwitch(True=AppendVector(A=@n28, B=@n29), False=Constant4Vector() {constant=<rgba 0.2772 0 0.526 1>}.RGBA, Value=NamedRerouteUsage())
   Color 3=StaticSwitch(True=AppendVector(A=@n30, B=@n31), False=Constant4Vector() {constant=<rgba 0.2484 0 0.375 1>}.RGBA, Value=NamedRerouteUsage())
   t=Multiply(A=@n2.Tunnel Size Mask) {constB=1.35, desc=Adjust Gradient Center point}
  ) {materialFunction=&MF_FX_Lerp3_Color}
  n35 = Multiply(
   A=Multiply(
    A=StaticSwitchParameter(True=@n27, False=@n34.Result V3) {defaultValue=true, parameterName=UseRainbowEQColor?}
    B=Constant() {desc=Color Boost, r=3}
   )
   B=@n15
  )
  n36 = Constant() {r=0.035}
  n37 = MaterialFunctionCall(Speed X=@n36, Speed Y=@n36) {materialFunction=&MF_FX_Tunnel_UV}
  n38 = TextureObjectParameter() {parameterName=Tube Main Pattern, samplerType=SAMPLERTYPE_Masks, texture=/Game/Global/Textures/Default/Masks_and_Gradients/CircuitBoardSeries}
  n39 = TextureSample!(UVs=@n37.Tunnel UV Static, Tex=@n38, "DDX(UVs)"=DDX(Value=@n37.uv), "DDY(UVs)"=DDY(Value=@n37.uv), Apply View MipBias) {mipValueMode=TMVM_Derivative, texture=/Game/Effects/Fort_Effects/Effects/Characters/Generic/ResIn/DiffusedStylizedGrid}
  n40 = OneMinus(None=Max(A=@n39.R, B=@n39.R))
  n41 = Saturate(None=Multiply(A=Step(Y=@n40) {constX=0.5}) {constB=0.85})
  n42 = NamedRerouteUsage()
  n43 = StaticSwitch(True=@n42, False=VectorParameter() {parameterName=Tube Pattern Main Color, defaultValue=<rgba 1.9636 0 2 1>}.RGB, Value=NamedRerouteUsage())
  n44 = Multiply(A=@n43, B=Multiply(A=@n41, B=@n37.Tunnel Saturated Mask))
  n45 = Frac(None=TextureCoordinate() {uTiling=0.5, vTiling=0.5})
  n46 = MaterialFunctionCall(Float2=Sine(None=@n45)) {materialFunction=&BreakOutFloat2Components}
  n47 = Multiply(A=@n46.R, B=@n46.R)
  n48 = Power(Base=@n47) {constExponent=5}
  n49 = Multiply(A=@n41, B=Abs(None=Multiply(A=@n48) {constB=4})) {desc=Apply edge mask to tube pattern}
  n50 = Power(Base=MaterialFunctionCall!(In Texcoords) {materialFunction=/Game/Effects/Fort_Effects/MaterialFunctions/MF_Circle_SDF}.Result) {constExponent=1.25}
  n51 = StaticSwitch(
   True=Multiply(A=NamedRerouteUsage()) {constB=0.025, desc=Darken Primary Color}
   False=VectorParameter() {parameterName=Intro Screen A BG Color Tint, defaultValue=<rgba 0.0662 0 0.0781 1>}.RGB
   Value=NamedRerouteUsage()
  )
  n52 = Multiply(A=@n51, B=Multiply(A=Multiply(A=@n50) {constB=0.8}, B=@n2.Tunnel Saturated Mask))
  n53 = NamedRerouteUsage()
  n54 = Saturate(None=Multiply(A=NamedRerouteUsage()) {constB=2})
  n55 = Reroute()
  n56 = Saturate(None=@n55)
  n57 = StaticSwitchParameter(
   True=LinearInterpolate(A=@n52, B=LinearInterpolate(A=@n51, B=@n53, Alpha=@n54), Alpha=LinearInterpolate(Alpha=@n56) {constA=0.7})
   False=@n52
  ) {parameterName=Use SDF BG Pattern}
  n58 = Multiply(A=NamedRerouteUsage()) {constB=0.25, desc=Lower contribution}
  LinearInterpolate(A=LinearInterpolate(A=LinearInterpolate(A=@n57, B=@n44, Alpha=@n49), B=Multiply(A=@n58, B=@n43), Alpha=@n58), B=@n35, Alpha=@n15)
  MaterialFunctionCall!(X, S, W) {materialFunction=/Game/Effects/Fort_Effects/MaterialFunctions/MF_Stroke}
  MaterialFunctionCall!(
   In TexCoords
   In Time
   Motion Speed
   Motion Intensity=ScalarParameter() {parameterName=Tunnel Motion Intensity}
   Divide Value
   Tile Mult=ScalarParameter() {defaultValue=1, parameterName=Tunnel Tiling}
   Speed Down Tunnel=Constant() {r=0.085}
  ) {materialFunction=/Game/Effects/Niagara/MasterMaterials/Functions/Mojo_MF/MF_FX_Tunnel_UV_Moving_V1}
  n59 = Constant()
  n60 = Constant() {r=0.1}
  Multiply(A=TextureSample!(
    UVs=MaterialFunctionCall(Speed X=@n59, Speed Y=@n59, Size=@n60) {materialFunction=&MF_FX_Tunnel_UV}.Tunnel UV Static
    Tex
    Apply View MipBias
   ) {texture=/Game/Maps/Textures/checkerGrid}.R)
  n61 = Multiply(A=@n47)
  n62 = Multiply(
   A=Power(Base=Abs(None=@n61), Exp=ScalarParameter() {defaultValue=1, parameterName=Edge Dot Pow Exp}) {constExponent=1}
   B=ScalarParameter() {defaultValue=1.5, parameterName=Edge Dot Mult}
  ) {constB=1.5}
  n63 = Constant() {desc="Tiling ", r=64}
  n64 = Multiply(A=TextureCoordinate(), B=@n63)
  n65 = MaterialFunctionCall!(
   x=MaterialFunctionCall(In Texcoords=Frac(None=@n64)) {materialFunction=&MF_Circle_SDF}.Result
   size=LinearInterpolate(Alpha=Saturate(None=@n62)) {constA=0.8, constB=0, desc=min / max radius values}
  ) {materialFunction=/Game/Effects/Fort_Effects/MaterialFunctions/MF_Fill}
  Multiply(A=@n65.Result)
  Multiply(A=VectorParameter() {parameterName=Tube Pattern Main Color, defaultValue=<rgba 1.9636 0 2 1>}.RGB) {constB=0.025, desc="darken primary color "}
  Multiply(A=@n42) {constB=2, desc=boost}
  Multiply(A=@n15, B=@n23)
  Multiply(A=@n23, B=@n12.Result)
  Multiply(A=@n45, B=Constant2Vector() {g=20, r=40}) {constB=3}
  n66 = TextureObjectParameter() {
   parameterName=BG Pattern Texture 1
   samplerType=SAMPLERTYPE_LinearColor
   texture=/Game/Global/Textures/Default/Masks_and_Gradients/HexPattern
  }
  n67 = MaterialFunctionCall(Speed Down Tunnel=Constant() {r=0.2}) {materialFunction=&MF_FX_Tunnel_UV_Moving_V1}
  n68 = DDX(Value=@n67.uv)
  n69 = DDY(Value=@n67.uv)
  n70 = Time()
  n71 = Frac(None=Multiply(A=@n70) {constB=0.45})
  n72 = LinearInterpolate(
   A=TextureSample!(UVs=@n67.Tunnel UV, Tex=@n66, "DDX(UVs)"=@n68, "DDY(UVs)"=@n69, Apply View MipBias) {mipValueMode=TMVM_Derivative, texture=/Game/Global/Textures/Default/Masks_and_Gradients/T-BattlePass-Tile-Triangles}.R
   B=TextureSample(
    UVs=@n67.Tunnel UV
    Tex=TextureObjectParameter() {parameterName=BG Pattern Texture 2, texture=&T-BattlePass-Tile-Triangles}
    "DDX(UVs)"=@n68
    "DDY(UVs)"=@n69
   ) {mipValueMode=TMVM_Derivative, texture=&T-BattlePass-Tile-Triangles}.R
   Alpha=Abs(None=ConstantBiasScale(None=@n71) {bias=-0.5, scale=2})
  )
  n73 = Multiply(A=@n70) {constB=0.15}
  n74 = Abs(None=ConstantBiasScale(None=Frac(None=@n73)) {bias=-0.5, scale=2})
  n75 = Frac(None=Multiply(A=@n70) {constB=0.25})
  n76 = MaterialFunctionCall(X=@n72, S=LinearInterpolate(Alpha=@n74) {constB=0.1}, W=Abs(None=ConstantBiasScale(None=@n75) {bias=-0.5, scale=2})) {materialFunction=&MF_Stroke}
  n77 = Multiply(A=@n76.Result, B=@n72)
  Multiply(A=@n67.Hole Mask, B=@n77) {desc=test}
  n78 = Frac(None=Multiply(A=ComponentMask(None=@n2.Tunnel UV Static) {g=true}, B=NamedRerouteUsage()))
  n79 = Constant() {r=1}
  n80 = NamedRerouteUsage()
  n81 = Constant() {r=1}
  n82 = Multiply(
   A=If(A=@n78, B=@n80, "A > B"=@n79, "A == B"=@n79, "A < B"=Constant())
   B=If(A=@n78, B=OneMinus(None=@n80), "A > B"=Constant(), "A == B"=@n81, "A < B"=@n81)
  )
  n83 = Frac(None=ComponentMask(None=@n2.Tunnel UV Static) {r=true})
  n84 = Constant() {r=1}
  n85 = NamedRerouteUsage()
  n86 = Constant() {r=1}
  n87 = Multiply(
   A=If(A=@n83, B=@n85, "A > B"=@n84, "A == B"=@n84, "A < B"=Constant())
   B=If(A=@n83, B=OneMinus(None=@n85), "A > B"=Constant(), "A == B"=@n86, "A < B"=@n86)
  )
  n88 = Constant() {r=0.1}
  n89 = Saturate(None=OneMinus(None=MaterialFunctionCall(x=@n2.Tunnel Saturated Mask, size=@n88) {materialFunction=&MF_Fill}.Result)) {desc=mask out center section}
  NamedRerouteDeclaration(None=Multiply(A=Multiply(A=@n82, B=@n87), B=@n89)) {name=BoxMasks, nodeColor=<rgba 0.3828 1 0 1>}
  n90 = MaterialFunctionCall(x=@n2.Tunnel Saturated Mask, size=Constant() {r=0.75}) {materialFunction=&MF_Fill}
  n91 = NamedRerouteUsage()
  n92 = Divide(A=Floor(None=Multiply(A=@n2.Tunnel Saturated Mask, B=@n91)), B=@n91)
  n93 = Divide(A=Floor(None=Multiply(A=@n91, B=@n11)), B=@n91)
  n94 = Constant()
  n95 = Constant() {r=1}
  NamedRerouteDeclaration(None=Multiply(A=Saturate(None=@n90.Result) {desc=mask out max bound for EQ}, B=If(A=@n92, B=@n93, "A > B"=@n94, "A == B"=@n94, "A < B"=@n95))) {name=RevealMask, nodeColor=<rgba 1 0 0.7969 1>}
  NamedRerouteDeclaration(None=@n74) {name=Oscillating01Grad, nodeColor=<rgba 1 0 0.6328 1>}
  NamedRerouteDeclaration(None=@n67.Hole Mask) {name=TunnelMovingHoleGrad, nodeColor=<rgba 1 0 0.75 1>}
  NamedRerouteDeclaration(None=@n11) {name=FFTDebug, nodeColor=<rgba 0 0.4609 1 1>}
  n96 = Constant() {r=15}
  n97 = Multiply(A=TextureCoordinate(), B=Constant2Vector() {g=1})
  n98 = MaterialFunctionCall(Float2=ComponentMask(None=CollectionParameter() {collection=/Game/Athena/Events/AudioAnalysis/MPC_Audio_Reactive_Parameters, parameterName=Spectrum_Music}) {g=true, r=true}) {materialFunction=&BreakOutFloat2Components}
  n99 = CollectionParameter() {collection=/Game/NeverCook/FM/Radiant/Audio/Radiant_Audio_MPC, parameterName=MusicBeatProgress}
  n100 = CollectionParameter() {collection=&MPC_Audio_Reactive_Parameters, parameterName=Amplitude_Music}
  n101 = Saturate(None=Multiply(
    A=MaterialFunctionCall!(
     In TexCoords=@n97
     Fake FFT Noise Texture
     Fake FFT Noise Size
     Fake FFT Noise Speed X=@n98.R
     Fake FFT Noise Speed Y=@n98.R
     Fake FFT Noise Intensity Mult=@n99
     Fake FFT Noise Pow
    ) {materialFunction=/Game/Effects/Niagara/MasterMaterials/Functions/Mojo_MF/MF_FX_Fake_FFT_Noise}.Result Horizontal
    B=LinearInterpolate(Alpha=@n100)
   ))
  n102 = LinearInterpolate(Alpha=@n101) {constB=0.15}
  n103 = NamedRerouteUsage()
  n104 = MaterialFunctionCall!(
   UV
   Square Grid Width=Constant() {r=0.25}
   RandomizeUVPlacement=StaticBool() {value=true}
   RadialUV=StaticBool() {value=true}
   BarAmount Vert=@n96
   BarAmount=@n96
   BarWidth=Constant() {r=0.5}
   MountainHardness=Constant() {r=0.5}
   FFT=NamedRerouteUsage()
   Use Fake FFT Noise=NamedRerouteDeclaration(None=StaticBool() {value=true}) {name=UseFakeFFTNoise, nodeColor=<rgba 0 1 0.3906 1>}
   Fake FFT Noise Speed X
   Fake FFT Noise Pow
   FFT_IntensityMult=LinearInterpolate(Alpha=@n101) {constB=1.65}
   FFT_Add=Max(A=Multiply(A=@n102, B=@n103), B=Constant() {r=0.15}) {desc=Make sure a minimum Ring of bars are displayed in the center}
  ) {materialFunction=/SparksMaps/Effects/Screens/Materials/MF/MF_FX_FFT_EQ}
  NamedRerouteDeclaration(None=Multiply(A=@n104.EQ Bar Boxes, B=NamedRerouteUsage())) {name=EQBarBoxes, nodeColor=<rgba 0.1953 1 0 1>}
  NamedRerouteDeclaration(None=Constant() {desc=BoxCount, r=12}) {name=BoxCount, nodeColor=<rgba 0 0.4844 1 1>}
  NamedRerouteDeclaration(None=Constant() {desc=EdgeWidth_Vert, r=0.15}) {name=EdgeWidth_Vert, nodeColor=<rgba 0.0312 0 1 1>}
  NamedRerouteDeclaration(None=Constant() {desc=EdgeWidth_Horiz, r=0.2}) {name=EdgeWidth_Horiz, nodeColor=<rgba 0.1484 1 0 1>}
  NamedRerouteDeclaration(None=Constant() {desc=ColorRampMult, r=1.5}) {name=ColorRampMult, nodeColor=<rgba 0.0312 0 1 1>}
  NamedRerouteDeclaration(None=@n104.EQ Mountain) {name=EQMountain, nodeColor=<rgba 0.6172 0 1 1>}
  n105 = Constant() {desc=Sine_Smoothstep_Min, r=0.05}
  n106 = Constant() {desc=Sine_Smoothstep_Max}
  n107 = Multiply(A=Constant() {desc=Sine_Speed, r=1}, B=Frac(None=Time()))
  n108 = MaterialFunctionCall!(In UV=MaterialFunctionCall!(Repeat Count=Constant2Vector() {g=1, r=1}, texcoord=TextureCoordinate()) {materialFunction=/Game/Environments/Papaya/Materials/MaterialFunctions/MF_UVMirror}.Result) {materialFunction=/Game/Effects/Niagara/MasterMaterials/Functions/Mojo_MF/MF_FX_Symmetrical_UV}
  n109 = Constant2Vector() {g=0.25}
  n110 = Multiply(A=Time(), B=Constant() {r=0.1})
  n111 = Step(Y=ConstantBiasScale(None=Sine(None=@n110))) {constX=0.5}
  n112 = MaterialFunctionCall(Float2=LinearInterpolate(A=@n108.Symmetric UV Vert, B=Add(A=@n108.Symmetric UV Horiz, B=@n109) {constB=0.25}, Alpha=Saturate(None=@n111))) {materialFunction=&BreakOutFloat2Components}
  n113 = Multiply(A=@n112.R, B=Constant() {desc=Sine_Width, r=3.14})
  n114 = Constant() {r=1}
  n115 = Multiply(
   A=@n55
   B=LinearInterpolate(Alpha=MaterialFunctionCall(Time Mult=@n114) {materialFunction=&MF_FX_Flicker}.Raw Flicker) {constA=0.5, constB=1.5}
  )
  n116 = Saturate(None=NamedRerouteUsage())
  n117 = Multiply(
   A=Sine(None=Add(A=@n107, B=@n113))
   B=LinearInterpolate(Alpha=StaticSwitchParameter(True=@n115, False=@n116) {parameterName=UseBeatForSineWaveHeight?}) {constB=0.08}
  )
  n118 = Saturate(None=Distance(A=Add(A=@n112.G, B=@n117), B=Constant() {r=0.5}))
  n119 = Saturate(None=Multiply(A=SmoothStep(Min=@n105, Max=@n106, Value=@n118)))
  NamedRerouteDeclaration(None=@n119) {name=SineGradMask, nodeColor=<rgba 1 0.3516 0 1>}
  NamedRerouteDeclaration(None=@n89) {name=MiddleHoleMask, nodeColor=<rgba 1 0 0.8906 1>}
  n120 = Saturate(None=MaterialFunctionCall(Time Mult=Constant() {r=0.1}) {materialFunction=&MF_FX_Flicker}.Raw Flicker)
  NamedRerouteDeclaration(None=Multiply(A=@n65.Result, B=LinearInterpolate(Alpha=@n120) {constA=0.45, constB=0.8})) {name=EdgeDotMask, nodeColor=<rgba 0.1719 0 1 1>}
  n121 = Constant() {r=1}
  NamedRerouteDeclaration(None=MaterialFunctionCall(
    Square Grid Width=Constant() {r=0.25}
    RandomizeUVPlacement=StaticBool() {value=true}
    RadialUV=StaticBool()
    BarAmount Vert=@n121
    BarAmount=@n121
    BarWidth=Constant() {r=0.5}
    MountainHardness=Constant() {r=0.5}
    FFT=NamedRerouteUsage()
    Use Fake FFT Noise=NamedRerouteUsage()
   ) {materialFunction=&MF_FX_FFT_EQ}."FFT Sampled (Mountain)") {name=SampledFFTValue, nodeColor=<rgba 1 0.3516 0 1>}
  NamedRerouteDeclaration(None=Multiply(A=@n65.Result, B=@n22.Result V3) {desc=apply color}) {name=EdgeDotColor, nodeColor=<rgba 0.4297 1 0 1>}
  NamedRerouteDeclaration(None=@n8) {name=FFT_Sample, nodeColor=<rgba 0.125 0 1 1>}
  NamedRerouteDeclaration() {name=ColorOptionBool, nodeColor=<rgba 1 0 0.7734 1>}
  n122 = MaterialFunctionCall!(CustomUPosition) {materialFunction=/SparksStageShow/Systems/StageManager/FXManager/MaterialFunctions/MF_MojoStageColors}
  NamedRerouteDeclaration(None=@n122.Primary) {name=PrimaryColorLut, nodeColor=<rgba 0.7812 0 1 1>}
  NamedRerouteDeclaration(None=@n122.Secondary) {name=SecondaryColorLut, nodeColor=<rgba 1 0.0703 0 1>}
  n123 = ComponentMask(None=MaterialFunctionCall() {materialFunction=&MF_FX_Lerp3_Color}.Result V4) {b=true, g=true, r=true}
  n124 = Constant() {desc="Tiling ", r=64}
  n125 = Multiply(A=TextureCoordinate(), B=@n124)
  n126 = MaterialFunctionCall(
   x=MaterialFunctionCall(In Texcoords=Frac(None=@n125)) {materialFunction=&MF_Circle_SDF}.Result
   size=LinearInterpolate(Alpha=@n119) {constB=0.9, desc=min / max radius values}
  ) {materialFunction=&MF_Fill}
  NamedRerouteDeclaration(None=Multiply(A=@n126.Result, B=@n123) {desc=apply color}) {name=SineDotColor, nodeColor=<rgba 0.7344 1 0 1>}
  NamedRerouteDeclaration(None=@n122.Tertiary) {name=TertiaryColorLut, nodeColor=<rgba 0 1 0.3906 1>}
  NamedRerouteDeclaration(None=@n24.A) {name=ParticleAlpha, nodeColor=<rgba 1 0.9609 0 1>}
  NamedRerouteDeclaration(None=@n126.Result) {name=SineDotMask, nodeColor=<rgba 0 1 0.6016 1>}
  NamedRerouteDeclaration(None=@n123) {name=SineGradColor, nodeColor=<rgba 0 0.5781 1 1>}
  NamedRerouteDeclaration(None=Constant() {desc=ColorRampSlide, r=-0.7}) {name=ColorRampSlide, nodeColor=<rgba 0 0.0625 1 1>}
  n127 = Multiply(A=SmoothStep(Min=Constant() {desc=Sine_Smoothstep_Min_BG, r=0.08}, Max=Constant() {desc=Sine_Smoothstep_Max_BG}, Value=@n118)) {constB=1.5}
  NamedRerouteDeclaration(None=Saturate(None=OneMinus(None=@n127))) {name=SineWaveBGDarken, nodeColor=<rgba 0.6172 0 1 1>}
  NamedRerouteDeclaration(None=@n77) {name=BGPatternTunnelMask, nodeColor=<rgba 1 0.75 0 1>}
  n128 = Multiply(A=@n67.Hole Mask, B=@n76.Result) {desc=hole grad mask}
  n129 = MaterialFunctionCall(middle=Constant() {r=0.01}, t=@n128) {materialFunction=&MF_FX_Lerp3_Color}
  NamedRerouteDeclaration(None=Multiply(A=ComponentMask(None=@n129.Result V4) {b=true, g=true, r=true}, B=@n128)) {name=BGPatternTunnelColor, nodeColor=<rgba 0 1 0.4141 1>}
  NamedRerouteUsage()
  NamedRerouteUsage()
  n130 = Power(Base=@n47) {constExponent=1.5}
  n131 = Multiply(A=@n35, B=@n2.Tunnel Saturated Mask)
  n132 = Multiply(A=Multiply(A=NamedRerouteUsage(), B=@n131), B=Abs(None=Multiply(A=@n130) {constB=4}))
  n133 = NamedRerouteUsage()
  n134 = NamedRerouteUsage()
  n135 = NamedRerouteUsage()
  n136 = LinearInterpolate(A=Multiply(A=@n134, B=Multiply(A=@n133) {constB=0.45, desc=Reduce Glow Intensity}), B=NamedRerouteUsage(), Alpha=@n135)
  n137 = Saturate(None=Max(A=@n134, B=@n135) {desc=sine combined mask})
  n138 = Abs(None=Multiply(A=@n47) {constB=4})
  n139 = Multiply(
   A=StaticSwitchParameter(True=LinearInterpolate(A=@n132, B=@n136, Alpha=@n137), False=@n131) {parameterName=Use Sine Wave Dots?}
   B=Saturate(None=Power(Base=@n138) {constExponent=0.7})
  )
  n140 = NamedRerouteUsage()
  n141 = NamedRerouteUsage()
  Reroute(None=StaticSwitchParameter(True=LinearInterpolate(A=@n139, B=@n140, Alpha=@n141) {desc=dots}, False=@n139) {parameterName=Use Edge Dots?}) {desc=Out emissive color}
  Reroute(None=@n7)
  Reroute(None=Divide(A=Floor(None=@n64), B=@n63))
  Reroute(None=Divide(A=Floor(None=@n125), B=@n124))
  StaticBool() {value=true}
  TextureObjectParameter() {parameterName=T_FFT, samplerType=SAMPLERTYPE_Grayscale, texture=&LowResBlurredNoise}
 ]
}
