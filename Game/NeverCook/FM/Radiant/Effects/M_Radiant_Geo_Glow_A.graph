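# Translucent, two-sided unlit material for glowing Radiant effect geometry, built for Niagara mesh particles.
# Scrolling scanlines and a world-normal solid-edge mask combine with view-facing shading to drive brightness and opacity over a three-color gradient.
# Particle color, alpha, dynamic parameters and red vertex color modulate the result, with switchable eye-adaptation compensation on emissive.
# A sine-driven, height-masked vertical offset feeds the WPO reroute, alongside leftover utility loop, flicker and rectangle-SDF branches.
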Material /Game/NeverCook/FM/Radiant/Effects/M_Radiant_Geo_Glow_A
graph {
 properties {
  bAutomaticallySetUsageInEditor=false
  bUsedWithNiagaraMeshParticles=true
  blendMode=BLEND_Translucent
  previewMesh=&/SM_Radient_Tube_Cylinder_A
  shadingModel=MSM_Unlit
  twoSided=true
  usedShadingModels=Unlit
 }
 nodes [
  n1 = TextureCoordinate()
  n2 = Constant() {r=4}
  n3 = Multiply(A=Time(), B=Constant() {desc=Speed Mult, r=0.5}) {constB=5}
  n4 = Subtract(A=Fmod(A=Divide(A=@n3, B=@n2) {constB=4}, B=Constant() {r=1.9}), B=Constant() {r=0.4}) {desc=Y}
  n5 = Subtract(A=ComponentMask(None=@n1) {g=true}, B=@n4)
  n6 = Multiply(A=@n5, B=Constant() {r=110})
  n7 = Constant() {r=2}
  n8 = Add(A=Multiply(A=Power(Base=@n6, Exp=@n7), B=Constant() {r=-1}), B=Constant() {r=0.8}) {desc=str}
  n9 = Constant() {r=0.01}
  n10 = Subtract(A=ComponentMask(None=@n1) {r=true}, B=Clamp(None=Multiply(A=@n8, B=@n9)))
  AppendVector(A=@n10, B=ComponentMask(None=@n1) {g=true})
  ConstantBiasScale()
  Constant() {desc=W, r=0.15}
  Constant() {r=1}
  Constant() {r=0.5}
  Divide(A=WorldPosition().XYZ, B=Constant() {r=50})
  n11 = ScalarParameter() {
   defaultValue=1.68
   desc=Eye Adaptation in the old scene was 1.68 - scale to match old appearance
   group=FX Emissive EyeAdaptation
   parameterName=FX Emissive EyeAdaptation Prescale
   sliderMax=4
  }
  n12 = NamedRerouteUsage()
  n13 = Multiply(A=@n12) {constB=2}
  n14 = VectorParameter() {parameterName=Primary Color, defaultValue=<rgba 0 9.9461 25 1>}
  n15 = VectorParameter() {parameterName=Secondary Color, defaultValue=<rgba 0.4295 0 10 1>}
  n16 = Constant3Vector() {constant=<rgba 0.1126 1 0 1>}
  n17 = LinearInterpolate(A=LinearInterpolate(A=@n14.RGB, B=@n15.RGB, Alpha=Saturate(None=@n13)), B=@n16, Alpha=Saturate(None=Subtract(A=@n13)))
  n18 = Multiply(A=@n17, B=Constant3Vector() {constant=<rgba 0.1406 0.1406 0.1406 1>})
  n19 = VertexNormalWS()
  n20 = Power(Base=Abs(None=VertexNormalWS()), Exp=Constant() {r=8})
  n21 = Constant3Vector() {constant=<rgba 1 1 1 0>}
  n22 = MaterialFunctionCall(Color=Divide(A=@n20, B=DotProduct(A=@n20, B=@n21))) {materialFunction=&SplitComponents}
  n23 = OneMinus(None=@n22.B)
  n24 = Time()
  n25 = ScalarParameter() {defaultValue=0.2, parameterName=Scan Line Speed}
  n26 = MaterialFunctionCall(Float2=TextureCoordinate()) {materialFunction=&BreakOutFloat2Components}
  n27 = OneMinus(None=@n26.R)
  n28 = Frac(None=Add(
    A=MaterialFunctionCall(X=@n26.G, Y=@n27) {materialFunction=&MakeFloat2}.Result
    B=Multiply(A=@n24, B=@n25) {constB=-0.5, desc=Speed}
   ))
  n29 = Sine(None=Multiply(A=ComponentMask(None=@n28) {g=true}, B=ScalarParameter() {defaultValue=7, parameterName=Scanline Detail Y Tile}) {desc=Y tile})
  n30 = Multiply(A=Subtract(A=ConstantBiasScale(None=@n29)) {constB=0.99}, B=Constant() {r=40})
  n31 = Saturate(None=Add(A=Multiply(A=@n23, B=@n30), B=Multiply(A=@n22.B) {constB=0.5}))
  n32 = Multiply(A=DotProduct(A=Multiply(A=@n19) {constB=-1}, B=CameraVectorWS()), B=OneMinus(None=@n31) {desc=Ignore Solid Edge/Line Mask})
  n33 = ParticleColor()
  n34 = DynamicParameter() {defaultValue=<rgba 0 1 1 1>, paramNames [Use Particle Color, Param2, Param3, Param4]}
  n35 = LinearInterpolate(A=LinearInterpolate(A=Constant() {r=0.35}, B=Constant() {r=0.65}, Alpha=@n32), B=Constant() {r=1}, Alpha=@n31)
  n36 = Multiply(A=LinearInterpolate(A=LinearInterpolate(A=@n17, B=@n18, Alpha=@n32), B=@n33.RGB, Alpha=@n34.Use Particle Color), B=@n35)
  n37 = ComponentMask(None=Multiply(A=@n36, B=@n11) {constB=1.3}) {b=true, g=true, r=true}
  n38 = ScalarParameter() {defaultValue=1, desc=Mix to Original, group=FX Emissive EyeAdaptation, parameterName=FX Emissive EyeAdaptation Alpha}
  EyeAdaptationInverse(LightValueInput=@n37, AlphaInput=@n38) {desc=disabling due to fails on Juno}
  n39 = DepthFade(Opacity=Constant() {r=1}, FadeDistance=Constant() {r=100})
  n40 = VertexColor()
  n41 = Multiply(A=LinearInterpolate(A=Constant() {r=1}, B=@n34.Param4, Alpha=@n33.A), B=Multiply(A=@n35, B=@n40.R))
  LinearInterpolate(A=@n41, B=Constant(), Alpha=Multiply(A=OneMinus(None=@n39)))
  n42 = Subtract(A=ComponentMask(None=@n1) {g=true}, B=@n4)
  n43 = Abs(None=@n42)
  n44 = Constant() {r=0.3}
  n45 = Power(Base=Subtract(A=Constant() {r=1}, B=Power(Base=@n43, Exp=@n44)), Exp=Constant() {r=3}) {desc=colorAdd}
  n46 = Constant() {r=0.5}
  n47 = Constant() {r=10}
  MaterialFunctionCall(Y=Add(A=Multiply(A=@n45, B=@n46)), Z=Add(A=Multiply(A=@n45, B=@n47))) {materialFunction=&MakeFloat3}
  n48 = Constant2Vector() {g=2, r=2}
  n49 = MaterialFunctionCall!(st=Reroute() {desc=st}, s=@n48) {materialFunction=/Game/Effects/Niagara/MasterMaterials/Functions/Mojo_MF/MF_FX_Rect_SDF}
  n50 = Constant() {r=0.05}
  MaterialFunctionCall!(X=@n49.Result, S, W=@n50) {materialFunction=/Game/Effects/Niagara/MasterMaterials/Functions/Mojo_MF/MF_FX_Stroke}
  MaterialFunctionCall(A=@n14.RGB, B=@n15.RGB, C=@n16, Alpha=@n12) {materialFunction=&Lerp_3Color}
  n51 = Saturate(None=Fmod(A=@n3, B=Constant() {r=1}))
  MaterialFunctionCall(Number=@n51) {materialFunction=&DebugScalarValues}
  n52 = Multiply(A=@n50, B=Constant() {r=0.5})
  n53 = Constant() {desc=S, r=0.5}
  n54 = Subtract(A=Step(Y=@n53, X=Add(A=@n49.Result, B=@n52)), B=Step(Y=@n53, X=Subtract(A=@n49.Result, B=@n52)))
  Max(A=@n31, B=ComponentMask(None=Clamp(None=@n54)) {r=true})
  n55 = Frac(None=@n3)
  Multiply(A=Add(A=@n55) {constB=-0.5}) {constB=2}
  n56 = TextureCoordinate()
  n57 = MaterialFunctionCall(Input=@n40.R, Input Low=Constant(), Input High=Constant() {r=1}, Target Low=Constant() {r=1}, Target High=Constant() {r=0.7}) {materialFunction=&RemapValueRange}
  Multiply(A=@n56, B=AppendVector(A=ComponentMask(None=@n57.Result) {r=true}, B=Constant() {r=1}))
  n58 = Sine(None=@n3)
  NamedRerouteDeclaration(None=ConstantBiasScale(None=@n58)) {name=Color, nodeColor=<rgba 0.1719 1 0 1>}
  ParticlePositionWS()
  Reroute(None=MaterialFunctionCall(Emissive=Reroute()) {materialFunction=&MF_FX_EmissiveAdjustments}.Result)
  Reroute(None=@n10)
  Reroute(None=@n8)
  Reroute(None=@n51) {desc=output 0-1 loop}
  n59 = ConstantBiasScale(None=Frac(None=@n3)) {bias=-0.5, scale=2}
  Reroute(None=@n59) {desc=output -1 to 1 loop}
  Reroute(None=Abs(None=@n59)) {desc="Oscillating 0 - 1 "}
  Reroute(None=@n58) {desc=Sin Oscillating -1 - 1}
  n60 = Sine(None=@n3) {period=0.8095}
  n61 = Sine(None=@n3) {period=0.1807}
  Reroute(None=Saturate(None=LinearInterpolate(A=@n58, B=@n60, Alpha=@n61))) {desc=Random Flicker Clamp 0 - 1}
  Reroute(None=Saturate(None=@n4)) {desc=0-1 Loop Clamped Delaypauses for a short time on 0 and 1 bc values extend above and below 0-1 range but are clamped}
  n62 = Add(A=@n1, B=@n3)
  n63 = Multiply(A=ComponentMask(None=Frac(None=@n62)) {g=true}, B=Constant() {desc=Scanline Y Tile, r=1}) {desc=Y tile}
  Reroute(None=ConstantBiasScale(None=Sine(None=@n63))) {desc=Scanline 0-1 Tiling Loop}
  Reroute(None=Saturate(None=@n45)) {desc=Scanline clamp 0-1 loop no Tile}
  Reroute(None=Saturate(None=@n42)) {desc=1-0 Wipe Loop}
  n64 = Subtract(A=Multiply(A=@n56, B=Constant() {r=2}), B=Constant() {r=1})
  n65 = Divide(A=ComponentMask(None=@n64) {r=true}, B=ComponentMask(None=@n48) {r=true})
  n66 = Divide(A=ComponentMask(None=@n64) {g=true}, B=ComponentMask(None=@n48) {g=true})
  Reroute(None=Max(A=Abs(None=@n65), B=Abs(None=@n66)))
  n67 = EyeAdaptation()
  n68 = StaticSwitchParameter(True=LinearInterpolate(A=@n37, B=Divide(A=@n37, B=@n67.EyeAdaptation), Alpha=@n38), False=@n36) {
   defaultValue=true
   desc="Enable/Disable EyeAdaptation (auto exposure) compensation"
   group=FX Emissive EyeAdaptation
   parameterName=FX Emissive EyeAdaptation
  }
  Reroute(None=@n68) {desc=emissive color}
  Reroute(None=@n41) {desc=opacity}
  n69 = Sine(None=Multiply(A=ScalarParameter() {defaultValue=0.2, parameterName=TimeMult}, B=Time()) {constB=5}) {desc=Sin oscilating -1 to 1}
  n70 = Constant3Vector() {constant=<rgba 0 0 100 0>}
  n71 = Divide(A=WorldPosition().XYZ, B=ObjectBounds())
  n72 = Power(Base=ComponentMask(None=ConstantBiasScale(None=@n71)) {b=true}, Exp=Constant() {r=100})
  Reroute(None=Multiply(A=Multiply(A=@n69, B=@n70) {constB=100}, B=Saturate(None=@n72))) {desc=WPO}
  SmoothStep(Value=@n5) {constMax=0.55, constMin=0.45}
  Step(Y=@n55) {constX=0.5}
  Time()
 ]
}
