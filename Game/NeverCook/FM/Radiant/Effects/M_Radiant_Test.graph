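# Translucent unlit test particle material for Radiant effects, usable on mesh and Niagara mesh particles, producing a hue-cycling glow.
# A SynthWave curve atlas color is hue-shifted by looping time, then multiplied by particle color.
# A static switch adds prescaled eye-adaptation compensation to keep emissive brightness consistent under auto exposure.
# Particle alpha times an invertible Fresnel feeds a depth fade, and several leftover nodes appear unused.
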Material /Game/NeverCook/FM/Radiant/Effects/M_Radiant_Test
graph {
 properties {
  bAutomaticallySetUsageInEditor=false
  bUsedWithMeshParticles=true
  bUsedWithNiagaraMeshParticles=true
  blendMode=BLEND_Translucent
  shadingModel=MSM_Unlit
  usedShadingModels=Unlit
 }
 nodes [
  n1 = ParticleColor()
  n2 = Multiply(
   A=@n1.A
   B=MaterialFunctionCall(
    Invert Fresnel=StaticBoolParameter() {defaultValue=true, parameterName=Invert Fresnel}
    Power=ScalarParameter() {defaultValue=2, parameterName=Fresnel Power}
   ) {materialFunction=&Fresnel_Function}.Result
  )
  DepthFade(Opacity=@n2, FadeDistance=ScalarParameter() {defaultValue=100, parameterName=FadeDistance})
  Divide(A=ComponentMask(None=ParticlePositionWS()) {g=true, r=true}) {constB=0.01}
  n3 = ScalarParameter() {
   defaultValue=1.68
   desc=Eye Adaptation in the old scene was 1.68 - scale to match old appearance
   group=FX Emissive EyeAdaptation
   parameterName=FX Emissive EyeAdaptation Prescale
   sliderMax=4
  }
  n4 = TextureCoordinate()
  n5 = Constant2Vector() {g=1}
  n6 = CurveAtlasRowParameter(CurveTime=ComponentMask(None=Multiply(A=@n4, B=@n5)) {r=true}) {
   atlas=/Game/Creative/Sets/PropSets/Primitives/Customization/Textures/ColorCurve/CC_CM_Primitives_01
   curve=/Game/Creative/Sets/PropSets/Primitives/Customization/Textures/ColorCurve/C_CM_SynthWave_01
   defaultValue=2
   parameterName=ColorCurve
  }
  n7 = Time()
  n8 = Fmod(A=Multiply(A=Add(A=@n7) {constB=0.5}, B=Constant() {desc=Speed Mult, r=0.1}) {constB=5}, B=Constant() {r=1})
  n9 = Multiply(
   A=MaterialFunctionCall(Hue Shift Percentage=Saturate(None=@n8), Texture=@n6) {materialFunction=&HueShift}.Result
   B=@n1.RGB
  )
  n10 = ComponentMask(None=Multiply(A=@n9, B=@n3) {constB=1.3}) {b=true, g=true, r=true}
  n11 = ScalarParameter() {defaultValue=1, desc=Mix to Original, group=FX Emissive EyeAdaptation, parameterName=FX Emissive EyeAdaptation Alpha}
  EyeAdaptationInverse(LightValueInput=@n10, AlphaInput=@n11) {desc=disabling due to fails on Juno}
  Multiply(A=VertexColor(), B=LinearInterpolate(Alpha=@n2))
  Reroute(None=MaterialFunctionCall(Emissive=Reroute()) {materialFunction=&MF_FX_EmissiveAdjustments}.Result)
  n12 = EyeAdaptation()
  StaticSwitchParameter(True=LinearInterpolate(A=@n10, B=Divide(A=@n10, B=@n12.EyeAdaptation), Alpha=@n11), False=@n9) {
   defaultValue=true
   desc="Enable/Disable EyeAdaptation (auto exposure) compensation"
   group=FX Emissive EyeAdaptation
   parameterName=FX Emissive EyeAdaptation
  }
  TextureSampleParameter2D() {parameterName=Param}
  WorldPosition()
 ]
}
