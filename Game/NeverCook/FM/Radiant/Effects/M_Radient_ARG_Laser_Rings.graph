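# Unlit, two-sided additive effect material for laser rings, usable on mesh particles, Niagara meshes and spline meshes.
# World position offset combines SplineThicken, widened by particle alpha, with panning Perlin-noise wobble faded along UV V.
# Emissive mixes a gradient mask, view-to-tangent falloff, time-hue-shifted color-curve sample and particle color.
# An alpha chain blends Start and End Alpha along V; several time-loop, scanline and debug branches appear unused.
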Material /Game/NeverCook/FM/Radiant/Effects/M_Radient_ARG_Laser_Rings
graph {
 properties {
  bCastRayTracedShadows=false
  bUsedWithMeshParticles=true
  bUsedWithNiagaraMeshParticles=true
  bUsedWithSplineMeshes=true
  blendMode=BLEND_Additive
  pixelDepthOffsetMode=PDOM_Legacy
  shadingModel=MSM_Unlit
  twoSided=true
  usedShadingModels=Unlit
 }
 nodes [
  n1 = ScalarParameter() {defaultValue=0.5, parameterName=SplineWidthTip}
  n2 = ScalarParameter() {defaultValue=3, parameterName=SplineWidthBase}
  n3 = ParticleColor()
  n4 = Power(Base=@n3.A) {constExponent=0.5}
  n5 = TextureCoordinate()
  n6 = Power(Base=ComponentMask(None=@n5) {g=true}, Exp=ScalarParameter() {defaultValue=2, parameterName=Power})
  n7 = ScalarParameter() {defaultValue=2.5, parameterName=WPO Wobble}
  n8 = ParticleRandom()
  n9 = ScalarParameter() {defaultValue=1, parameterName=WPO Wobble Pan Speed}
  n10 = Constant2Vector() {g=1, r=0.05}
  n11 = Multiply(
   A=Panner(Coordinate=Add(A=@n5, B=@n8), Speed=Multiply(A=@n9, B=@n10)) {speedX=0.1, speedY=2}
   B=ScalarParameter() {defaultValue=2, parameterName=WPO Wobble Tile}
  )
  n12 = Transform(None=ConstantBiasScale(None=TextureSample(UVs=@n11) {samplerType=SAMPLERTYPE_Masks, texture=/Game/Global/Textures/Default/Noise/T_Generic_Perlin}.RGB) {bias=-0.5, scale=2})
  Add(
   A=MaterialFunctionCall(WidthBase=LinearInterpolate(A=@n1, B=@n2, Alpha=@n4), WidthTip=@n1, UVs for Thickness=@n5) {materialFunction=&SplineThicken}.WorldPositionOffset
   B=Multiply(A=OneMinus(None=@n6), B=Multiply(A=@n7, B=@n12))
  )
  n13 = TextureCoordinate()
  n14 = Constant() {r=4}
  n15 = Multiply(A=Time(), B=Constant() {desc=Speed Mult, r=0.1}) {constB=5}
  n16 = Subtract(A=Fmod(A=Divide(A=@n15, B=@n14) {constB=4}, B=Constant() {r=1.9}), B=Constant() {r=0.4}) {desc=Y}
  n17 = Subtract(A=ComponentMask(None=@n13) {g=true}, B=@n16)
  n18 = Multiply(A=@n17, B=Constant() {r=110})
  n19 = Constant() {r=2}
  n20 = Add(A=Multiply(A=Power(Base=@n18, Exp=@n19), B=Constant() {r=-1}), B=Constant() {r=0.8}) {desc=str}
  n21 = Constant() {r=0.01}
  n22 = Subtract(A=ComponentMask(None=@n13) {r=true}, B=Clamp(None=Multiply(A=@n20, B=@n21)))
  AppendVector(A=@n22, B=ComponentMask(None=@n13) {g=true})
  ConstantBiasScale()
  n23 = TextureCoordinate()
  CurveAtlasRowParameter(CurveTime=MaterialFunctionCall(Float2=@n23) {materialFunction=&BreakOutFloat2Components}.R) {atlas=/Game/Athena/LinearColorCurves/CA_Contrail_Abduction, curve=/Game/Athena/LinearColorCurves/Curve_Contrail_Abduction_TractorBeam_Ring}
  Divide(A=WorldPosition().XYZ, B=Constant() {r=50})
  DynamicParameter()
  n24 = TextureSample(UVs=MaterialFunctionCall(XYZ=@n23) {materialFunction=&Swizzle}.YX) {texture=/Game/Global/Textures/Default/Masks_and_Gradients/gradient_stasis}
  n25 = Abs(None=DotProduct(A=CameraVectorWS(), B=VertexTangentWS()))
  n26 = Saturate(None=Multiply(A=OneMinus(None=@n25)) {constB=2})
  n27 = Multiply(A=Multiply(A=@n24.G, B=@n24.G), B=@n26)
  n28 = TextureCoordinate()
  n29 = Constant2Vector() {g=1, r=0.001}
  n30 = CurveAtlasRowParameter(CurveTime=ComponentMask(None=Multiply(A=@n28, B=@n29)) {r=true}) {
   atlas=/Game/Creative/Sets/PropSets/Primitives/Customization/Textures/ColorCurve/CC_CM_Primitives_01
   curve=/Game/Creative/Sets/PropSets/Primitives/Customization/Textures/ColorCurve/C_CM_SynthWave_01
   defaultValue=2
   parameterName=ColorCurve
  }
  n31 = Saturate(None=Fmod(A=@n15, B=Constant() {r=1}))
  n32 = MaterialFunctionCall(Hue Shift Percentage=@n31, Texture=@n30) {materialFunction=&HueShift}
  MaterialFunctionCall(Emissive=Multiply(A=LinearInterpolate(Alpha=@n27) {constA=0.2}, B=Multiply(A=@n32.Result, B=@n3.RGB))) {materialFunction=&MF_FX_EmissiveAdjustments}
  MaterialFunctionCall(Number=@n31) {materialFunction=&DebugScalarValues}
  n33 = Subtract(A=ComponentMask(None=@n13) {g=true}, B=@n16)
  n34 = Abs(None=@n33)
  n35 = Constant() {r=0.3}
  n36 = Power(Base=Subtract(A=Constant() {r=1}, B=Power(Base=@n34, Exp=@n35)), Exp=Constant() {r=3}) {desc=colorAdd}
  n37 = Constant() {r=0.5}
  n38 = Constant() {r=10}
  MaterialFunctionCall(Y=Add(A=Multiply(A=@n36, B=@n37)), Z=Add(A=Multiply(A=@n36, B=@n38))) {materialFunction=&MakeFloat3}
  n39 = ScalarParameter() {parameterName=End Alpha}
  n40 = ScalarParameter() {defaultValue=4, parameterName=Start Alpha}
  Multiply(A=Multiply(A=Multiply(A=@n24.G, B=@n26), B=LinearInterpolate(A=@n39, B=@n40, Alpha=@n6) {constB=5}), B=@n3.A)
  n41 = Frac(None=@n15)
  Multiply(A=Add(A=@n41) {constB=-0.5}) {constB=2}
  Multiply(A=@n31)
  Multiply(A=@n13, B=Constant2Vector() {g=0.1, r=0.1})
  Reroute(None=@n22)
  Reroute(None=@n20)
  n42 = ConstantBiasScale(None=Frac(None=@n15)) {bias=-0.5, scale=2}
  Reroute(None=@n42) {desc=output -1 to 1 loop}
  Reroute(None=Abs(None=@n42)) {desc="Oscillating 0 - 1 "}
  n43 = Sine(None=@n15)
  Reroute(None=@n43) {desc=Sin Oscillating -1 - 1}
  Reroute(None=ConstantBiasScale(None=@n43)) {desc=Sin Oscillating 0 - 1}
  n44 = Sine(None=@n15) {period=0.8095}
  n45 = Sine(None=@n15) {period=0.1807}
  Reroute(None=Saturate(None=LinearInterpolate(A=@n43, B=@n44, Alpha=@n45))) {desc=Random Flicker Clamp 0 - 1}
  Reroute(None=Saturate(None=@n16)) {desc=0-1 Loop Clamped Delaypauses for a short time on 0 and 1 bc values extend above and below 0-1 range but are clamped}
  n46 = Add(A=@n13, B=@n15)
  n47 = Multiply(A=ComponentMask(None=Frac(None=@n46)) {g=true}, B=Constant() {desc=Scanline Y Tile, r=1}) {desc=Y tile}
  Reroute(None=ConstantBiasScale(None=Sine(None=@n47))) {desc=Scanline 0-1 Tiling Loop}
  Reroute(None=Saturate(None=@n36)) {desc=Scanline clamp 0-1 loop no Tile}
  Reroute(None=Saturate(None=@n33)) {desc=1-0 Wipe Loop}
  Sine(None=Time())
  SmoothStep(Value=@n17) {constMax=0.55, constMin=0.45}
  Step(Y=@n41) {constX=0.5}
 ]
}
