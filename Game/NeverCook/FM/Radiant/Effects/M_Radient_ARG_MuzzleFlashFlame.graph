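# Additive, unlit, two-sided muzzle flash flame material for Niagara sprites, mesh particles and skeletal meshes.
# Packed shape-texture channels, selected for planar or radial core, drive a fire curve that multiplies a time-cycled hue-shifted synthwave gradient and particle color into HDR emissive.
# Panning distorted noise, planar or radial, breaks up the shape, with dynamic parameters controlling noise and erosion.
# View-angle fade softens opacity, with depth fade outside mobile.
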Material /Game/NeverCook/FM/Radiant/Effects/M_Radient_ARG_MuzzleFlashFlame
graph {
 properties {
  bUsedWithMeshDeformer=true
  bUsedWithMeshParticles=true
  bUsedWithNiagaraMeshParticles=true
  bUsedWithNiagaraSprites=true
  bUsedWithSkeletalMesh=true
  blendMode=BLEND_Additive
  pixelDepthOffsetMode=PDOM_Legacy
  shadingModel=MSM_Unlit
  twoSided=true
  usedShadingModels=Unlit
 }
 nodes [
  n1 = TextureSampleParameter2D() {group=FlameShape, parameterName=MuzzleShape, texture=/Game/Effects/Niagara/Systems/Weapon_Masters/T_MuzzleFlashes_M}
  n2 = LinearInterpolate(
   A=LinearInterpolate(A=@n1.R, B=@n1.R, Alpha=VertexColor().R)
   B=LinearInterpolate(A=@n1.B, B=@n1.B, Alpha=DynamicParameter() {paramNames [Noise Amount, Erosion, Core Selection, Param4]}.Core Selection)
   Alpha=ScalarParameter() {parameterName=Is Radial Core}
  ) {constAlpha=0}
  CurveAtlasRowParameter(CurveTime=@n2) {
   atlas=/Game/Creative/Sets/PropSets/Primitives/Customization/Textures/ColorCurve/CC_CM_Primitives_01
   curve=/Game/Creative/Sets/PropSets/Primitives/Customization/Textures/ColorCurve/C_CM_SynthWave_01
   defaultValue=2
   parameterName=ColorCurve
  }
  n3 = TextureCoordinate()
  n4 = Constant2Vector() {g=1, r=0.001}
  n5 = CurveAtlasRowParameter(CurveTime=ComponentMask(None=Multiply(A=@n3, B=@n4)) {r=true}) {
   atlas=&CC_CM_Primitives_01
   curve=&C_CM_SynthWave_01
   defaultValue=2
   parameterName=ColorCurve
  }
  n6 = Fmod(A=Multiply(A=Time(), B=Constant() {desc=Speed Mult, r=0.1}) {constB=5}, B=Constant() {r=1})
  n7 = ParticleColor()
  n8 = Multiply(
   A=MaterialFunctionCall(Hue Shift Percentage=Saturate(None=@n6), Texture=@n5) {materialFunction=&HueShift}.Result
   B=@n7.RGB
  )
  MaterialFunctionCall(Hue Shift Percentage=ScalarParameter() {parameterName=HueShift}, Texture=@n8) {materialFunction=&HueShift}
  n9 = CurveAtlasRowParameter(CurveTime=@n2) {
   atlas=/Game/Curie/Effects/temp/Curves/CA_Curie_Fire_Visuals
   curve=/Game/Curie/Effects/temp/Curves/Curve_Fiery_B
   defaultValue=8
   parameterName=Curves
  }
  n10 = Multiply(A=Multiply(A=@n8, B=@n7.RGB), B=@n9.R)
  n11 = ScalarParameter() {defaultValue=1, group=FlameShape, parameterName=Max_HDR_Intensity}
  n12 = DotProduct(A=CameraVectorWS(), B=VertexNormalWS())
  n13 = SmoothStep(Value=Power(Base=Abs(None=@n12) {desc=Make this work on two-sided materials}, Exp=ScalarParameter() {parameterName="FadeSensitivity (optional) (S)"}))
  n14 = LinearInterpolate(A=ScalarParameter() {defaultValue=1, group=FlameShape, parameterName=Min_HDR_Intensity}, B=@n11, Alpha=@n13)
  MaterialFunctionCall(Emissive=MaterialFunctionCall(
    LDR Input=Clamp(None=@n10)
    HDR Tint=@n10
    HDR Intensity Multiplier=StaticSwitchParameter(True=@n14, False=@n11) {group=FlameShape, parameterName=Use Complex HDR Multiplier}
   ) {materialFunction=&DeriveHDRfromLDR}.RemappedRange) {materialFunction=&MF_FX_EmissiveAdjustments}
  n15 = ScalarParameter() {defaultValue=20, parameterName=FadeDistance}
  n16 = DynamicParameter() {paramNames [Noise Amount, Erosion, Core Selection, Param4]}
  n17 = ScalarParameter() {defaultValue=1, group=Noise, parameterName=U Distort Freq, sortPriority=5}
  n18 = ScalarParameter() {defaultValue=1, group=Noise, parameterName=V Distort Freq, sortPriority=6}
  n19 = Panner(Coordinate=Multiply(A=TextureCoordinate(), B=AppendVector(A=@n17, B=@n18))) {speedY=1}
  n20 = Multiply(
   A=ComponentMask(None=TextureSampleParameter2D(UVs=@n19) {group=Noise, parameterName=Texture, sortPriority=4, texture=/Game/Weapons/FORT_Pistols/Effects/Muzzle/T_Cloud_d}.RGB) {g=true, r=true}
   B=ScalarParameter() {defaultValue=1, group=Noise, parameterName=Distortion Amplitude, sortPriority=7}
  )
  n21 = TextureCoordinate()
  n22 = ComponentMask(None=@n21) {r=true}
  n23 = ScalarParameter() {defaultValue=3, parameterName=Noise_U_Scale}
  n24 = ComponentMask(None=@n21) {g=true}
  n25 = ScalarParameter() {defaultValue=1.5, parameterName=Noise_V_Scale}
  n26 = AppendVector(
   A=Add(A=Multiply(A=@n22, B=@n23), B=ScalarParameter() {parameterName=Nouse_U_Offset})
   B=Add(A=Multiply(A=@n24, B=@n25), B=ScalarParameter() {parameterName=Nouse_V_Offset})
  )
  n27 = AppendVector(
   A=ScalarParameter() {defaultValue=1, group=Noise, parameterName=U Distort Freq, sortPriority=5}
   B=ScalarParameter() {defaultValue=1, group=Noise, parameterName=V Distort Freq, sortPriority=6}
  )
  n28 = Subtract(A=Multiply(A=TextureCoordinate(), B=@n27), B=Multiply(A=@n27) {constB=0.5})
  n29 = TextureSampleParameter2D(UVs=Panner(Coordinate=MaterialFunctionCall(Vector or UVs=@n28) {materialFunction=&VectorToRadialValue}.Radial Coordinates) {speedY=-1}) {group=Noise, parameterName=Texture, sortPriority=4, texture=&T_Cloud_d}
  n30 = TextureCoordinate()
  n31 = Multiply(A=ComponentMask(None=@n30) {r=true}, B=ScalarParameter() {defaultValue=3, parameterName=Noise_U_Scale})
  n32 = Multiply(A=ComponentMask(None=@n30) {g=true}, B=ScalarParameter() {defaultValue=1.5, parameterName=Noise_V_Scale})
  n33 = AppendVector(A=Add(A=@n31, B=ScalarParameter() {parameterName=Nouse_U_Offset}), B=Add(A=@n32, B=ScalarParameter() {parameterName=Nouse_V_Offset}))
  n34 = Multiply(A=AppendVector(A=@n31, B=@n32)) {constB=0.5}
  n35 = Add(
   A=Multiply(
    A=ComponentMask(None=@n29.RGB) {g=true, r=true}
    B=ScalarParameter() {defaultValue=1, group=Noise, parameterName=Distortion Amplitude, sortPriority=7}
   )
   B=MaterialFunctionCall(Vector or UVs=Subtract(A=@n33, B=@n34)) {materialFunction=&VectorToRadialValue}.Radial Coordinates
  )
  n36 = TextureSampleParameter2D(UVs=LinearInterpolate(
    A=Add(A=@n20, B=@n26)
    B=MaterialFunctionCall(Vector or UVs=@n35) {materialFunction=&VectorToRadialValue}.Radial Coordinates
    Alpha=ScalarParameter() {parameterName=Is Radial Core}
   ) {constAlpha=0}) {group=Noise, parameterName=Noise, texture=/Game/Global/Textures/Default/Tileable_Patterns/T_Tornado_Tile}
  n37 = ScalarParameter() {defaultValue=1, group=Noise, parameterName=Noise Exp}
  n38 = Multiply(A=@n2, B=Saturate(None=Power(Base=@n36.R, Exp=@n37) {constExponent=1.5}))
  n39 = Multiply(
   A=@n7.A
   B=MaterialFunctionCall!(Dynamic Param=@n16.Erosion, Alpha=LinearInterpolate(A=@n2, B=@n38, Alpha=@n16.Noise Amount)) {materialFunction=/Game/Effects/Fort_Effects/Materials/_Functions/Erode_Alpha}.Result
  )
  Multiply(A=ShadingPathSwitch(Default=DepthFade(Opacity=@n39, FadeDistance=@n15), Mobile=@n39), B=@n13)
  Reroute(None=@n7.RGB)
 ]
}
