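# Unlit, two-sided translucent material for Niagara mesh particles, named as thruster airflow for Radiant.
# Two panning greyscale textures (directional noise, plasma) distort the UVs that drive a soft edge mask, multiplied by end fades along V.
# An inverted Fresnel softens the silhouette, its falloff varying with camera angle to the instance's X axis.
# Particle alpha and depth fade shape opacity; tinted particle color gets optional eye-adaptation compensation.
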Material /Game/NeverCook/FM/Radiant/Effects/M_Radient_ARG_Thruster_AirFlow_B
graph {
 properties {
  bUsedWithNiagaraMeshParticles=true
  blendMode=BLEND_Translucent
  pixelDepthOffsetMode=PDOM_Legacy
  shadingModel=MSM_Unlit
  twoSided=true
  usedShadingModels=Unlit
 }
 nodes [
  n1 = TextureCoordinate()
  n2 = VectorParameter() {
   group=MainTexture
   parameterName=MainTexture_NoiseControl
   sortPriority=60
   channelNames {a=VStrength, b=UStrength, g=VOffset, r=UOffset}
   defaultValue=<rgba -0.5 -0.5 0 0.3>
  }
  n3 = MaterialFunctionCall(Float4=AppendVector(A=@n2, B=@n2.VStrength)) {materialFunction=&BreakOutFloat4Components}
  n4 = VectorParameter() {group=MainTexture, parameterName=MainTexture_UVControl, channelNames {a=VPan, b=UPan, g=VScale, r=UScale}, defaultValue=<rgba 4 0.05 0 20>}
  n5 = MaterialFunctionCall!(
   UVInput=Add(A=TextureCoordinate(), B=Constant2Vector())
   TimeInput=Time() {bOverride_Period=true, period=13333}
   Control_RGB=@n4
   Control_A=@n4.VPan
  ) {materialFunction=/Game/Effects/Niagara/MasterMaterials/Functions/Common/FXCommon_QuickUVControl}
  n6 = TextureObjectParameter() {
   group=MainTexture
   parameterName=MainTexture_Texture
   samplerType=SAMPLERTYPE_LinearGrayscale
   sortPriority=0
   texture=/Game/Effects/Fort_Effects/Emotes/UltraLight/Texture/T_DirectionalNoise_Greyscale
  }
  n7 = TextureSample(UVs=@n5.UV, Tex=@n6) {samplerType=SAMPLERTYPE_LinearColor}
  n8 = AppendVector(A=@n7.R, B=@n7.R)
  n9 = Multiply(A=Add(A=@n8, B=AppendVector(A=@n3.R, B=@n3.R)), B=AppendVector(A=@n3.B, B=@n3.B))
  n10 = VectorParameter() {
   group=SecTexture
   parameterName=SecTexture_NoiseControl
   sortPriority=60
   channelNames {a=VStrength, b=UStrength, g=VOffset, r=UOffset}
   defaultValue=<rgba -0.5 -0.5 0 0.25>
  }
  n11 = MaterialFunctionCall(Float4=AppendVector(A=@n10, B=@n10.VStrength)) {materialFunction=&BreakOutFloat4Components}
  n12 = VectorParameter() {group=SecTexture, parameterName=SecTexture_UVControl, channelNames {a=VPan, b=UPan, g=VScale, r=UScale}, defaultValue=<rgba 5 0.1 0 10>}
  n13 = MaterialFunctionCall(
   UVInput=Add(A=TextureCoordinate(), B=Constant2Vector())
   TimeInput=Time() {bOverride_Period=true, period=13333}
   Control_RGB=@n12
   Control_A=@n12.VPan
  ) {materialFunction=&FXCommon_QuickUVControl}
  n14 = TextureObjectParameter() {
   group=SecTexture
   parameterName=SecTexture_Texture
   samplerType=SAMPLERTYPE_LinearGrayscale
   sortPriority=0
   texture=/Game/Effects/Fort_Effects/Emotes/UltraLight/Texture/T_Plasma_A_B_Greyscale
  }
  n15 = TextureSample(UVs=@n13.UV, Tex=@n14) {samplerType=SAMPLERTYPE_LinearColor}
  n16 = AppendVector(A=@n15.R, B=@n15.R)
  n17 = Add(A=@n16, B=AppendVector(A=@n11.R, B=@n11.R))
  n18 = AppendVector(A=@n11.B, B=@n11.B)
  n19 = MaterialFunctionCall(Float2=Add(A=Multiply(A=@n17, B=@n18), B=Add(A=@n1, B=@n9))) {materialFunction=&BreakOutFloat2Components}
  n20 = VectorParameter() {group=EdgeMask, parameterName=EdgeMask_EdgeMask_Control, channelNames {a="", b=Power, g=Multiply, r=Offset}, defaultValue=<rgba 0 1.25 4 1>}
  n21 = Saturate(None=Multiply(A=Add(A=@n19.G, B=@n20.Offset), B=@n20.Multiply))
  n22 = CameraVectorWS()
  n23 = PixelNormalWS()
  n24 = OneMinus(None=Saturate(None=DotProduct(A=@n22, B=@n23)))
  n25 = ScalarParameter() {defaultValue=1, parameterName=F2}
  n26 = Constant3Vector() {constant=<rgba -1 0 0 0>}
  n27 = DotProduct(
   A=Normalize(VectorInput=Transform(None=@n26) {transformSourceType=TRANSFORMSOURCE_Instance})
   B=MaterialFunctionCall() {materialFunction=&CameraDirectionVector}.Result
  )
  n28 = Power(
   Base=OneMinus(None=Power(Base=@n24, Exp=@n25))
   Exp=LinearInterpolate(A=ScalarParameter() {defaultValue=15, parameterName=F}, Alpha=Abs(None=@n27))
  )
  n29 = Constant4Vector() {constant=<rgba 0 1 0 0>}
  n30 = MaterialFunctionCall(Float3=@n29.RGBA) {materialFunction=&BreakOutFloat3Components}
  n31 = MaterialFunctionCall(Float2=TextureCoordinate()) {materialFunction=&BreakOutFloat2Components}
  n32 = Multiply(A=Add(A=@n31.G, B=@n30.R), B=@n30.G)
  n33 = VectorParameter() {
   group=EdgeMask
   parameterName=EdgeMask_EdgeMask_InvertControl
   channelNames {a="", b=Power, g=Multiply, r=Offset}
   defaultValue=<rgba 0 15 2 1>
  }
  n34 = MaterialFunctionCall(Float3=@n33) {materialFunction=&BreakOutFloat3Components}
  n35 = Multiply(A=Add(A=OneMinus(None=@n31.G), B=@n34.R), B=@n34.G)
  n36 = Multiply(A=Power(Base=Saturate(None=@n32), Exp=@n30.B), B=Power(Base=Saturate(None=@n35), Exp=@n34.B))
  n37 = ParticleColor()
  n38 = Multiply(A=Multiply(A=Power(Base=@n21, Exp=@n20.Power), B=Multiply(A=@n36, B=@n28)), B=@n37.A)
  DepthFade(Opacity=@n38, FadeDistance=ScalarParameter() {defaultValue=150, parameterName=DF})
  n39 = ScalarParameter() {
   defaultValue=1.68
   desc=Eye Adaptation in the old scene was 1.68 - scale to match old appearance
   group=FX Emissive EyeAdaptation
   parameterName=FX Emissive EyeAdaptation Prescale
   sliderMax=4
  }
  n40 = Multiply(A=VectorParameter() {parameterName=Color, defaultValue=<rgba 1 1 1 1>}.RGB, B=@n37.RGB)
  n41 = ComponentMask(None=Multiply(A=@n40, B=@n39) {constB=1.3}) {b=true, g=true, r=true}
  n42 = ScalarParameter() {defaultValue=1, desc=Mix to Original, group=FX Emissive EyeAdaptation, parameterName=FX Emissive EyeAdaptation Alpha}
  EyeAdaptationInverse(LightValueInput=@n41, AlphaInput=@n42) {desc=disabling due to fails on Juno}
  Fresnel() {baseReflectFraction=0, exponent=1}
  MaterialFunctionCall!(NoiseValue=@n8, Control_RGB=@n2, Control_A=@n2) {materialFunction=/Game/Effects/Niagara/MasterMaterials/Functions/Common/FXCommon_NoiseControl}
  MaterialFunctionCall!(UorV=@n31.G, Invert_EdgeControl=@n33, NoInvert_EdgeControl=@n29.RGBA) {materialFunction=/Game/Effects/Niagara/MasterMaterials/Functions/Common/FXCommon_EdgeMask_Ex}
  MaterialFunctionCall(NoiseValue=@n16, Control_RGB=@n10, Control_A=@n10) {materialFunction=&FXCommon_NoiseControl}
  Multiply(A=@n38)
  OneMinus(None=@n1)
  Reroute(None=MaterialFunctionCall(Emissive=Reroute()) {materialFunction=&MF_FX_EmissiveAdjustments}.Result)
  Reroute(None=@n15.RGB) {desc=out rgb}
  Reroute(None=@n15.R) {desc=R}
  Reroute(None=@n15.G) {desc=G}
  Reroute(None=@n15.B) {desc=B}
  Reroute(None=@n15.A) {desc=A}
  Reroute(None=@n13.UV) {desc=UV}
  Reroute(None=@n7.RGB) {desc=out rgb}
  Reroute(None=@n7.R) {desc=R}
  Reroute(None=@n7.G) {desc=G}
  Reroute(None=@n7.B) {desc=B}
  Reroute(None=@n7.A) {desc=A}
  Reroute(None=@n5.UV) {desc=UV}
  Reroute(None=MaterialFunctionCall(Texture=@n6, UVControl_RGB=@n4, UVControl_A=@n4) {materialFunction=&FXCommon_TextureSample}.RG)
  Reroute(None=MaterialFunctionCall(Texture=@n14, UVControl_RGB=@n12, UVControl_A=@n12) {materialFunction=&FXCommon_TextureSample}.RG)
  n43 = OneMinus(None=ComponentMask(None=TextureCoordinate()) {g=true})
  Saturate(None=Multiply(A=Power(Base=@n43)) {constB=2})
  n44 = EyeAdaptation()
  StaticSwitchParameter(True=LinearInterpolate(A=@n41, B=Divide(A=@n41, B=@n44.EyeAdaptation), Alpha=@n42), False=@n40) {
   defaultValue=true
   desc="Enable/Disable EyeAdaptation (auto exposure) compensation"
   group=FX Emissive EyeAdaptation
   parameterName=FX Emissive EyeAdaptation
  }
 ]
}
