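# Translucent unlit VFX material for the Radiant laser beam, usable on particles, Niagara ribbons and meshes.
# A panning trail mask blends with two multiplied caustics-noise layers into a shared Tiling signal driving three-zone color and alpha gradients.
# Alpha combines particle alpha with depth, near-camera and view-angle fades.
# The graph also holds a hue-cycling color curve, exposure-compensated emissive branches per quality level, and assorted time-loop helpers.
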Material /Game/NeverCook/FM/Radiant/Effects/M_Radient_Laser_Beam
graph {
 properties {
  bUsedWithMeshDeformer=true
  bUsedWithMeshParticles=true
  bUsedWithNiagaraMeshParticles=true
  bUsedWithNiagaraRibbons=true
  bUsedWithParticleSprites=true
  bUsedWithSkeletalMesh=true
  blendMode=BLEND_Translucent
  pixelDepthOffsetMode=PDOM_Legacy
  shadingModel=MSM_Unlit
  usedShadingModels=Unlit
 }
 nodes [
  n1 = TextureCoordinate()
  n2 = ScalarParameter() {defaultValue=1, group=Main Gradient, parameterName=GradientTiling}
  n3 = AppendVector(A=@n2, B=@n2)
  n4 = ScalarParameter() {defaultValue=-1, group=Main Gradient, parameterName=Panner Speed}
  n5 = Constant()
  n6 = TextureSample(
   UVs=Panner(
    Coordinate=Multiply(A=@n1, B=@n3)
    Speed=MaterialFunctionCall(X=@n4, Y=@n5) {materialFunction=&MakeFloat2}.Result
   ) {speedX=1}
   Tex=TextureObjectParameter() {parameterName=PrimaryMask, texture=/Game/Effects/Fort_Effects/Textures/Smoke/T_Lobber_Trail_Mask}
  ) {texture=&T_Lobber_Trail_Mask}
  n7 = Multiply(A=TextureCoordinate(), B=AppendVector(A=Constant() {r=0.25}, B=Constant() {r=5}))
  n8 = MaterialFunctionCall(X=Constant() {r=5}, Y=Constant() {r=20}) {materialFunction=&MakeFloat2}
  n9 = Multiply(A=TextureCoordinate(), B=AppendVector(A=Constant(), B=Constant() {r=10}))
  n10 = MaterialFunctionCall(X=Constant() {r=2}, Y=Constant() {r=-8}) {materialFunction=&MakeFloat2}
  n11 = Multiply(
   A=TextureSample(UVs=Panner(Coordinate=@n7, Speed=@n8.Result) {speedX=1}) {texture=/VehicleCosmetics/DriftTrail/DriftTrail_ElectroBeam/Textures/CausticsNoise}.R
   B=TextureSample(UVs=Panner(Coordinate=@n9, Speed=@n10.Result) {speedX=1}) {texture=&CausticsNoise}.R
  )
  Add(A=@n6.R, B=@n11)
  n12 = TextureCoordinate()
  n13 = Constant() {r=4}
  n14 = Multiply(A=Time(), B=Constant() {desc=Speed Mult, r=0.1}) {constB=5}
  n15 = Subtract(A=Fmod(A=Divide(A=@n14, B=@n13) {constB=4}, B=Constant() {r=1.9}), B=Constant() {r=0.4}) {desc=Y}
  n16 = Subtract(A=ComponentMask(None=@n12) {g=true}, B=@n15)
  n17 = Multiply(A=@n16, B=Constant() {r=110})
  n18 = Constant() {r=2}
  n19 = Add(A=Multiply(A=Power(Base=@n17, Exp=@n18), B=Constant() {r=-1}), B=Constant() {r=0.8}) {desc=str}
  n20 = Constant() {r=0.01}
  n21 = Subtract(A=ComponentMask(None=@n12) {r=true}, B=Clamp(None=Multiply(A=@n19, B=@n20)))
  AppendVector(A=@n21, B=ComponentMask(None=@n12) {g=true})
  ConstantBiasScale()
  Constant() {r=1}
  Divide(A=WorldPosition().XYZ, B=Constant() {r=50})
  n22 = ScalarParameter() {
   defaultValue=1.68
   desc=Eye Adaptation in the old scene was 1.68 - scale to match old appearance
   group=FX Emissive EyeAdaptation
   parameterName=FX Emissive EyeAdaptation Prescale
   sliderMax=4
  }
  n23 = TextureCoordinate()
  n24 = Constant2Vector() {g=1, r=0.001}
  n25 = CurveAtlasRowParameter(CurveTime=ComponentMask(None=Multiply(A=@n23, B=@n24)) {r=true}) {
   atlas=/Game/Creative/Sets/PropSets/Primitives/Customization/Textures/ColorCurve/CC_CM_Primitives_01
   curve=/Game/Creative/Sets/PropSets/Primitives/Customization/Textures/ColorCurve/C_CM_SynthWave_01
   defaultValue=2
   parameterName=ColorCurve
  }
  n26 = Saturate(None=Fmod(A=@n14, B=Constant() {r=1}))
  n27 = ParticleColor()
  n28 = Multiply(
   A=MaterialFunctionCall(Hue Shift Percentage=@n26, Texture=@n25) {materialFunction=&HueShift}.Result
   B=@n27.RGB
  )
  n29 = ComponentMask(None=Multiply(A=@n28, B=@n22) {constB=1.3}) {b=true, g=true, r=true}
  n30 = ScalarParameter() {defaultValue=1, desc=Mix to Original, group=FX Emissive EyeAdaptation, parameterName=FX Emissive EyeAdaptation Alpha}
  EyeAdaptationInverse(LightValueInput=@n29, AlphaInput=@n30) {desc=disabling due to fails on Juno}
  n31 = ScalarParameter() {defaultValue=3, parameterName=Boost}
  n32 = Multiply(
   A=MaterialFunctionCall(In=@n11, Contrast=Constant() {r=0.15}) {materialFunction=&CheapContrast}.Result
   B=Constant() {r=10}
  )
  n33 = LinearInterpolate(A=@n6.R, B=@n32, Alpha=ScalarParameter() {defaultValue=0.75, parameterName=NoiseLerp})
  n34 = NamedRerouteDeclaration(None=Saturate(None=Multiply(A=@n33, B=@n31))) {name=Tiling, nodeColor=<rgba 1 0.7266 0 1>}
  n35 = ScalarParameter() {group=Main Gradient, parameterName=Color Offset 1}
  LinearInterpolate(
   A=VectorParameter() {parameterName=Color Mid, defaultValue=<rgba 0 0.6168 1 0>}.RGB
   B=VectorParameter() {parameterName=Color Center, defaultValue=<rgba 1 1 1 0>}.RGB
   Alpha=MaterialFunctionCall(In=Add(A=@n34, B=@n35) {constB=0.2}, Contrast=ScalarParameter() {defaultValue=0.4, group=Main Gradient, parameterName=Color Contrast 1}) {materialFunction=&CheapContrast}.Result
  )
  n36 = NamedRerouteUsage()
  n37 = ScalarParameter() {defaultValue=0.4, group=Main Gradient, parameterName=Color Offset 2}
  LinearInterpolate(
   A=VectorParameter() {parameterName=Color Outside, defaultValue=<rgba 1 0 0.7722 1>}.RGB
   B=@n27.RGB
   Alpha=MaterialFunctionCall(In=Add(A=@n36, B=@n37) {constB=0.2}, Contrast=ScalarParameter() {defaultValue=3, group=Main Gradient, parameterName=Color Contrast 2}) {materialFunction=&CheapContrast}.Result
  )
  MaterialFunctionCall(Number=@n26) {materialFunction=&DebugScalarValues}
  n38 = Subtract(A=ComponentMask(None=@n12) {g=true}, B=@n15)
  n39 = Abs(None=@n38)
  n40 = Constant() {r=0.3}
  n41 = Power(Base=Subtract(A=Constant() {r=1}, B=Power(Base=@n39, Exp=@n40)), Exp=Constant() {r=3}) {desc=colorAdd}
  n42 = Constant() {r=0.5}
  n43 = Constant() {r=10}
  MaterialFunctionCall(Y=Add(A=Multiply(A=@n41, B=@n42)), Z=Add(A=Multiply(A=@n41, B=@n43))) {materialFunction=&MakeFloat3}
  n44 = Frac(None=@n14)
  Multiply(A=Add(A=@n44) {constB=-0.5}) {constB=2}
  Multiply(A=@n26)
  Multiply(A=@n12, B=Constant2Vector() {g=0.1, r=0.1})
  Multiply(A=@n6.R, B=@n32)
  n45 = EyeAdaptation()
  n46 = StaticSwitchParameter(True=LinearInterpolate(A=@n29, B=Divide(A=@n29, B=@n45.EyeAdaptation), Alpha=@n30), False=@n28) {
   defaultValue=true
   desc="Enable/Disable EyeAdaptation (auto exposure) compensation"
   group=FX Emissive EyeAdaptation
   parameterName=FX Emissive EyeAdaptation
  }
  Reroute(None=@n46)
  n47 = ScalarParameter() {defaultValue=1, group=Emissive Adjustments, parameterName=Cosmetic Emissive EyeAdaptation Alpha}
  n48 = MaterialFunctionCall!(
   Emissive=Multiply(A=@n28, B=ScalarParameter() {defaultValue=2, group=Emissive Adjustments, parameterName=HighGIQualityMultiplier})
   Alpha=@n47
   Material
  ) {materialFunction=/Game/Packages/Fortress_SharedMaterials/Base_Material_Functions/BaseCharacterShaderFunctions/MF_CosmeticEmissiveEyeAdaptation}
  n49 = RayTracingQualitySwitch(
   Normal=@n48.Emissive
   RayTraced=Multiply(A=ScalarParameter() {defaultValue=3, group=Emissive Adjustments, parameterName=LumenBoost}, B=@n48.Emissive)
  )
  n50 = MaterialFunctionCall(
   Emissive=Multiply(A=@n28, B=ScalarParameter() {defaultValue=1.5, group=Emissive Adjustments, parameterName=LowEndMultiplier}) {constB=1.5, desc=Low End and Mobile Screen Multiplier}
   Alpha=@n47
  ) {materialFunction=&MF_CosmeticEmissiveEyeAdaptation}
  Reroute(None=ShadingPathSwitch(Default=QualitySwitch(Default=@n49, Low=@n50.Emissive), Mobile=@n50.Emissive))
  Reroute(None=MaterialFunctionCall(Emissive=Reroute()) {materialFunction=&MF_FX_EmissiveAdjustments}.Result)
  Reroute(None=@n21)
  Reroute(None=@n19)
  n51 = ConstantBiasScale(None=Frac(None=@n14)) {bias=-0.5, scale=2}
  Reroute(None=@n51) {desc=output -1 to 1 loop}
  Reroute(None=Abs(None=@n51)) {desc="Oscillating 0 - 1 "}
  n52 = Sine(None=@n14)
  Reroute(None=@n52) {desc=Sin Oscillating -1 - 1}
  Reroute(None=ConstantBiasScale(None=@n52)) {desc=Sin Oscillating 0 - 1}
  n53 = Sine(None=@n14) {period=0.8095}
  n54 = Sine(None=@n14) {period=0.1807}
  Reroute(None=Saturate(None=LinearInterpolate(A=@n52, B=@n53, Alpha=@n54))) {desc=Random Flicker Clamp 0 - 1}
  Reroute(None=Saturate(None=@n15)) {desc=0-1 Loop Clamped Delaypauses for a short time on 0 and 1 bc values extend above and below 0-1 range but are clamped}
  n55 = Add(A=@n12, B=@n14)
  n56 = Multiply(A=ComponentMask(None=Frac(None=@n55)) {g=true}, B=Constant() {desc=Scanline Y Tile, r=1}) {desc=Y tile}
  Reroute(None=ConstantBiasScale(None=Sine(None=@n56))) {desc=Scanline 0-1 Tiling Loop}
  Reroute(None=Saturate(None=@n41)) {desc=Scanline clamp 0-1 loop no Tile}
  Reroute(None=Saturate(None=@n38)) {desc=1-0 Wipe Loop}
  n57 = ScalarParameter() {defaultValue=100, group=Alpha, parameterName=NearCameraFade}
  n58 = ScalarParameter() {defaultValue=1, group=Alpha, parameterName=Alpha Mid}
  n59 = ScalarParameter() {defaultValue=1, group=Alpha, parameterName=Alpha Center}
  n60 = MaterialFunctionCall(
   In=Add(A=NamedRerouteUsage(), B=ScalarParameter() {defaultValue=0.2, group=Alpha, parameterName=Alpha Offset 1}) {constB=0.2}
   Contrast=ScalarParameter() {defaultValue=2, group=Alpha, parameterName=Alpha Contrast 1}
  ) {materialFunction=&CheapContrast}
  n61 = Add(A=NamedRerouteUsage(), B=ScalarParameter() {group=Alpha, parameterName=Alpha Offset 2}) {constB=0.2}
  n62 = ScalarParameter() {group=Alpha, parameterName=Alpha Contrast 2}
  n63 = Multiply(
   A=DepthFade(
    Opacity=MaterialFunctionCall(Distance=@n57) {materialFunction=&Near_Camera_Fade}.Result
    FadeDistance=ScalarParameter() {defaultValue=35, group=Alpha, parameterName=FadeDistance}
   ) {fadeDistanceDefault=50}
   B=LinearInterpolate(
    A=ScalarParameter() {parameterName=Alpha Outside}
    B=LinearInterpolate(A=@n58, B=@n59, Alpha=@n60.Result)
    Alpha=MaterialFunctionCall(In=@n61, Contrast=@n62) {materialFunction=&CheapContrast}.Result
   )
  )
  n64 = Constant() {r=0.5}
  Reroute(None=Multiply(
    A=Multiply(A=@n27.A, B=@n63)
    B=MaterialFunctionCall("FadeSensitivity (optional)"=@n64) {materialFunction=&FadeBasedOnViewAngle}.Result
   ))
  Reroute(None=@n27.RGB)
  Sine(None=Time())
  SmoothStep(Value=@n16) {constMax=0.55, constMin=0.45}
  Step(Y=@n44) {constX=0.5}
 ]
}
