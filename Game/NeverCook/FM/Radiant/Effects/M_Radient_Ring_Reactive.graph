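# Translucent unlit Niagara mesh and ribbon material for a Radiant ring effect with idle and reactive states.
# Dynamic parameters blend noise-masked idle waves into sine ripples along vertex normals, combined with SplineThicken for WPO, and shift hue.
# Emissive maps two panning noises through a curve atlas, adding a scrolling core texture and particle color.
# Opacity combines core, side mask, quality-switched depth and angle fades; SwitchDirection flips UV axes.
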Material /Game/NeverCook/FM/Radiant/Effects/M_Radient_Ring_Reactive
graph {
 properties {
  bUsedWithNiagaraMeshParticles=true
  bUsedWithNiagaraRibbons=true
  blendMode=BLEND_Translucent
  pixelDepthOffsetMode=PDOM_Legacy
  shadingModel=MSM_Unlit
  usedShadingModels=Unlit
 }
 nodes [
  Add(A=Time())
  n1 = TextureCoordinate()
  n2 = LinearInterpolate(
   A=@n1
   B=MaterialFunctionCall(XYZ=@n1) {materialFunction=&Swizzle}.YX
   Alpha=NamedRerouteUsage()
  )
  n3 = Add(A=@n2, B=Multiply(A=Time(), B=Constant2Vector() {g=1}))
  n4 = Add(A=Multiply(A=Sine(None=@n3)) {constB=0.5}) {constB=0.5}
  n5 = Subtract(A=Saturate(None=ComponentMask(None=@n4) {g=true})) {constB=0.7}
  n6 = DynamicParameter() {defaultValue=<rgba 1 10 1 0>, paramNames [MeshSize,WPOStr,MotionMask,IsReactive]}
  n7 = Multiply(A=Saturate(None=Multiply(A=@n5) {constB=5}), B=@n6.MotionMask)
  n8 = VectorParameter() {parameterName=IdleTextureAdjustments, channelNames {a="", b=Speed, g=ScaleY, r=ScaleX}, defaultValue=<rgba 1 1 0.5 1>}
  n9 = Multiply(A=TextureCoordinate(), B=ComponentMask(None=@n8) {g=true, r=true})
  n10 = ComponentMask(None=@n9) {g=true}
  n11 = ComponentMask(None=@n9) {r=true}
  n12 = NamedRerouteUsage()
  n13 = Constant()
  n14 = TextureObjectParameter() {
   parameterName=DistortionTexture
   samplerType=SAMPLERTYPE_LinearGrayscale
   texture=&T_SimplePerlin_Greyscale
  }
  n15 = TextureSample(UVs=Panner(Coordinate=LinearInterpolate(A=@n10, B=@n11, Alpha=@n12), Speed=AppendVector(A=@n8.Speed, B=@n13)) {speedX=0.5}, Tex=@n14) {samplerType=SAMPLERTYPE_LinearGrayscale, texture=&T_SimplePerlin_Greyscale}
  n16 = Saturate(None=Multiply(A=Power(Base=@n15.R) {constExponent=3}) {constB=3})
  n17 = TextureCoordinate()
  n18 = Time()
  n19 = VectorParameter() {parameterName=RippleLinesAmount, defaultValue=<rgba 0 6 0 6>}
  n20 = ComponentMask(None=@n19.RGB) {g=true, r=true}
  n21 = Multiply(
   A=Add(A=@n17, B=Multiply(A=@n18) {constB=0.5})
   B=LinearInterpolate(
    A=@n20
    B=MaterialFunctionCall(XY=@n20) {materialFunction=&Swizzle}.YX
    Alpha=NamedRerouteUsage()
   )
  ) {constB=6}
  n22 = Add(A=Multiply(A=Sine(None=@n21)) {constB=0.5}) {constB=0.5}
  n23 = Time()
  n24 = AppendVector(A=@n19.B, B=@n19.B)
  n25 = Multiply(
   A=Add(A=@n17, B=Multiply(A=@n23) {constB=-0.25})
   B=LinearInterpolate(
    A=@n24
    B=MaterialFunctionCall(XY=@n24) {materialFunction=&Swizzle}.YX
    Alpha=NamedRerouteUsage()
   )
  ) {constB=6}
  n26 = Add(A=Multiply(A=Sine(None=@n25)) {constB=0.5}) {constB=0.5}
  n27 = ComponentMask(None=@n26) {g=true}
  n28 = ComponentMask(None=@n26) {r=true}
  n29 = NamedRerouteUsage()
  n30 = Min(
   A=LinearInterpolate(A=ComponentMask(None=@n22) {g=true}, B=ComponentMask(None=@n22) {r=true}, Alpha=NamedRerouteUsage())
   B=Multiply(A=LinearInterpolate(A=@n27, B=@n28, Alpha=@n29)) {constB=0.5}
  )
  n31 = VertexNormalWS()
  n32 = Multiply(
   A=LinearInterpolate(A=Multiply(A=@n7, B=@n16), B=Multiply(A=@n6.MotionMask, B=@n30), Alpha=@n6.IsReactive)
   B=Multiply(A=Multiply(A=@n31, B=@n6.MeshSize), B=@n6.WPOStr) {constB=10}
  )
  n33 = ScalarParameter() {defaultValue=7, parameterName=SplineThickness}
  n34 = Add(A=WorldPosition() {worldPositionShaderOffset=WPT_ExcludeAllShaderOffsets}.XYZ, B=@n32)
  LinearInterpolate(
   B=Add(
    A=@n32
    B=MaterialFunctionCall(WidthBase=@n33, WidthTip=@n33, WorldPosition=@n34) {materialFunction=&SplineThicken}.WorldPositionOffset
   )
   Alpha=ScalarParameter() {defaultValue=1, parameterName=Use WPO}
  )
  n35 = ScalarParameter() {parameterName=HueShiftIdle}
  n36 = ScalarParameter() {parameterName=HueShiftReactive}
  n37 = DynamicParameter() {parameterIndex=1, defaultValue=<rgba 0 0 0 0>, paramNames [ReactiveHueChange, "-- UNUSED -- ", "-- UNUSED -- ", "-- UNUSED -- "]}
  n38 = Sine(None=Multiply(A=Time()) {constB=0.6})
  n39 = LinearInterpolate(Alpha=Add(A=Multiply(A=@n38) {constB=0.5}) {constB=0.5}) {constA=1, constB=0.3}
  n40 = VectorParameter() {parameterName=ColorTextureMaskA, channelNames {a=SpeedY, b=SpeedX, g=ScaleY, r=ScaleX}, defaultValue=<rgba 0.4 1 0.1 -0.15>}
  n41 = ComponentMask(None=@n40) {g=true, r=true}
  n42 = TextureObjectParameter() {parameterName=ColorMaskTexture, texture=/Game/Effects/Fort_Effects/Emotes/UltraLight/Texture/T_SoftNoise_BC7}
  n43 = TextureSample(UVs=Panner(Coordinate=Multiply(A=@n2, B=@n41), Speed=AppendVector(A=@n40.SpeedX, B=@n40.SpeedX)) {speedX=0.1, speedY=-0.15}, Tex=@n42) {texture=&T_SoftNoise_BC7}
  n44 = VectorParameter() {parameterName=ColorTextureMaskB, channelNames {a=SpeedY, b=SpeedX, g=ScaleY, r=ScaleX}, defaultValue=<rgba 0.3 1 -0.8 0.25>}
  n45 = ComponentMask(None=@n44) {g=true, r=true}
  n46 = TextureSample(UVs=Panner(Coordinate=Multiply(A=@n2, B=@n45), Speed=AppendVector(A=@n44.SpeedX, B=@n44.SpeedX)) {speedX=-0.08, speedY=0.25}, Tex=@n42) {texture=&T_SoftNoise_BC7}
  n47 = CurveAtlasRowParameter(CurveTime=Power(Base=Max(A=@n43.R, B=@n46.G)) {constExponent=1.5}) {
   atlas=/Game/Weapons/FORT_Melee/Pickaxe_RelayStick/Textures/CA_RelayStick
   curve=/Game/Weapons/FORT_Melee/Pickaxe_RelayStick/Textures/C_RelayStick_CircularFresnel
   defaultValue=1
  }
  n48 = TextureCoordinate()
  n49 = LinearInterpolate(
   A=MaterialFunctionCall(XYZ=@n48) {materialFunction=&Swizzle}.YX
   B=@n48
   Alpha=NamedRerouteUsage()
  )
  n50 = Panner(Coordinate=Multiply(A=Constant2Vector() {g=1, r=2}, B=@n49)) {speedX=-0.3}
  n51 = TextureObjectParameter() {
   parameterName=CoreTexture
   texture=/Game/Weapons/FORT_Melee/Pickaxe_Future_Samurai_Summer_Male/FX/T_Pickaxe_Future_Samurai_Summer_Male_Trail_01
  }
  n52 = Panner(Coordinate=Multiply(A=@n49, B=Constant2Vector() {g=-1, r=3})) {speedX=0.5}
  n53 = ScalarParameter() {parameterName=SideMaskBlendMin}
  n54 = ScalarParameter() {defaultValue=1, parameterName=SideMaskBlendMax}
  n55 = TextureCoordinate()
  n56 = Subtract(A=LinearInterpolate(A=ComponentMask(None=@n55) {r=true}, B=ComponentMask(None=@n55) {g=true}, Alpha=NamedRerouteUsage())) {constB=0.5}
  n57 = OneMinus(None=Multiply(A=Abs(None=@n56)) {constB=2})
  n58 = LinearInterpolate(
   A=Min(
    A=TextureSample(UVs=@n50, Tex=@n51) {texture=&T_Pickaxe_Future_Samurai_Summer_Male_Trail_01}.R
    B=TextureSample(UVs=@n52, Tex=@n51) {texture=&T_Pickaxe_Future_Samurai_Summer_Male_Trail_01}.R
   )
   B=Saturate(None=SmoothStep(Min=@n53, Max=@n54, Value=@n57) {constMin=0.5})
   Alpha=ScalarParameter() {parameterName=SideMaskBlendStrenght}
  )
  n59 = Multiply(A=Multiply(A=@n47) {constB=6}, B=@n58)
  n60 = ParticleColor()
  n61 = Multiply(A=Desaturation(None=Add(A=@n47, B=@n59), Fraction=ScalarParameter() {defaultValue=0.2, parameterName=Desaturation}), B=@n60.RGB)
  MaterialFunctionCall(Emissive=MaterialFunctionCall(Hue Shift Percentage=LinearInterpolate(A=@n35, B=@n36, Alpha=@n37.ReactiveHueChange), Texture=Multiply(A=@n39, B=@n61)) {materialFunction=&HueShift}.Result) {materialFunction=&MF_FX_EmissiveAdjustments}
  n62 = ScalarParameter() {defaultValue=1, parameterName=SidemaskStrenght}
  n63 = Saturate(None=Add(A=@n58, B=LinearInterpolate(B=@n57, Alpha=@n62) {constAlpha=1, constB=0}))
  n64 = DepthFade(FadeDistance=ScalarParameter() {parameterName=DepthFade})
  n65 = Constant() {r=1}
  n66 = QualitySwitch(Default=@n64, Low=@n65, High=@n64, Medium=@n65, Epic=@n64)
  n67 = LinearInterpolate(
   B=MaterialFunctionCall("FadeSensitivity (optional)"=ScalarParameter() {defaultValue=2, parameterName=AngleFadeSensitivity}) {materialFunction=&FadeBasedOnViewAngle}.Result
   Alpha=Sign(None=ScalarParameter() {parameterName=UseAngleFade})
  ) {constA=1}
  n68 = Constant() {r=1}
  Multiply(A=Multiply(A=Multiply(A=@n63, B=@n66), B=QualitySwitch(Default=@n67, Low=@n68, High=@n67, Medium=@n68, Epic=@n67)), B=@n60.A)
  NamedRerouteDeclaration(None=Sign(None=ScalarParameter() {parameterName=SwitchDirection, sliderMax=1})) {name=SwitchDirection, nodeColor=<rgba 1 0.8203 0 1>}
 ]
}
