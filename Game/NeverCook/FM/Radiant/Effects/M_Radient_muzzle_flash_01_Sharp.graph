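# Additive, unlit, two-sided muzzle flash material for sprite and mesh particles.
# A flash texture blends with a panned tornado pattern through dynamic parameters, while a synthwave curve-atlas color, hue-shifted over time and tinted by particle color, drives emissive.
# An Is Translucent switch moves that blended mask from emissive to opacity, which also uses particle alpha and camera, facing and depth fades.
# Many time-based oscillator, scanline and wipe helper nodes appear unused.
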Material /Game/NeverCook/FM/Radiant/Effects/M_Radient_muzzle_flash_01_Sharp
graph {
 properties {
  bUsedWithMeshParticles=true
  bUsedWithNanite=true
  bUsedWithNiagaraMeshParticles=true
  bUsedWithParticleSprites=true
  blendMode=BLEND_Additive
  opacityMaskClipValue=0.933
  pixelDepthOffsetMode=PDOM_Legacy
  shadingModel=MSM_Unlit
  twoSided=true
  usedShadingModels=Unlit
 }
 nodes [
  n1 = TextureCoordinate()
  n2 = Constant() {r=4}
  n3 = Multiply(A=Time(), B=Constant() {desc=Speed Mult, r=0.1}) {constB=5}
  n4 = Subtract(A=Fmod(A=Divide(A=@n3, B=@n2) {constB=4}, B=Constant() {r=1.9}), B=Constant() {r=0.4}) {desc=Y}
  n5 = Subtract(A=ComponentMask(None=@n1) {g=true}, B=@n4)
  n6 = Multiply(A=@n5, B=Constant() {r=110})
  n7 = Constant() {r=2}
  n8 = Add(A=Multiply(A=Power(Base=@n6, Exp=@n7), B=Constant() {r=-1}), B=Constant() {r=0.8}) {desc=str}
  n9 = Constant() {r=0.01}
  n10 = Subtract(A=ComponentMask(None=@n1) {r=true}, B=Clamp(None=Multiply(A=@n8, B=@n9)))
  AppendVector(A=@n10, B=ComponentMask(None=@n1) {g=true})
  ConstantBiasScale()
  Divide(A=WorldPosition().XYZ, B=Constant() {r=50})
  n11 = Saturate(None=Fmod(A=@n3, B=Constant() {r=1}))
  MaterialFunctionCall(Number=@n11) {materialFunction=&DebugScalarValues}
  n12 = Subtract(A=ComponentMask(None=@n1) {g=true}, B=@n4)
  n13 = Abs(None=@n12)
  n14 = Constant() {r=0.3}
  n15 = Power(Base=Subtract(A=Constant() {r=1}, B=Power(Base=@n13, Exp=@n14)), Exp=Constant() {r=3}) {desc=colorAdd}
  n16 = Constant() {r=0.5}
  n17 = Constant() {r=10}
  MaterialFunctionCall(Y=Add(A=Multiply(A=@n15, B=@n16)), Z=Add(A=Multiply(A=@n15, B=@n17))) {materialFunction=&MakeFloat3}
  n18 = TextureSampleParameter2D() {parameterName=FlashTexture, texture=/Game/Weapons/FORT_Pistols/Effects/Muzzle/T_Muzzle_Flash_01_Sharp}
  n19 = DynamicParameter() {defaultValue=<rgba 0 1 1 1>, paramNames [TextureBlend,PanSpeed,UTiling,VTiling]}
  n20 = Panner(
   Coordinate=Multiply(A=TextureCoordinate() {uTiling=2, vTiling=0.5}, B=AppendVector(A=@n19.UTiling, B=@n19.UTiling))
   Time=Multiply(A=@n19.PanSpeed, B=Time())
  ) {speedY=1}
  n21 = LinearInterpolate(
   A=@n18.R
   B=Multiply(A=@n18.R, B=TextureSample(UVs=@n20) {texture=/Game/Global/Textures/Default/Tileable_Patterns/T_Tornado_Tile}.R)
   Alpha=@n19.TextureBlend
  )
  n22 = TextureCoordinate()
  n23 = Constant2Vector() {g=1, r=0.001}
  n24 = CurveAtlasRowParameter(CurveTime=ComponentMask(None=Multiply(A=@n22, B=@n23)) {r=true}) {
   atlas=/Game/Creative/Sets/PropSets/Primitives/Customization/Textures/ColorCurve/CC_CM_Primitives_01
   curve=/Game/Creative/Sets/PropSets/Primitives/Customization/Textures/ColorCurve/C_CM_SynthWave_01
   defaultValue=2
   parameterName=ColorCurve
  }
  n25 = ParticleColor()
  n26 = Multiply(
   A=MaterialFunctionCall(Hue Shift Percentage=@n11, Texture=@n24) {materialFunction=&HueShift}.Result
   B=@n25.RGB
  )
  MaterialFunctionCall(Emissive=StaticSwitchParameter(True=@n26, False=Multiply(A=@n21, B=@n26)) {parameterName=Is Translucent}) {materialFunction=&MF_FX_EmissiveAdjustments}
  n27 = Frac(None=@n3)
  Multiply(A=Add(A=@n27) {constB=-0.5}) {constB=2}
  Multiply(A=@n11)
  Multiply(A=@n1, B=Constant2Vector() {g=0.1, r=0.1})
  Reroute(None=@n10)
  Reroute(None=@n8)
  n28 = ConstantBiasScale(None=Frac(None=@n3)) {bias=-0.5, scale=2}
  Reroute(None=@n28) {desc=output -1 to 1 loop}
  Reroute(None=Abs(None=@n28)) {desc="Oscillating 0 - 1 "}
  n29 = Sine(None=@n3)
  Reroute(None=@n29) {desc=Sin Oscillating -1 - 1}
  Reroute(None=ConstantBiasScale(None=@n29)) {desc=Sin Oscillating 0 - 1}
  n30 = Sine(None=@n3) {period=0.8095}
  n31 = Sine(None=@n3) {period=0.1807}
  Reroute(None=Saturate(None=LinearInterpolate(A=@n29, B=@n30, Alpha=@n31))) {desc=Random Flicker Clamp 0 - 1}
  Reroute(None=Saturate(None=@n4)) {desc=0-1 Loop Clamped Delaypauses for a short time on 0 and 1 bc values extend above and below 0-1 range but are clamped}
  n32 = Add(A=@n1, B=@n3)
  n33 = Multiply(A=ComponentMask(None=Frac(None=@n32)) {g=true}, B=Constant() {desc=Scanline Y Tile, r=1}) {desc=Y tile}
  Reroute(None=ConstantBiasScale(None=Sine(None=@n33))) {desc=Scanline 0-1 Tiling Loop}
  Reroute(None=Saturate(None=@n15)) {desc=Scanline clamp 0-1 loop no Tile}
  Reroute(None=Saturate(None=@n12)) {desc=1-0 Wipe Loop}
  SmoothStep(Value=@n5) {constMax=0.55, constMin=0.45}
  n34 = ScalarParameter() {defaultValue=150, parameterName=NearFadeDist}
  n35 = DotProduct(A=CameraVectorWS(), B=VertexNormalWS())
  n36 = Multiply(A=Power(Base=Abs(None=@n35))) {constB=4}
  n37 = DepthFade(
   Opacity=Multiply(
    A=MaterialFunctionCall(Distance=@n34) {materialFunction=&Near_Camera_Fade}.Result
    B=Clamp(None=@n36)
   )
   FadeDistance=ScalarParameter() {defaultValue=12, parameterName=DepthFadeDist}
  )
  n38 = QualitySwitch(Default=@n37, Low=Constant() {r=1}, High=@n37, Medium=@n37)
  StaticSwitchParameter(True=Multiply(A=@n21, B=Multiply(A=@n25.A, B=@n38)), False=Multiply(A=Multiply(A=@n18.R, B=@n25.A), B=@n38)) {parameterName=Is Translucent}
  Step(Y=@n27) {constX=0.5}
 ]
}
