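# Test material in the Radiant effects folder for inspecting audio-driven values from the Radiant_Audio_MPC material parameter collection.
# It reads the MusicBeatProgress scalar and feeds it into the Number input of the DebugScalarValues material function, which is meant for numeric readout.
# The material is unlit, and the graph shows no link from the debug function to a material output.
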Material /Game/NeverCook/FM/Radiant/Effects/M_Test_Audio_MPC
graph {
 properties {bAutomaticallySetUsageInEditor=false, shadingModel=MSM_Unlit, usedShadingModels=Unlit}
 nodes [MaterialFunctionCall(Number=CollectionParameter() {collection=/Game/NeverCook/FM/Radiant/Audio/Radiant_Audio_MPC, parameterName=MusicBeatProgress}) {materialFunction=&DebugScalarValues}]
}
