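# Niagara beam effect for the Radiant assault asset, drawing a music-reactive beam between user-set start and end points.
# Cone meshes and multicolored lights interpolate along the beam, with music beat and amplitude pulsing light color and mesh size.
# Beat-driven bursts of glitch cube meshes and sprites are offset using the beam endpoints, alongside burst-spawned EQ box meshes.
# Audio debug sprite and EQ bar hit emitters are disabled.
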NiagaraSystem /Game/NeverCook/FM/Radiant/Effects/NS_Assault_Radient_ARG_Beam
dependencies {
 usedDataInterfaces [
  &NiagaraDataInterfaceCurve
  &NiagaraDataInterfaceColorCurve
  &NiagaraDataInterfaceArrayColor
  /Script/Niagara.NiagaraDataInterfaceAudioSpectrum
  &NiagaraDataInterfaceArrayFloat3
  &NiagaraDataInterfaceVector2DCurve
  &NiagaraDataInterfaceParticleRead
 ]
 usedDynamicInputs [&FindQuatBetweenTwoVectors]
 usedModules [
  &self:SystemSpawnScript.SystemScriptSource.SystemScriptGraph.NiagaraNodeAssignment_2.SetVariables_#1
  &SystemState
  &self:SystemSpawnScript.SystemScriptSource.SystemScriptGraph.NiagaraNodeAssignment_0.SetVariables_#2
  &self:SystemSpawnScript.SystemScriptSource.SystemScriptGraph.NiagaraNodeAssignment_1.SetVariables_#3
  &self:SystemSpawnScript.SystemScriptSource.SystemScriptGraph.NiagaraNodeAssignment_3.SetVariables_#4
 ]
}
emitters {
 AirflowConeHot {
  emitterSpawnScript [/Game/Effects/Niagara/Modules/Partition/FN_PartitionAttributes(Number of Bone Sets=1, Particles Per Partition=12)]
  emitterUpdateScript [
   &EmitterState(
    Inactive Response=::NewEnumerator1
    Life Cycle Mode=::NewEnumerator1
    Loop Behavior=::NewEnumerator1
    Loop Duration Mode=::NewEnumerator1
    UseLoopDelay=false
   )
   &SpawnBurst_Instantaneous(Spawn Count=1, Use Spawn Probability=false)
   &SpawnRate(Spawn Probability=0.5, SpawnRate=20, Use Spawn Probability=false) {enabled=false}
  ]
  particleSpawnScript [
   &InitializeParticle(
    Color=<rgba 15 0 14.4383 1>
    Lifetime=0.15
    Mesh Scale Mode=::NewEnumerator3
    Mesh Uniform Scale=1.25
    UsePositionOffset=true
    Write Material Random=true
   ) {module=InitializeParticle001}
   /Game/Effects/Niagara/Modules/Partition/FN_Particle_partition_v5/FN_Particle_Partition_v5(Enable Ribbon ID=true)
   &InitialMeshOrientation(
    Orientation Axis=<0 0 1>
    Orientation Coordinate Space=::World
    Orientation Vector=&RotateVectorByQuaternion(Vector To Rotate=<1 0 0>, Quaternion=$System.BeamOrientation)
    Use Orientation Vector=true
    Use Rotation Vector=false
   ) {enabled=false}
   &InitialMeshOrientation(Mesh Orientation Mode=::NewEnumerator6, Orientation Quaternion:Quat4f=$System.BeamOrientation, Use Rotation Vector=false) {module=InitialMeshOrientation001}
   /SetVariables(Particles.WobbleOverLength:NiagaraFloat=&FloatFromCurve(
     FloatCurve=DataInterface() {propertyValues {Curve {keys [
         {arriveTangent=5.2404, interpMode=RCIM_Cubic, leaveTangent=5.2404, tangentMode=RCTM_User, time=0, value=0}
         <at 0.5 1 RCIM_Cubic>
         {arriveTangent=-4.4363, interpMode=RCIM_Cubic, leaveTangent=-4.4363, tangentMode=RCTM_User, time=1, value=0}
        ]}}}
     CurveIndex=$Particles.Partition.NormalizedID
    )) {enabled=false, isSetParameters=true, module=SetVariables_#5}
   &AddVelocity(Velocity=<0.1 0 0>) {enabled=false}
  ]
  particleUpdateScript [
   &ParticleState(Kill Particles When Lifetime Has Elapsed=false)
   &ScaleMeshSize(Scale Factor=&VectorFromFloat(Value=&FloatFromCurve(FloatCurve=DataInterface() {propertyValues {Curve {keys [{interpMode=RCIM_Cubic, tangentMode=RCTM_User, time=0, value=1}, {interpMode=RCIM_Cubic, tangentMode=RCTM_User, time=1, value=0.5}]}}}))) {enabled=false}
   &SolveForcesAndVelocity(Clamp Velocity=false, Limit Acceleration=false) {enabled=false}
   /SetVariables(Particles.Position=&Lerp_Vector(
     A=$User.BeamStart_0_0
     B=Linked() {linkedVariable {name=User.BeamEnd_0, type=NiagaraPosition}}
     Alpha=&FloatFromCurve(FloatCurve=DataInterface() {propertyValues {Curve {keys [
          {tangentMode=RCTM_User, time=0, value=0}
          {arriveTangent=1, interpMode=RCIM_Cubic, leaveTangent=1, time=0.1, value=0}
          {tangentMode=RCTM_User, time=1, value=1}
         ]}}})
    )) {isSetParameters=true, module=SetVariables_#6}
   &ScaleColor(
    Curve Index=&ReturnNormalizedExecIndex(Normalized Index Scale=2)
    Linear Color Curve=DataInterface() {propertyValues {
      AlphaCurve {keys [<at 0 1>,<at 0.4487 1>,<at 1 1>]}
      BlueCurve {keys [<at 0 0.7379>,<at 0.134 0.5457>,<at 0.2454 0.7157>,<at 0.3458 0.073>,<at 0.4572 0.8714>,<at 0.5613 1>,<at 0.6617 1>,<at 0.7694 0.073>,<at 0.8759 0.9042>]}
      GreenCurve {keys [<at 0 0>,<at 0.134 0>,<at 0.2454 0.0176>,<at 0.3458 1>,<at 0.4572 1>,<at 0.5613 0.8797>,<at 0.6617 0>,<at 0.7694 1>,<at 0.8759 0>]}
      RedCurve {keys [<at 0 0.7682>,<at 0.134 0.1779>,<at 0.2454 0.0782>,<at 0.3458 0>,<at 0.4572 0>,<at 0.5613 0>,<at 0.6617 0.0423>,<at 0.7694 0>,<at 0.8759 1>]}
     }}
    Scale Mode=::NewEnumerator2
   ) {enabled=false, module=ScaleColor001}
   &ScaleColor(
    Scale Alpha=&FloatFromCurve(FloatCurve=DataInterface() {propertyValues {Curve {keys [{interpMode=RCIM_Cubic, tangentMode=RCTM_User, time=0, value=1}, {interpMode=RCIM_Cubic, tangentMode=RCTM_User, time=1, value=0.1}]}}})
    ScaleA=true
    ScaleRGB=true
   )
  ]
  rendererClasses [&NiagaraMeshRendererProperties]
  renderers [
   {rendererClass=&NiagaraMeshRendererProperties, rendererIndex=0}
   {rendererClass=&NiagaraMeshRendererProperties, rendererIndex=1}
  ]
 }
 Beam_Lights {
  emitterSpawnScript [&FN_PartitionAttributes(Number of Bone Sets=1, Particles Per Partition=12)]
  emitterUpdateScript [
   &EmitterState(Life Cycle Mode=::NewEnumerator1, Loop Duration=0.5, UseLoopDelay=false)
   &SpawnPerUnit(
    Module.Use Max Movement Threshold=true
    Module.Use Movement Tolerance=true
    Movement Tolerance=0
    Spawn Probability=1
    Spawn Spacing=30
    Use Spawn Probability=true
   ) {enabled=false}
   /Game/Effects/Niagara/Modules/Spawn/FN_SpawnByEngineDistance(Multiplier=5, Step=20) {enabled=false}
   &SpawnBurst_Instantaneous(Spawn Count=10, Spawn Probability=0.5, Use Spawn Probability=true, UseLoopCountLimit=true) {enabled=false, module=SpawnBurst_Instantaneous001}
   &SpawnBurst_Instantaneous(Spawn Count=$Emitter.FN_PartitionAttributes.Spawn Total, Use Spawn Probability=false) {enabled=false}
   &SpawnRate(Spawn Probability=0.5, SpawnRate=500, Use Spawn Probability=false)
  ]
  particleSpawnScript [
   &PartitionParticles(Number Of Partitions=4)
   &InitializeParticle(
    Lifetime Max=0.25
    Lifetime Min=0.15
    Lifetime Mode=::NewEnumerator1
    Mesh Scale Mode=::NewEnumerator3
    Mesh Uniform Scale=2
    Sprite Size Mode=::NewEnumerator3
    UsePositionOffset=true
    Write Material Random=true
   ) {module=InitializeParticle001}
   &FN_Particle_Partition_v5(Enable Ribbon ID=true)
   &InitializeRibbon(
    Color=<rgba 1 0 0.6104 0.15>
    Lifetime=0.7
    Module.Write Color=true
    Module.Write Lifetime=true
    Module.Write Mass=true
    Module.Write Material Random=true
    Module.Write Position=false
    Module.Write RibbonTwist=false
    Module.Write RibbonWidth=true
    Ribbon Width=14
   ) {enabled=false}
   &InitialMeshOrientation(
    Orientation Axis=<0 0 1>
    Orientation Coordinate Space=::World
    Orientation Vector=&RotateVectorByQuaternion(Vector To Rotate=<1 0 0>, Quaternion=$System.BeamOrientation)
    Use Orientation Vector=true
    Use Rotation Vector=false
   )
   &InitialMeshOrientation(Mesh Orientation Mode=::NewEnumerator6, Orientation Quaternion:Quat4f=$System.BeamOrientation, Use Rotation Vector=false) {enabled=false, module=InitialMeshOrientation001}
   /SetVariables(Particles.WobbleOverLength:NiagaraFloat=&FloatFromCurve(
     FloatCurve=DataInterface() {propertyValues {Curve {keys [
         {arriveTangent=5.2404, interpMode=RCIM_Cubic, leaveTangent=5.2404, tangentMode=RCTM_User, time=0, value=0}
         <at 0.5 1 RCIM_Cubic>
         {arriveTangent=-4.4363, interpMode=RCIM_Cubic, leaveTangent=-4.4363, tangentMode=RCTM_User, time=1, value=0}
        ]}}}
     CurveIndex=$Particles.Partition.NormalizedID
    )) {enabled=false, isSetParameters=true, module=SetVariables_#5}
   &AddVelocity(Velocity=<0.1 0 0>) {enabled=false}
  ]
  particleUpdateScript [
   &ParticleState()
   &ScaleMeshSize(Scale Factor=&VectorFromFloat(Value=&Lerp_Float(
      A=&FloatFromCurve(
       FloatCurve=DataInterface() {propertyValues {Curve {keys [
           {arriveTangent=9.0725, interpMode=RCIM_Cubic, leaveTangent=9.0725, tangentMode=RCTM_User, time=0, value=0}
           {arriveTangent=-2.5249, interpMode=RCIM_Cubic, leaveTangent=-2.5249, tangentMode=RCTM_User, time=0.1408, value=0.9847}
           {arriveTangent=-0.0405, interpMode=RCIM_Cubic, leaveTangent=-0.0405, tangentMode=RCTM_User, time=1, value=0}
          ]}}}
       Scale Curve=&Lerp_Float(
        A=0.5
        B=1
        Alpha=&MakeFloatFromVector(
         Vector=&MakeVectorFromLinearColorRGB(Color=$NPC.NPC_Audio_Reactive_Parameters.Spectrum_Music)
         Channel=::NewEnumerator2
        )
       )
      )
      B=&FloatFromCurve(
       FloatCurve=DataInterface() {propertyValues {Curve {keys [
           {arriveTangent=9.0725, interpMode=RCIM_Cubic, leaveTangent=9.0725, tangentMode=RCTM_User, time=0, value=0}
           {arriveTangent=-2.5249, interpMode=RCIM_Cubic, leaveTangent=-2.5249, tangentMode=RCTM_User, time=0.1408, value=0.9847}
           {arriveTangent=-0.0405, interpMode=RCIM_Cubic, leaveTangent=-0.0405, tangentMode=RCTM_User, time=1, value=0}
          ]}}}
       Scale Curve=&Lerp_Float(
        A=0.5
        B=1
        Alpha=&MakeFloatFromVector(
         Vector=&MakeVectorFromLinearColorRGB(Color=$NPC.NPC_Audio_Reactive_Parameters.Spectrum_Music)
         Channel=::NewEnumerator2
        )
       )
      )
      Alpha=$Particles.PartitionParticles.NormalizedPartition
     ))) {enabled=false, module=ScaleMeshSize004}
   &ScaleMeshSize(Scale Factor=&VectorFromFloat(Value=&FloatFromCurve(
      FloatCurve=DataInterface() {propertyValues {Curve {keys [
          {interpMode=RCIM_Cubic, tangentMode=RCTM_User, time=0, value=0}
          {arriveTangent=1.25, leaveTangent=1.25, time=0.2, value=1}
          {arriveTangent=-1.25, interpMode=RCIM_Cubic, leaveTangent=-1.25, time=0.8, value=1}
          {tangentMode=RCTM_User, time=1, value=0}
         ]}}}
      Scale Curve=1.5
     ))) {module=ScaleMeshSize003}
   &ScaleMeshSize(Scale Factor=&VectorFromFloat(Value=&FloatFromCurve(
      FloatCurve=DataInterface() {propertyValues {Curve {keys [
          {interpMode=RCIM_Cubic, tangentMode=RCTM_User, time=0, value=0.55}
          {arriveTangent=1.0101, time=0.1, value=1.05}
          {interpMode=RCIM_Cubic, tangentMode=RCTM_User, time=0.9, value=1}
          {tangentMode=RCTM_User, time=1, value=0}
         ]}}}
      CurveIndex=&ReturnNormalizedExecIndex()
     ))) {enabled=false, module=ScaleMeshSize002}
   &ScaleMeshSize(Scale Factor=&VectorFromFloat(Value=&RandomRangeFloat(
      Minimum=&Lerp_Float(
       A=0.5
       B=1.1
       Alpha=&FloatFromCurve(
        FloatCurve=DataInterface() {propertyValues {Curve {keys [
            {arriveTangent=9.0725, interpMode=RCIM_Cubic, time=0, value=0}
            <at 0.1 1 RCIM_Cubic>
            {arriveTangent=-0.0088, interpMode=RCIM_Cubic, leaveTangent=-0.0088, tangentMode=RCTM_User, time=1, value=0}
           ]}}}
        CurveIndex=$NPC.Radiant_Music_NPC.MusicBeatProgress
       )
      )
      Maximum=$NPC.NPC_Audio_Reactive_Parameters.Amplitude_Music
     ))) {module=ScaleMeshSize005}
   &ScaleMeshSize(Scale Factor=&VectorFromFloat(Value=&Add_Float(
      A=&Sine(Normalized Angle=&Subtract_Float(A=$Particles.NormalizedAge, B=$Emitter.Age), Period=0.5, Scale=0.25, Bias=0)
      B=1
     ))) {enabled=false}
   &ScaleMeshSize(Scale Factor=&MakeVector(
     X=&Add_Float(
      A=&Sine(Normalized Angle=&Subtract_Float(A=$Particles.NormalizedAge, B=$Emitter.Age), Period=0.5, Scale=0.25, Bias=0)
      B=1
     )
     Y=&Add_Float(
      A=&Sine(Normalized Angle=&Subtract_Float(A=$Particles.NormalizedAge, B=$Emitter.Age), Period=0.5, Scale=0.25, Bias=0)
      B=1
     )
     Z=0.75
    )) {enabled=false, module=ScaleMeshSize001}
   /SetVariables(Particles.Position=&Lerp_Vector(
     A=&TransformVector(Vector=$User.BeamStart_0_0, Source Space=::World, Destination Space=::World)
     B=&TransformVector(Vector=Linked() {linkedVariable {name=User.BeamEnd_0, type=NiagaraPosition}}, Source Space=::World, Destination Space=::World)
     Alpha=&FloatFromCurve(FloatCurve=DataInterface() {propertyValues {Curve {keys [{tangentMode=RCTM_User, time=0, value=0}, {tangentMode=RCTM_User, time=1, value=1}]}}})
    )) {enabled=false, isSetParameters=true, module=SetVariables_#7}
   /SetVariables(Particles.Position=&Lerp_Vector(
     A=$User.BeamStart_0_0
     B=Linked() {linkedVariable {name=User.BeamEnd_0, type=NiagaraPosition}}
     Alpha=&FloatFromCurve(FloatCurve=DataInterface() {propertyValues {Curve {keys [{tangentMode=RCTM_User, time=0, value=0}, {tangentMode=RCTM_User, time=1, value=1}]}}})
    )) {isSetParameters=true, module=SetVariables_#6}
   &KillParticles(Kill Particles=&self:ScratchDynamicInput(BeamEnd:Vector3f=Linked() {linkedVariable {name=User.BeamEnd_0, type=NiagaraPosition}})) {enabled=false}
   &ScaleColor(
    Curve Index=&ReturnNormalizedExecIndex(Normalized Index Scale=2)
    Linear Color Curve=DataInterface() {propertyValues {
      AlphaCurve {keys [<at 0 1>,<at 0.4487 1>,<at 1 1>]}
      BlueCurve {keys [<at 0 0.7379>,<at 0.134 0.5457>,<at 0.2454 0.7157>,<at 0.3458 0.073>,<at 0.4572 0.8714>,<at 0.5613 1>,<at 0.6617 1>,<at 0.7694 0.073>,<at 0.8759 0.9042>]}
      GreenCurve {keys [<at 0 0>,<at 0.134 0>,<at 0.2454 0.0176>,<at 0.3458 1>,<at 0.4572 1>,<at 0.5613 0.8797>,<at 0.6617 0>,<at 0.7694 1>,<at 0.8759 0>]}
      RedCurve {keys [<at 0 0.7682>,<at 0.134 0.1779>,<at 0.2454 0.0782>,<at 0.3458 0>,<at 0.4572 0>,<at 0.5613 0>,<at 0.6617 0.0423>,<at 0.7694 0>,<at 0.8759 1>]}
     }}
    Scale Mode=::NewEnumerator2
   ) {enabled=false, module=ScaleColor001}
   &ScaleColor(
    Scale RGB=&VectorFromFloat(Value=&Lerp_Float(
      A=0.35
      B=5
      Alpha=&FloatFromCurve(
       FloatCurve=DataInterface() {propertyValues {Curve {keys [
           {arriveTangent=9.0725, interpMode=RCIM_Cubic, time=0, value=0}
           <at 0.1 1 RCIM_Cubic>
           {arriveTangent=-0.0088, interpMode=RCIM_Cubic, leaveTangent=-0.0088, tangentMode=RCTM_User, time=1, value=0}
          ]}}}
       CurveIndex=$NPC.Radiant_Music_NPC.MusicBeatProgress
      )
     ))
    ScaleA=true
    ScaleRGB=true
   ) {enabled=false, module=ScaleColor002}
   &FN_Light_v3(
    Color=&SelectLinearColorFromArray(
     Array Sampling Mode=::NewEnumerator1
     Direct Array Index=$Particles.Initial.PartitionParticles.Partition
     Color Selection Array=DataInterface() {propertyValues {ColorData [<rgba 1 0 0.9347 1>, <rgba 0 0.8332 1 1>, <rgba 0.1987 2 0 1>, <rgba 0 0.0149 1 1>]}}
     Random Seed=?
    )
    Create Color Attribute=true
    Create Exponent Attribute=true
    Create Position Attribute=true
    Create Radius Attribute=true
    Enable Particle Light Enable=true
    Exponent=8
    Light Enabled=true
    Position:Vector3f=&AddVectorToPosition(Position=$Particles.Position, Vector=<0 0 0>)
    Radius=350
   )
   /SetVariables(Particles.FN_Light_v3.Color:LinearColor=&Multiply_LinearColorByFloat(
     Linear Color=$Particles.FN_Light_v3.Color
     Float=&MakeFloatFromVector(
      Vector=&VectorFromFloat(Value=&Lerp_Float(
        A=0.35
        B=1
        Alpha=&FloatFromCurve(
         FloatCurve=DataInterface() {propertyValues {Curve {keys [
             {arriveTangent=9.0725, interpMode=RCIM_Cubic, time=0, value=0}
             <at 0.1 1 RCIM_Cubic>
             {arriveTangent=-0.0088, interpMode=RCIM_Cubic, leaveTangent=-0.0088, tangentMode=RCTM_User, time=1, value=0}
            ]}}}
         CurveIndex=$NPC.Radiant_Music_NPC.MusicBeatProgress
        )
       ))
      Channel=::NewEnumerator0
     )
    )) {isSetParameters=true}
  ]
  rendererClasses [
   &NiagaraMeshRendererProperties
   &NiagaraLightRendererProperties
   &NiagaraSpriteRendererProperties
  ]
  renderers [
   {rendererClass=&NiagaraMeshRendererProperties, rendererIndex=0}
   {rendererClass=&NiagaraLightRendererProperties, rendererIndex=1}
   {rendererClass=&NiagaraSpriteRendererProperties, rendererIndex=2}
  ]
 }
 Cube_Glitch {
  emitterUpdateScript [
   &EmitterState()
   &SpawnBurst_Instantaneous(
    Spawn Count=&Multiply_Int(
     A=500
     B=&MakeIntFromFloat(Float=$NPC.Radiant_Music_NPC.MusicBeatProgress, Conversion Mode=::NewEnumerator1)
    )
    Spawn Probability=$NPC.Radiant_Music_NPC.MusicBeatProgress
    Use Spawn Probability=false
    UseLoopCountLimit=false
   )
   &SpawnRate(Spawn Probability=0.85, SpawnRate=50, Use Spawn Probability=true) {enabled=false}
  ]
  particleSpawnScript [
   &PartitionParticles(Number Of Partitions=8, Partition Mode:ENiagara_PartitionMode=::NewEnumerator5) {enabled=false}
   /SetVariables(Particles.ScaleOVerLength:NiagaraFloat=&FloatFromCurve(
     FloatCurve=DataInterface() {propertyValues {Curve {keys [
         {arriveTangent=0.1131, interpMode=RCIM_Cubic, leaveTangent=0.1131, tangentMode=RCTM_User, time=0, value=0}
         <at 0.5 1 RCIM_Cubic>
         {arriveTangent=-0.0182, interpMode=RCIM_Cubic, leaveTangent=-0.0182, tangentMode=RCTM_User, time=1, value=0}
        ]}}}
     CurveIndex=&ReturnNormalizedExecIndex()
    )) {enabled=false, isSetParameters=true, module=SetVariables_#8}
   &InitializeParticle(
    Color=&SelectLinearColorFromArray(
     Array Sampling Mode=::NewEnumerator0
     Direct Array Index=?
     Color Selection Array=DataInterface() {propertyValues {ColorData [<rgba 2 0 1.7477 1>, <rgba 0 2 0.347 1>, <rgba 3 0 0.4415 1>]}}
     Random Seed=0
    )
    Lifetime Max=0.5
    Lifetime Min=0.25
    Lifetime Mode=::NewEnumerator1
    Mesh Scale Max=<5 5 5>
    Mesh Scale Min=<3 3 3>
    Mesh Scale Mode=::NewEnumerator2
    Sprite Size Max=<1 1>
    Sprite Size Min=<0.1 0.1>
    Sprite Size Mode=::NewEnumerator2
    UsePositionOffset=false
    Write Material Random=true
   )
   &DynamicMaterialParameters(
    Index 0 Param 3=0.35
    Index 0 Param 4=&RandomRangeFloat(Minimum=-1, Maximum=0.1)
    Index 1 Param 1=&RandomRangeFloat(Minimum=-100, Maximum=100)
    Module.Param0WriteEnabled=true
    Module.Param0WriteEnabled1=true
    Module.Param1WriteEnabled=true
    Module.Param1WriteEnabled1=true
    Module.Param2WriteEnabled=true
    Module.Param2WriteEnabled1=true
    Module.Param3WriteEnabled=true
    Module.Param3WriteEnabled1=true
    Write Parameter Index 1=true
   ) {module=DynamicMaterialParameters001}
   &ShapeLocation(MaskSpawnGroup=false, Override Seed=false, Sphere Radius=120, Sphere Surface Distribution=0.7976, UseOwnerScale=true)
  ]
  particleUpdateScript [
   &ParticleState()
   &ScaleSpriteSize(
    Non-Uniform Curve Scale=<3 3>
    Non-Uniform Curve Sprite Scale=DataInterface() {propertyValues {
      XCurve {keys [
        {arriveTangent=9.0725, interpMode=RCIM_Cubic, leaveTangent=9.0725, tangentMode=RCTM_User, time=0, value=0}
        {arriveTangent=-2.5249, interpMode=RCIM_Cubic, leaveTangent=-2.5249, tangentMode=RCTM_User, time=0.1408, value=0.9847}
        {arriveTangent=-0.0405, interpMode=RCIM_Cubic, leaveTangent=-0.0405, tangentMode=RCTM_User, time=1, value=0}
       ]}
      YCurve {keys [
        {arriveTangent=11.1459, interpMode=RCIM_Cubic, leaveTangent=11.1459, tangentMode=RCTM_User, time=0.2, value=-0.125}
        {arriveTangent=-4.9839, interpMode=RCIM_Cubic, leaveTangent=-4.9839, tangentMode=RCTM_User, time=0.3, value=1}
        {tangentMode=RCTM_User, time=1, value=0}
       ]}
     }}
    Scale Sprite Size Mode=::NewEnumerator3
   ) {enabled=false}
   &RotateAroundPoint(
    Derive the Rotation Axis=false
    Radius:NiagaraFloat=&Multiply_Float(
     A=&Lerp_Float(
      A=&RandomRangeFloat(Minimum=1, Maximum=1.5)
      B=&RandomRangeFloat(Minimum=3, Maximum=6)
      Alpha=&FloatFromCurve(
       FloatCurve=DataInterface() {propertyValues {Curve {keys [
           {arriveTangent=3.8353, interpMode=RCIM_Cubic, leaveTangent=3.8353, tangentMode=RCTM_User, time=0, value=0.5}
           {arriveTangent=-0.5, interpMode=RCIM_Cubic, leaveTangent=-0.5, time=0.3, value=0.9847}
           {interpMode=RCIM_Cubic, leaveTangent=-1.7958, time=1, value=0}
          ]}}}
       Scale Curve=0.1
      )
     )
     B=&MaskFloatBySpawnGroup(Masked Float=1.15, Passthrough Float=0.7)
    )
    Rotation Axis:Vector3f=$System.DiveVelocityScale
    Rotation Phase:NiagaraFloat=&RandomRangeFloat(Minimum=0, Maximum=1)
    Rotation Rate:NiagaraFloat=&MakeCustomFloatFromBool(Bool=&RandomBool(), True Float=0.6, False Float=-0.6)
    Set Particle Position=false
   ) {enabled=false}
   /SetVariables(Particles.Rotational_Position:NiagaraPosition=&ConvertVectorToPosition(
     Input Position=&Add_Vector(
      A=&ConvertPositionToVector(Input Position=$Particles.Position, Convert To Absolute Large World Vector=false)
      B=Linked() {linkedVariable {name=Output.RotateAroundPoint.OutputPosition, type=NiagaraPosition}}
     )
     Vector Is In Absolute Large World Space:NiagaraBool=true
    )) {enabled=false, isSetParameters=true}
   &ScaleMeshSize(Scale Factor=&Multiply_VectorByFloat(
     Vector=<1 1 1>
     Float=&Multiply_Float(
      A=0.075
      B=&FloatFromCurve(
       FloatCurve=DataInterface() {propertyValues {Curve {keys [
           {arriveTangent=-1.4768, interpMode=RCIM_Cubic, leaveTangent=-1.4768, tangentMode=RCTM_User, time=0, value=5}
           {arriveTangent=-0.9375, interpMode=RCIM_Cubic, leaveTangent=-0.9375, time=0.3, value=3.25}
           {arriveTangent=2.5, interpMode=RCIM_Cubic, leaveTangent=2.5, time=0.8, value=4.25}
           {arriveTangent=-20.9401, interpMode=RCIM_Cubic, leaveTangent=-20.9401, tangentMode=RCTM_User, time=0.9, value=4.75}
           <at 1 0.25>
          ]}}}
       Scale Curve=&Multiply_Float(
        A=0.6
        B=&ClampFloat(
         Float=&FN_LOD_Distance_Remap(Start Distance=200, End Distance=400, Minimum Output=0.5, Maximum Output=1)
         Min=0.5
        )
       )
      )
     )
    )) {module=ScaleMeshSize001}
   &ScaleColor(
    Scale Mode=::NewEnumerator1
    Scale RGBA=&MakeVector4_FromLinearColor(LinearColor=&LerpLinearColors(StartColor=<rgba 0.0445 0.4948 0.3751 1>, EndColor=<rgba 0.0081 0.3646 0.2803 1>, LerpFactor=$Particles.ZDistance))
    ScaleRGBA=true
   )
   &CameraOffset(Camera Offset Amount=-10) {enabled=false}
   /SetVariables(Particles.Position=&AddVectorToPosition(
     Position=$Particles.Initial.ShapeLocation.ShapePosition
     Vector=&self:ScratchDynamicInput_01(
      Default:Vector3f=&TransformVector(Vector=$User.BeamStart_0_0, Source Space=::World, Destination Space=::Local)
      Vector:Vector3f=&TransformVector(Vector=Linked() {linkedVariable {name=User.BeamEnd_0, type=NiagaraPosition}}, Source Space=::World, Destination Space=::Local)
      Spacing:NiagaraFloat=25
     )
    )) {isSetParameters=true, module=SetVariables_#9}
   &DynamicMaterialParameters(
    Index 0 Param 2=&RandomRangeFloat(Minimum=-1, Maximum=1)
    Index 0 Param 3=&RandomRangeFloat(
     Minimum=&FloatFromCurve(FloatCurve=DataInterface() {propertyValues {Curve {keys [
          <at 0 1.375 RCIM_Cubic>
          {arriveTangent=-2.7589, interpMode=RCIM_Cubic, leaveTangent=-2.7589, tangentMode=RCTM_User, time=0.7, value=0.875}
          {arriveTangent=-0.8021, interpMode=RCIM_Cubic, leaveTangent=-0.8021, tangentMode=RCTM_User, time=1, value=0.5}
         ]}}})
     Maximum=&FloatFromCurve(FloatCurve=DataInterface() {propertyValues {Curve {keys [{tangentMode=RCTM_User, time=0, value=2}, {tangentMode=RCTM_User, time=1, value=2}]}}})
    )
    Index 0 Param 4=1
    Index 1 Param 4=&FloatFromCurve(FloatCurve=DataInterface() {propertyValues {Curve {keys [
         {arriveTangent=4.1785, interpMode=RCIM_Cubic, leaveTangent=4.1785, tangentMode=RCTM_User, time=0, value=0}
         {arriveTangent=0.9995, interpMode=RCIM_Cubic, leaveTangent=0.9995, tangentMode=RCTM_User, time=0.2223, value=0.6825}
         {arriveTangent=-0.0405, interpMode=RCIM_Cubic, leaveTangent=-0.0405, tangentMode=RCTM_User, time=1, value=1}
        ]}}})
    Module.Param0WriteEnabled=false
    Module.Param0WriteEnabled1=false
    Module.Param1WriteEnabled=true
    Module.Param1WriteEnabled1=false
    Module.Param2WriteEnabled=true
    Module.Param2WriteEnabled1=false
    Module.Param3WriteEnabled=true
    Module.Param3WriteEnabled1=true
    Write Parameter Index 1=true
   ) {enabled=false, module=DynamicMaterialParameters002}
   &DynamicMaterialParameters(
    Index 0 Param 2=&RandomRangeFloat(Minimum=-1, Maximum=1)
    Index 0 Param 3=&RandomRangeFloat(Minimum=0.35, Maximum=0.7)
    Index 0 Param 4=&RandomRangeFloat(Minimum=-1, Maximum=0.1)
    Index 1 Param 1=&RandomRangeFloat(Minimum=-100, Maximum=100)
    Index 1 Param 4=&FloatFromCurve(FloatCurve=DataInterface() {propertyValues {Curve {keys [
         {arriveTangent=4.1785, interpMode=RCIM_Cubic, leaveTangent=4.1785, tangentMode=RCTM_User, time=0, value=0}
         {arriveTangent=0.9995, interpMode=RCIM_Cubic, leaveTangent=0.9995, tangentMode=RCTM_User, time=0.2223, value=0.6825}
         {arriveTangent=-0.0405, interpMode=RCIM_Cubic, leaveTangent=-0.0405, tangentMode=RCTM_User, time=1, value=1}
        ]}}})
    Module.Param0WriteEnabled=true
    Module.Param0WriteEnabled1=true
    Module.Param1WriteEnabled=true
    Module.Param1WriteEnabled1=true
    Module.Param2WriteEnabled=true
    Module.Param2WriteEnabled1=true
    Module.Param3WriteEnabled=true
    Module.Param3WriteEnabled1=true
    Write Parameter Index 1=true
   )
   &ScaleMeshSize(Scale Factor=&VectorFromFloat(Value=1))
   &ScaleColor(
    Scale Alpha=$NPC.Radiant_Music_NPC.MusicBeatProgress
    Scale RGB=&VectorFromFloat(Value=&Lerp_Float(
      A=0
      B=5
      Alpha=&FloatFromCurve(FloatCurve=DataInterface() {propertyValues {Curve {keys [
           {arriveTangent=9.0725, interpMode=RCIM_Cubic, time=0, value=0}
           <at 0.1 1 RCIM_Cubic>
           {arriveTangent=-0.0088, interpMode=RCIM_Cubic, leaveTangent=-0.0088, tangentMode=RCTM_User, time=1, value=0}
          ]}}})
     ))
    ScaleA=true
    ScaleRGB=true
   ) {enabled=false, module=ScaleColor002}
  ]
  rendererClasses [&NiagaraMeshRendererProperties]
  renderers [{rendererClass=&NiagaraMeshRendererProperties, rendererIndex=0}]
 }
 Cube_Glitch001 {
  emitterUpdateScript [
   &EmitterState()
   &SpawnRate(Spawn Probability=0.85, SpawnRate=150, Use Spawn Probability=true) {enabled=false}
   &SpawnBurst_Instantaneous(
    Spawn Count=&Multiply_Int(
     A=10
     B=&MakeIntFromFloat(Float=$NPC.Radiant_Music_NPC.MusicBeatProgress, Conversion Mode=::NewEnumerator1)
    )
    Spawn Probability=$NPC.Radiant_Music_NPC.MusicBeatProgress
    Use Spawn Probability=false
    UseLoopCountLimit=false
   )
  ]
  particleSpawnScript [
   /SetVariables(Particles.ChromaOffset=50) {isSetParameters=true}
   &InitializeParticle(
    Color=&SelectLinearColorFromArray(
     Array Sampling Mode=::NewEnumerator0
     Direct Array Index=?
     Color Selection Array=DataInterface() {propertyValues {ColorData [<rgba 2 0 1.7477 1>, <rgba 0 2 0.347 1>, <rgba 3 0 0.4415 1>]}}
     Random Seed=0
    )
    Lifetime Max=0.2
    Lifetime Min=0.1
    Lifetime Mode=::NewEnumerator1
    Mesh Scale Max=<5 5 5>
    Mesh Scale Min=<3 3 3>
    Mesh Scale Mode=::NewEnumerator2
    Sprite Size Max=<0.7 0.7>
    Sprite Size Min=<0.1 0.1>
    Sprite Size Mode=::NewEnumerator2
    UsePositionOffset=false
    Write Material Random=true
   )
   &AddVelocity(Rotation Coordinate Space=::World, Velocity=<0 0 -50>) {module=AddVelocity001}
   &ShapeLocation(MaskSpawnGroup=false, Override Seed=false, Sphere Surface Distribution=1, UseOwnerScale=true)
  ]
  particleUpdateScript [
   &ParticleState()
   /SetVariables(Particles.ChromaOffset:NiagaraFloat=&FloatFromCurve(
     FloatCurve=DataInterface() {propertyValues {Curve {keys [{interpMode=RCIM_Cubic, tangentMode=RCTM_User, time=0, value=1}, {interpMode=RCIM_Cubic, tangentMode=RCTM_User, time=1, value=0}]}}}
     Scale Curve=0.5
    )) {isSetParameters=true}
   &ScaleMeshSize(Scale Factor=&Multiply_VectorByFloat(
     Vector=<1 1 1>
     Float=&Multiply_Float(
      A=0.075
      B=&FloatFromCurve(
       FloatCurve=DataInterface() {propertyValues {Curve {keys [
           {arriveTangent=-1.4768, interpMode=RCIM_Cubic, leaveTangent=-1.4768, tangentMode=RCTM_User, time=0, value=5}
           {arriveTangent=-0.9375, interpMode=RCIM_Cubic, leaveTangent=-0.9375, time=0.3, value=3.25}
           {arriveTangent=2.5, interpMode=RCIM_Cubic, leaveTangent=2.5, time=0.8, value=4.25}
           {arriveTangent=-20.9401, interpMode=RCIM_Cubic, leaveTangent=-20.9401, tangentMode=RCTM_User, time=0.9, value=4.75}
           <at 1 0.25>
          ]}}}
       Scale Curve=&Multiply_Float(
        A=0.6
        B=&ClampFloat(
         Float=&FN_LOD_Distance_Remap(Start Distance=200, End Distance=400, Minimum Output=0.5, Maximum Output=1)
         Min=0.5
        )
       )
      )
     )
    )) {module=ScaleMeshSize001}
   /SetVariables(Particles.Color001:LinearColor=<rgba 0.005 0 0.1 0.5>) {isSetParameters=true}
   /SetVariables(Particles.Color002=<rgba 0.0206 1 0.6109 0.5>) {isSetParameters=true}
   /SetVariables(Particles.Color003:LinearColor=<rgba 0.8542 0 1 0.5>) {isSetParameters=true}
   &ScaleSpriteSize(Scale Factor=&RandomRangeVector2D(
     Minimum=&Vector2DFromCurve(
      Vector2Curve=DataInterface() {propertyValues {
        XCurve {keys [
          {interpMode=RCIM_Cubic, tangentMode=RCTM_User, time=0, value=1}
          {arriveTangent=6.1847, interpMode=RCIM_Cubic, leaveTangent=6.1847, tangentMode=RCTM_User, time=0.8, value=2.5}
          {tangentMode=RCTM_User, time=1, value=0}
         ]}
        YCurve {keys [
          {interpMode=RCIM_Cubic, tangentMode=RCTM_User, time=0, value=0}
          {arriveTangent=13.8249, interpMode=RCIM_Cubic, leaveTangent=13.8249, tangentMode=RCTM_User, time=0.4, value=3.5}
          {arriveTangent=8.3745, interpMode=RCIM_Cubic, leaveTangent=8.3745, tangentMode=RCTM_User, time=1, value=8}
         ]}
       }}
      CurveIndex=&Waveform("Waveform [1]"=::NewEnumerator0, "[1] Frequency"=1, "[1] Amplitude Min/Max"=<0 1>)
      ScaleCurve=<2 8>
     )
     Maximum=&Vector2DFromCurve(
      Vector2Curve=DataInterface() {propertyValues {
        XCurve {keys [
          {interpMode=RCIM_Cubic, tangentMode=RCTM_User, time=0, value=1}
          {arriveTangent=7.224, interpMode=RCIM_Cubic, leaveTangent=7.224, tangentMode=RCTM_User, time=0.6, value=1.5}
          {tangentMode=RCTM_User, time=1, value=8}
         ]}
        YCurve {keys [
          {interpMode=RCIM_Cubic, tangentMode=RCTM_User, time=0, value=0}
          {arriveTangent=0.5387, interpMode=RCIM_Cubic, leaveTangent=0.5387, tangentMode=RCTM_User, time=0.5, value=2}
          {arriveTangent=8.3745, interpMode=RCIM_Cubic, leaveTangent=8.3745, tangentMode=RCTM_User, time=1, value=0}
         ]}
       }}
      CurveIndex=&Waveform("Waveform [1]"=::NewEnumerator0, "[1] Frequency"=1, "[1] Amplitude Min/Max"=<0 1>)
      ScaleCurve=<2 8>
     )
    ))
   /SetVariables(Particles.SpriteSize001:Vector2f=&Multiply_Vector2D_ByFloat(Float=1, Vector2D=$Particles.SpriteSize)) {enabled=false, isSetParameters=true}
   /SetVariables(Particles.Position=&AddVectorToPosition(
     Position=$Particles.Initial.ShapeLocation.ShapePosition
     Vector=&self:ScratchDynamicInput_01(
      Default:Vector3f=&TransformVector(Vector=$User.BeamStart_0_0, Source Space=::World, Destination Space=::Local)
      Vector:Vector3f=&TransformVector(Vector=Linked() {linkedVariable {name=User.BeamEnd_0, type=NiagaraPosition}}, Source Space=::World, Destination Space=::Local)
      Spacing:NiagaraFloat=35
     )
    )) {isSetParameters=true, module=SetVariables_#9}
   &SolveForcesAndVelocity(Clamp Velocity=false, Limit Acceleration=false)
  ]
  rendererClasses [&NiagaraSpriteRendererProperties]
  renderers [
   {rendererClass=&NiagaraSpriteRendererProperties}
   {rendererClass=&NiagaraSpriteRendererProperties, rendererIndex=1}
   {rendererClass=&NiagaraSpriteRendererProperties, rendererIndex=2}
  ]
 }
 Debug_Audio_Reactivity_Sprite {
  bEnabled=false
  emitterUpdateScript [
   &EmitterState(Life Cycle Mode=::NewEnumerator1, Loop Behavior=::NewEnumerator1, Loop Duration Mode=::NewEnumerator1, UseLoopDelay=false)
   &SpawnBurst_Instantaneous(Use Spawn Probability=false, UseLoopCountLimit=false)
  ]
  particleSpawnScript [
   &InitializeParticle(
    Lifetime Max=2.5
    Lifetime Mode=::NewEnumerator1
    Sprite Size Mode=::NewEnumerator3
    Uniform Sprite Size=25
    UsePositionOffset=false
    Write Material Random=true
   )
   &ShapeLocation(MaskSpawnGroup=false, Override Seed=false, Sphere Radius=40, UseOwnerScale=true) {enabled=false}
   &AddVelocity(
    Module.Use Falloff Distance=false
    Origin Offset=<0 0 -7>
    Override Seed=false
    Velocity Mode=::NewEnumerator1
    Velocity Speed=&RandomRangeFloat(Minimum=75, Maximum=400)
   ) {enabled=false}
  ]
  particleUpdateScript [
   &ParticleState(Kill Particles When Lifetime Has Elapsed=false, Let Infinitely Lived Particles Die When Emitter Deactivates=true)
   &GravityForce() {enabled=false}
   &Drag(Drag=0.75, Use Linear Drag=true, Use Rotational Drag=false) {enabled=false}
   &ScaleSpriteSize(Scale Factor=&Vector2DFromFloat(Value=&Lerp_Float(
      A=0.5
      B=3
      Alpha=&FloatFromCurve(
       FloatCurve=DataInterface() {propertyValues {Curve {keys [
           {arriveTangent=9.0725, interpMode=RCIM_Cubic, leaveTangent=9.0725, tangentMode=RCTM_User, time=0, value=0}
           {arriveTangent=-2.5249, interpMode=RCIM_Cubic, leaveTangent=-2.5249, tangentMode=RCTM_User, time=0.1408, value=0.9847}
           {arriveTangent=-0.0405, interpMode=RCIM_Cubic, leaveTangent=-0.0405, tangentMode=RCTM_User, time=1, value=0}
          ]}}}
       CurveIndex=&MakeFloatFromVector(
        Vector=&MakeVectorFromLinearColorRGB(Color=/Niagara/DynamicInputs/TypeConversions/MakeLinearColorFromFloat(Value:NiagaraFloat=$NPC.NPC_Audio_Reactive_Parameters.BeatDeltaTime))
        Channel=::NewEnumerator0
       )
      )
     ))) {enabled=false}
   &ScaleSpriteSize(Scale Factor=&Vector2DFromFloat(Value=&Lerp_Float(
      A=1
      B=5
      Alpha=&FloatFromCurve(
       FloatCurve=DataInterface() {propertyValues {Curve {keys [
           {arriveTangent=9.0725, interpMode=RCIM_Cubic, leaveTangent=9.0725, tangentMode=RCTM_User, time=0, value=0}
           {arriveTangent=-2.5249, interpMode=RCIM_Cubic, leaveTangent=-2.5249, tangentMode=RCTM_User, time=0.1408, value=0.9847}
           {arriveTangent=-0.0405, interpMode=RCIM_Cubic, leaveTangent=-0.0405, tangentMode=RCTM_User, time=1, value=0}
          ]}}}
       CurveIndex=$NPC.NPC_Audio_Reactive_Parameters.Amplitude_Music
      )
     ))) {module=ScaleSpriteSize001}
   &ScaleColor(
    Scale Alpha=&FloatFromCurve(FloatCurve=DataInterface() {propertyValues {Curve {keys [<at -0.0046 1.0113 RCIM_Cubic>,<at 1.0011 0.002 RCIM_Cubic>]}}})
    ScaleA=true
    ScaleRGB=true
   ) {enabled=false}
   &SolveForcesAndVelocity(Clamp Velocity=false, Limit Acceleration=false) {enabled=false}
   &Color(Color=&MakeLinearColorFromVectorAndFloat(
     "Vector (RGB)"=&MakeVectorFromLinearColorRGB(Color=$NPC.NPC_Audio_Reactive_Parameters.Spectrum_Music)
     "Float (Alpha)"=1
    )) {enabled=false}
  ]
  rendererClasses [&NiagaraSpriteRendererProperties]
  renderers [
   {rendererClass=&NiagaraSpriteRendererProperties}
   {rendererClass=&NiagaraSpriteRendererProperties, rendererIndex=1}
  ]
 }
 EQBarHit {
  bEnabled=false
  emitterUpdateScript [
   &EmitterState(
    Inactive Response=::NewEnumerator1
    Life Cycle Mode=::NewEnumerator1
    Loop Behavior=::NewEnumerator1
    Loop Duration Mode=::NewEnumerator1
    UseLoopDelay=false
   )
   &SpawnBurst_Instantaneous(Use Spawn Probability=false, UseLoopCountLimit=false)
  ]
  particleSpawnScript [
   &InitializeParticle(
    Color=<rgba 25 25 25 1>
    Lifetime Max=2.5
    Lifetime Mode=::NewEnumerator1
    Mesh Scale Mode=::NewEnumerator3
    Mesh Uniform Scale=10
    Position Offset:NiagaraPosition=$User.BeamEnd_0
    Sprite Rotation Mode=::NewEnumerator2
    UsePositionOffset=false
    Write Material Random=true
   )
   &InitialMeshOrientation(Mesh Orientation Mode=::NewEnumerator6, Orientation Quaternion:Quat4f=$System.BeamOrientation, Use Rotation Vector=false) {module=InitialMeshOrientation001}
  ]
  particleUpdateScript [
   &ParticleState(Kill Particles When Lifetime Has Elapsed=false, Let Infinitely Lived Particles Die When Emitter Deactivates=true)
   &ScaleMeshSize(Scale Factor=&VectorFromFloat(Value=3))
   /SetVariables(Particles.Position=&Lerp_Vector(
     A=$User.BeamStart_0_0
     B=Linked() {linkedVariable {name=User.BeamEnd_0, type=NiagaraPosition}}
     Alpha=&FloatFromCurve(FloatCurve=DataInterface() {propertyValues {Curve {keys [
          {tangentMode=RCTM_User, time=0, value=0}
          {arriveTangent=1, interpMode=RCIM_Cubic, leaveTangent=1, time=0.1, value=0}
          {tangentMode=RCTM_User, time=1, value=1}
         ]}}})
    )) {isSetParameters=true}
   &ScaleColor(
    Scale Alpha=&FloatFromCurve(FloatCurve=DataInterface() {propertyValues {Curve {keys [{interpMode=RCIM_Cubic, tangentMode=RCTM_User, time=0, value=0}, {interpMode=RCIM_Cubic, tangentMode=RCTM_User, time=1, value=1}]}}})
    ScaleA=true
    ScaleRGB=true
   )
   &SolveForcesAndVelocity(Clamp Velocity=false, Limit Acceleration=false)
  ]
  rendererClasses [&NiagaraSpriteRendererProperties, &NiagaraMeshRendererProperties]
  renderers [
   {rendererClass=&NiagaraSpriteRendererProperties}
   {rendererClass=&NiagaraMeshRendererProperties, rendererIndex=1}
  ]
 }
 EQ_Boxes {
  emitterSpawnScript [&FN_PartitionAttributes(Number of Bone Sets=1, Particles Per Partition=12) {enabled=false}]
  emitterUpdateScript [
   /SetVariables(Emitter.GridLength:NiagaraFloat=&Multiply_Float(A=1500, B=1)) {isSetParameters=true}
   &EmitterState(
    Inactive Response=::NewEnumerator1
    Life Cycle Mode=::NewEnumerator1
    Loop Behavior=::NewEnumerator1
    Loop Duration Mode=::NewEnumerator1
    UseLoopDelay=false
   )
   /Niagara/Modules/Spawn/Location/SpawnParticlesInGrid(X Count=15, Y Count=3, Z Count=1) {enabled=false}
   &SpawnRate(SpawnRate=100, Use Spawn Probability=false) {enabled=false}
   &SpawnBurst_Instantaneous(Spawn Count=700, Spawn Probability=0.85, Use Spawn Probability=false, UseLoopCountLimit=true) {module=SpawnBurst_Instantaneous001}
  ]
  particleSpawnScript [
   &PartitionParticles(Number Of Partitions=1)
   &InitializeParticle(
    Lifetime Max=3
    Lifetime Mode=::NewEnumerator1
    Mesh Scale Max=<0.06 0.05 0.5>
    Mesh Scale Min=<0 0 0>
    Mesh Scale Mode=::NewEnumerator2
    Position Offset=<0 0 -500>
    UsePositionOffset=true
    Write Material Random=true
   ) {module=InitializeParticle001}
   &InitialMeshOrientation(
    Mesh Orientation Mode=::NewEnumerator2
    Rotation=&SelectVectorFromArray(
     Array Sampling Mode=::NewEnumerator0
     Direct Array Index=?
     Vector Selection Array=DataInterface() {propertyValues {FloatData [<0 0 0>, <0.5 0 0>, <0 0.75 0>, <0 0.25 0>]}}
     Random Seed=0
    )
    Use Rotation Vector=true
   )
   /Niagara/Modules/Spawn/Location/V2/GridLocation(
    Align To Grid Cell=true
    Dimensions Definition:ENiagaraGridPlacementType=::NewEnumerator1
    Grid Origin:NiagaraPosition=&AddVectorToPosition(
     Position=/Niagara/DynamicInputs/Helpers/SimulationPosition()
     Vector=&MakeVector(X=10, Y=&Add_Float(A=&Multiply_Float(A=$Emitter.GridLength, B=0.5), B=150), Z=-10)
    )
    Module.Jitter Inside Grid Cells=true
    Module.Override Random Seed=false
    Offset:Vector3f=<0 1 0>
    XYZ Dimensions:Vector3f=&MakeVector(X=3, Y=$Emitter.GridLength, Z=0.1)
   ) {enabled=false}
   &AddVelocity(Velocity=<0 800 0>) {enabled=false}
  ]
  particleUpdateScript [
   &SolveForcesAndVelocity(Clamp Velocity=false, Limit Acceleration=false) {enabled=false}
   &ScaleColor(
    Linear Color Curve=DataInterface() {propertyValues {
      AlphaCurve {keys [<at 0 1>,<at 0.4487 1>,<at 1 1>]}
      BlueCurve {keys [<at 0 0.7379>,<at 0.134 0.5457>,<at 0.2454 0.7157>,<at 0.3458 0.073>,<at 0.4572 0.8714>,<at 0.5613 1>,<at 0.6617 1>,<at 0.7694 0.073>,<at 0.8759 0.9042>]}
      GreenCurve {keys [<at 0 0>,<at 0.134 0>,<at 0.2454 0.0176>,<at 0.3458 1>,<at 0.4572 1>,<at 0.5613 0.8797>,<at 0.6617 0>,<at 0.7694 1>,<at 0.8759 0>]}
      RedCurve {keys [<at 0 0.7682>,<at 0.134 0.1779>,<at 0.2454 0.0782>,<at 0.3458 0>,<at 0.4572 0>,<at 0.5613 0>,<at 0.6617 0.0423>,<at 0.7694 0>,<at 0.8759 1>]}
     }}
    Scale Mode=::NewEnumerator2
   )
   &ParticleState(Kill Particles When Lifetime Has Elapsed=false, Let Infinitely Lived Particles Die When Emitter Deactivates=true)
   &ScaleMeshSize(Scale Factor=&MakeVector(
     X=1
     Y=1
     Z=&RandomRangeFloat(
      Minimum=&FloatFromCurve(
       FloatCurve=DataInterface() {propertyValues {Curve {keys [
           {arriveTangent=0.5032, interpMode=RCIM_Cubic, leaveTangent=0.5032, tangentMode=RCTM_User, time=0, value=1}
           {arriveTangent=0.125, interpMode=RCIM_Cubic, leaveTangent=0.125, time=0.1, value=0.375}
           {arriveTangent=-2.5249, interpMode=RCIM_Cubic, leaveTangent=-2.5249, tangentMode=RCTM_User, time=0.2, value=1.025}
           {arriveTangent=1.5, interpMode=RCIM_Cubic, leaveTangent=1.5, time=0.3, value=0.175}
           <at 0.5 1.475 RCIM_Cubic>
           {arriveTangent=-2, interpMode=RCIM_Cubic, leaveTangent=-2, time=0.7, value=0.175}
           {arriveTangent=0.625, interpMode=RCIM_Cubic, leaveTangent=0.625, time=0.8, value=0.875}
           {arriveTangent=0.625, interpMode=RCIM_Cubic, leaveTangent=0.625, time=0.9, value=0.3}
           {arriveTangent=-0.0405, interpMode=RCIM_Cubic, leaveTangent=-0.0405, tangentMode=RCTM_User, time=1, value=1}
          ]}}}
       Scale Curve=0.5
      )
      Maximum=&FloatFromCurve(
       FloatCurve=DataInterface() {propertyValues {Curve {keys [
           {arriveTangent=0.5032, interpMode=RCIM_Cubic, leaveTangent=0.5032, tangentMode=RCTM_User, time=0, value=1}
           {arriveTangent=0.125, interpMode=RCIM_Cubic, leaveTangent=0.125, time=0.1, value=1.4}
           {arriveTangent=-2.5249, interpMode=RCIM_Cubic, leaveTangent=-2.5249, tangentMode=RCTM_User, time=0.2, value=1.025}
           {arriveTangent=1.375, interpMode=RCIM_Cubic, leaveTangent=1.375, time=0.3, value=0.85}
           {arriveTangent=-0.3333, interpMode=RCIM_Cubic, leaveTangent=-0.3333, time=0.4, value=1.3}
           {arriveTangent=-1.0625, interpMode=RCIM_Cubic, leaveTangent=-1.0625, time=0.6, value=0.75}
           {arriveTangent=1.6667, interpMode=RCIM_Cubic, leaveTangent=1.6667, time=0.8, value=0.875}
           {arriveTangent=0.625, interpMode=RCIM_Cubic, leaveTangent=0.625, time=0.9, value=1.25}
           {arriveTangent=-0.0405, interpMode=RCIM_Cubic, leaveTangent=-0.0405, tangentMode=RCTM_User, time=1, value=1}
          ]}}}
       Scale Curve=0.35
      )
     )
    )) {enabled=false}
   &ScaleColor(
    Curve Index=&ReturnNormalizedExecIndex(Normalized Index Scale=2)
    Linear Color Curve=DataInterface() {propertyValues {
      AlphaCurve {keys [<at 0 1>,<at 0.4487 1>,<at 1 1>]}
      BlueCurve {keys [<at 0 0.7682>,<at 0.134 0.5457>,<at 0.2454 0.7157>,<at 0.3458 0.073>,<at 0.4572 0.8714>,<at 0.5613 1>,<at 0.6617 1>,<at 0.7694 0.073>,<at 0.8759 1>]}
      GreenCurve {keys [<at 0 0.0453>,<at 0.134 0>,<at 0.2454 0.0176>,<at 0.3458 1>,<at 0.4572 1>,<at 0.5613 0.8797>,<at 0.6617 0>,<at 0.7694 1>,<at 0.8759 0>]}
      RedCurve {keys [<at 0 0>,<at 0.134 0.1779>,<at 0.2454 0.0782>,<at 0.3458 0>,<at 0.4572 0>,<at 0.5613 0>,<at 0.6617 0.0423>,<at 0.7694 0>,<at 0.8759 0.0663>]}
     }}
    Scale Mode=::NewEnumerator2
   ) {module=ScaleColor001}
   /SetVariables(Particles.Position=&AddVectorToPosition(
     Position=<0 0 0>
     Vector=&self:ScratchDynamicInput_01(
      Default:Vector3f=&TransformVector(Vector=$User.BeamStart_0_0, Source Space=::World, Destination Space=::Local)
      Vector:Vector3f=&TransformVector(Vector=Linked() {linkedVariable {name=User.BeamEnd_0, type=NiagaraPosition}}, Source Space=::World, Destination Space=::Local)
      Spacing:NiagaraFloat=200
     )
    )) {isSetParameters=true}
   /SetVariables(Particles.Position=&Lerp_Vector(
     A=&Multiply_Vector(A=&TransformVector(Vector=$User.BeamStart_0_0, Source Space=::World, Destination Space=::Local), B=<1 1 1>)
     B=&Multiply_Vector(
      A=&TransformVector(Vector=Linked() {linkedVariable {name=User.BeamEnd_0, type=NiagaraPosition}}, Source Space=::World, Destination Space=::Local)
      B=<1 1 1>
     )
     Alpha=&FloatFromCurve(FloatCurve=DataInterface() {propertyValues {Curve {keys [{tangentMode=RCTM_User, time=0, value=0}, {tangentMode=RCTM_User, time=1, value=1}]}}})
    )) {enabled=false, isSetParameters=true}
   &DynamicMaterialParameters(
    Index 0 Param 1=0.5
    Index 0 Param 4=1
    Module.Index0Param0Write=true
    Module.Index0Param1Write=true
    Module.Index0Param2Write=true
    Module.Index0Param3Write=true
   )
   &ScaleMeshSize(Scale Factor=&MakeVector(
     X=1
     Y=1
     Z=&FloatFromCurve(
      FloatCurve=DataInterface() {propertyValues {Curve {keys [
          {arriveTangent=9.0725, interpMode=RCIM_Cubic, time=0, value=0}
          <at 0.1 1 RCIM_Cubic>
          {arriveTangent=-0.0088, interpMode=RCIM_Cubic, leaveTangent=-0.0088, tangentMode=RCTM_User, time=1, value=0}
         ]}}}
      CurveIndex=&RandomRangeFloat(Minimum=$NPC.NPC_Audio_Reactive_Parameters.Amplitude_Music, Maximum=$NPC.Radiant_Music_NPC.MusicBeatProgress)
     )
    )) {module=ScaleMeshSize005}
   &ScaleMeshSize(Scale Factor=&MakeVector(
     X=1
     Y=1
     Z=&Lerp_Float(
      A=0.1
      B=3
      Alpha=&MakeFloatFromVector(
       Vector=&MakeVectorFromLinearColorRGB(Color=$NPC.NPC_Audio_Reactive_Parameters.Spectrum_MusicHistory2)
       Channel=::NewEnumerator1
      )
     )
    )) {module=ScaleMeshSize001}
   /SetVariables() {enabled=false, isSetParameters=true}
  ]
  rendererClasses [&NiagaraMeshRendererProperties]
  renderers [
   {rendererClass=&NiagaraMeshRendererProperties, rendererIndex=0}
   {rendererClass=&NiagaraMeshRendererProperties, rendererIndex=1}
  ]
 }
 Empty {
  bEnabled=false
  emitterUpdateScript [
   &EmitterState()
   &SpawnBurst_Instantaneous(Use Spawn Probability=false, UseLoopCountLimit=false)
  ]
  particleSpawnScript [&InitializeParticle(
    Color=<rgba 1 0 0.5 1>
    Lifetime=1000
    Sprite Size Mode=::NewEnumerator3
    Uniform Sprite Size=35
    UsePositionOffset=false
    Write Material Random=true
   )]
  particleUpdateScript [
   &ParticleState()
   &self:SampleAudioSpectrum_03(
    New Audio Spectrum=DataInterface() {propertyValues {MaximumFrequency=20000, MinimumFrequency=20, NoiseFloorDb=-60, Resolution=512, Submix=null}}
    Scale Factor=30
   ) {module=SampleAudioSpectrum_03}
   &SolveForcesAndVelocity(Clamp Velocity=false, Limit Acceleration=false)
   /SetVariables(Particles.Position=&Multiply_Vector(
     A=Linked() {linkedVariable {name=Particles.Position, type=NiagaraPosition}}
     B=&Lerp_Vector(
      A=<1 1 1>
      B=<1 1 -1>
      Alpha=&Waveform("Waveform [1]"=::NewEnumerator8, "[1] Frequency"=1, "[1] Amplitude Min/Max"=<0 1>)
     )
    )) {enabled=false, isSetParameters=true}
   &GenerateLocationEvent(Event Send Rate=60, Use Event Probability=false, UseEventDelay=false)
  ]
  rendererClasses [&NiagaraSpriteRendererProperties]
  renderers [{rendererClass=&NiagaraSpriteRendererProperties}]
 }
 Empty001 {
  bEnabled=false
  emitterUpdateScript [
   &EmitterState()
   &SpawnBurst_Instantaneous(Use Spawn Probability=false, UseLoopCountLimit=false)
  ]
  particleSpawnScript [&InitializeParticle(
    Color=<rgba 1 0 0.5 1>
    Lifetime=1000
    Sprite Size Mode=::NewEnumerator3
    Uniform Sprite Size=5
    UsePositionOffset=false
    Write Material Random=true
   )]
  particleUpdateScript [
   &ParticleState()
   &self:SampleAudioSpectrum_02(
    New Audio Spectrum=DataInterface() {propertyValues {MaximumFrequency=20000, MinimumFrequency=20, NoiseFloorDb=-60, Resolution=512, Submix=null}}
    Scale Factor=30
   ) {module=SampleAudioSpectrum_02}
   &SolveForcesAndVelocity(Clamp Velocity=false, Limit Acceleration=false)
   &GenerateLocationEvent(Event Send Rate=60, Use Event Probability=false, UseEventDelay=false)
  ]
  rendererClasses [&NiagaraSpriteRendererProperties]
  renderers [{rendererClass=&NiagaraSpriteRendererProperties}]
 }
 Light {
  emitterUpdateScript [
   /SetVariables(Emitter.JustSphere=1) {isSetParameters=true}
   &EmitterState(
    Inactive Response=::NewEnumerator1
    Life Cycle Mode=::NewEnumerator1
    Loop Delay=&Add_Float(A=0.0258, B=$System.Delay)
    Loop Duration=0.1
    UseLoopDelay=false
   )
   &SpawnRate(Use Spawn Probability=false)
   &SpawnBurst_Instantaneous(Spawn Count=1, Use Spawn Probability=false)
  ]
  particleSpawnScript [&InitializeParticle(
    Color=<rgba 7.3958 7.694 10 0.15>
    Lifetime=0.1
    Mesh Scale Mode=::NewEnumerator3
    Mesh Uniform Scale=25
    Position Offset=<0 0 -50>
    Sprite Size Mode=::NewEnumerator3
    Uniform Sprite Size=150
    UsePositionOffset=false
    Write Material Random=true
   )]
  particleUpdateScript [
   &ParticleState()
   &FN_Light_v3(
    Color=&SelectLinearColorFromArray(
     Array Sampling Mode=::NewEnumerator0
     Direct Array Index=?
     Color Selection Array=DataInterface() {propertyValues {ColorData [<rgba 2 0 1.9278 1>, <rgba 0 1.6664 2 1>, <rgba 0.1987 2 0 1>]}}
     Random Seed=0
    )
    Create Color Attribute=true
    Create Exponent Attribute=true
    Create Position Attribute=true
    Create Radius Attribute=true
    Enable Particle Light Enable=true
    Exponent=6
    Light Enabled=true
    Position:Vector3f=&AddVectorToPosition(Position=$Particles.Position, Vector=<0 0 0>)
    Radius=1000
    Scale Color:NiagaraFloat=&FloatFromCurve(
     FloatCurve=DataInterface() {propertyValues {Curve {keys [
         {arriveTangent=9.0725, interpMode=RCIM_Cubic, leaveTangent=9.0725, tangentMode=RCTM_User, time=0, value=1}
         {arriveTangent=-2.5249, interpMode=RCIM_Cubic, leaveTangent=-2.5249, tangentMode=RCTM_User, time=0.05, value=25}
         {arriveTangent=-0.0405, interpMode=RCIM_Cubic, leaveTangent=-0.0405, tangentMode=RCTM_User, time=0.25, value=1}
        ]}}}
     CurveIndex=$Emitter.Age
     Scale Curve=5
    )
   )
   /SetVariables(Particles.FN_Light_v3.Color:LinearColor=&Multiply_LinearColorByFloat(
     Linear Color=$Particles.FN_Light_v3.Color
     Float=&MakeFloatFromVector(
      Vector=&VectorFromFloat(Value=&Lerp_Float(
        A=0.35
        B=1
        Alpha=&FloatFromCurve(
         FloatCurve=DataInterface() {propertyValues {Curve {keys [
             {arriveTangent=9.0725, interpMode=RCIM_Cubic, time=0, value=0}
             <at 0.1 1 RCIM_Cubic>
             {arriveTangent=-0.0088, interpMode=RCIM_Cubic, leaveTangent=-0.0088, tangentMode=RCTM_User, time=1, value=0}
            ]}}}
         CurveIndex=$NPC.Radiant_Music_NPC.MusicBeatProgress
        )
       ))
      Channel=::NewEnumerator0
     )
    )) {isSetParameters=true}
  ]
  rendererClasses [&NiagaraLightRendererProperties, &NiagaraSpriteRendererProperties]
  renderers [
   {rendererClass=&NiagaraLightRendererProperties, rendererIndex=0}
   {rendererClass=&NiagaraSpriteRendererProperties, rendererIndex=1}
  ]
 }
 MainRibbon001 {
  emitterSpawnScript [&FN_PartitionAttributes(Number of Bone Sets=1, Particles Per Partition=120)]
  emitterUpdateScript [
   &EmitterState()
   &SpawnRate(SpawnGroup=1, SpawnRate=5, Use Spawn Probability=false) {enabled=false}
   &SpawnBurst_Instantaneous(Spawn Count=$Emitter.FN_PartitionAttributes.Spawn Total, Use Spawn Probability=false)
  ]
  particleSpawnScript [
   /Niagara/Modules/Spawn/Initialization/InitializeParticle(
    Lifetime=5
    Module.Write Color=true
    Module.Write Lifetime=true
    Module.Write Mass=true
    Module.Write Material Random=true
    Module.Write Position=true
    Module.Write Scale=false
    Module.Write SpriteRotation=false
    Module.Write SpriteSize=true
    Sprite Size=<30 30>
   ) {enabled=false}
   /Niagara/Modules/Spawn/Initialization/V2/InitializeParticle(Lifetime=5, Mesh Scale Mode=::NewEnumerator3, Mesh Uniform Scale=0.75, UsePositionOffset=false, Write Material Random=true) {module=InitializeParticle001}
   &InitializeRibbon(
    Color=<rgba 5 5 5 1>
    Lifetime=0.7
    Module.Write Color=true
    Module.Write Lifetime=true
    Module.Write Mass=true
    Module.Write Material Random=true
    Module.Write Position=false
    Module.Write RibbonTwist=false
    Module.Write RibbonWidth=true
    Ribbon Width=25
   )
   &FN_Particle_Partition_v5(Enable Ribbon ID=true)
   /SetVariables(Particles.RibbonWidth=&Multiply_Float(
     A=$Particles.RibbonWidth
     B=&MakeFloatFromLinearColor(LinearColor=$NPC.NPC_Audio_Reactive_Parameters.Spectrum_MusicHistory2, Channel:ENiagaraLinearColor_Channels=::NewEnumerator0)
    )) {enabled=false, isSetParameters=true}
   /SetVariables(Particles.Renderer2_Color:LinearColor=<rgba 0.346 3 0 0.4>) {enabled=false, isSetParameters=true}
   /SetVariables(Particles.WobbleOverLength:NiagaraFloat=&FloatFromCurve(
     FloatCurve=DataInterface() {propertyValues {Curve {keys [
         {arriveTangent=5.2404, interpMode=RCIM_Cubic, time=0, value=0.5}
         <at 0.5 1 RCIM_Cubic>
         {interpMode=RCIM_Cubic, leaveTangent=-4.4363, time=1, value=0.5}
        ]}}}
     CurveIndex=$Particles.Partition.NormalizedID
    )) {enabled=false, isSetParameters=true, module=SetVariables_#10}
   /SetVariables(Particles.RibbonLinkOrder:NiagaraFloat=&ReturnNormalizedExecIndex()) {isSetParameters=true}
   /SetVariables() {isSetParameters=true}
  ]
  particleUpdateScript [
   &ParticleState()
   &self:SampleAudioSpectrum(New Audio Spectrum=DataInterface() {propertyValues {MaximumFrequency=20000, MinimumFrequency=20, NoiseFloorDb=-60, Resolution=512, Submix=&Fortnite_AudioAnalysis_Music}}) {enabled=false, module=SampleAudioSpectrum}
   &SystemLocation(MaskSpawnGroup=false) {enabled=false}
   /Game/Effects/Niagara/Modules/Curves/FN_HermiteCubicSpline_v1/FN_HermiteCubicSpline(
    Position Blend:NiagaraFloat=$Particles.Initial.Partition.NormalizedID
    Position End:Vector3f=/Niagara/DynamicInputs/Transforms/TransformPosition(Position=Linked() {linkedVariable {name=User.BeamEnd_0, type=NiagaraPosition}}, Source Space=::World, Destination Space=::Local)
    Position Start:Vector3f=$System.BeamOriginOffset
    Write Particles Position=true
   )
   /SetVariables(Particles.ParticleDistance:NiagaraFloat=&Multiply_Float(
     A=&Abs_Float(Float:NiagaraFloat=&VectorLength(Vector=Linked() {linkedVariable {name=Particles.Position, type=NiagaraPosition}}))
     B=0.2
    )) {isSetParameters=true}
   /SetVariables(Particles.Position=&Add_Vector(
     A=$Output.FN_HermiteCubicSpline.Position
     B=&Multiply_VectorByFloat(
      Vector=&RotateVectorByQuaternion(
       Vector To Rotate=&MakeVector(
        X=0
        Y=&Waveform(
         Global Amplitude Scale=5
         "Waveform [1]"=::NewEnumerator0
         "[1] Phase Scale"=-5
         "[1] Frequency"=2
         "[1] Amplitude Min/Max"=<-1 1>
         "[2] Phase"=$Emitter.Age
         "[3] Phase"=$Emitter.Age
        )
        Z=&Waveform(
         Global Amplitude Scale=5
         "Waveform [1]"=::NewEnumerator1
         "[1] Phase Scale"=-5
         "[1] Frequency"=2
         "[1] Amplitude Min/Max"=<-1 1>
         "[2] Phase"=$Emitter.Age
         "[3] Phase"=$Emitter.Age
        )
       )
       Quaternion=$System.BeamOrientation
      )
      Float=$Particles.WobbleOverLength
     )
    )) {isSetParameters=true, module=SetVariables_#11}
   /SetVariables(Particles.Position:Vector3f=$Output.FN_HermiteCubicSpline.Position) {enabled=false, isSetParameters=true}
   /SetVariables(Particles.Renderer2_Position:Vector3f=&Add_Vector(
     A=$Output.FN_HermiteCubicSpline.Position
     B=&Multiply_VectorByFloat(
      Vector=&RotateVectorByQuaternion(
       Vector To Rotate=&MakeVector(
        X=0
        Y=&Sine(
         Normalized Angle=&Add_Float(A=$Particles.Partition.NormalizedID, B=&Multiply_Float(A=$Emitter.Age, B=3))
         Period=&Divide_Float(A=$System.BeamLength, B=1500)
         Scale=1
         Bias=0
        )
        Z=&Cosine(
         Normalized Angle=&Add_Float(A=$Particles.Partition.NormalizedID, B=&Multiply_Float(A=$Emitter.Age, B=3))
         Period=&Divide_Float(A=$System.BeamLength, B=1500)
         Scale=1
         Bias=0
        )
       )
       Quaternion=$System.BeamOrientation
      )
      Float=$Particles.WobbleOverLength
     )
    )) {isSetParameters=true}
   &ScaleRibbonWidth(Ribbon Width Scale=&FloatFromCurve(
     FloatCurve=DataInterface() {propertyValues {Curve {keys [
         {arriveTangent=4.2519, interpMode=RCIM_Cubic, leaveTangent=4.2519, tangentMode=RCTM_User, time=0, value=0}
         {interpMode=RCIM_Cubic, tangentMode=RCTM_User, time=1, value=1}
        ]}}}
     CurveIndex=$Particles.RibbonLinkOrder
    ))
   &ScaleRibbonWidth(Ribbon Width Scale=&FloatFromCurve(
     FloatCurve=DataInterface() {propertyValues {Curve {keys [<at 0 1>,<at 1 1>]}}}
     CurveIndex=&ReturnNormalizedExecIndex()
     Scale Curve=&Lerp_Float(A=1, B=2, Alpha=$NPC.FortniteNPC.EventAudio_Amplitude)
    )) {enabled=false, module=ScaleRibbonWidth001}
   /SetVariables(Particles.MeshColor:LinearColor=&ColorFromCurve(ColorCurve=DataInterface() {propertyValues {
       AlphaCurve {keys [<at 0 1>,<at 0.4487 1>,<at 1 1>]}
       BlueCurve {keys [<at 0 0.7379>,<at 0.134 0.5457>,<at 0.2454 0.7157>,<at 0.3458 0.073>,<at 0.4572 0.8714>,<at 0.5613 1>,<at 0.6617 1>,<at 0.7694 0.073>,<at 0.8759 0.9042>]}
       GreenCurve {keys [<at 0 0>,<at 0.134 0>,<at 0.2454 0.0176>,<at 0.3458 1>,<at 0.4572 1>,<at 0.5613 0.8797>,<at 0.6617 0>,<at 0.7694 1>,<at 0.8759 0>]}
       RedCurve {keys [<at 0 0.7682>,<at 0.134 0.1779>,<at 0.2454 0.0782>,<at 0.3458 0>,<at 0.4572 0>,<at 0.5613 0>,<at 0.6617 0.0423>,<at 0.7694 0>,<at 0.8759 1>]}
      }})) {isSetParameters=true}
   /SetVariables(Particles.MeshScale:Vector3f=&VectorFromFloat(Value=&FloatFromCurve(FloatCurve=DataInterface() {propertyValues {Curve {keys [
          {arriveTangent=5.1119, interpMode=RCIM_Cubic, leaveTangent=5.1119, tangentMode=RCTM_User, time=0, value=1}
          {arriveTangent=-10.3517, interpMode=RCIM_Cubic, leaveTangent=-10.3517, tangentMode=RCTM_User, time=0.01, value=4}
          {arriveTangent=-10.0989, interpMode=RCIM_Cubic, leaveTangent=-0.8435, tangentMode=RCTM_Break, time=0.15, value=1.25}
          {tangentMode=RCTM_User, time=1, value=1}
         ]}}}))) {enabled=false, isSetParameters=true}
   &ScaleRibbonWidth(Ribbon Width Scale=&FloatFromCurve(
     FloatCurve=DataInterface() {propertyValues {Curve {keys [
         {interpMode=RCIM_Cubic, tangentMode=RCTM_User, time=0, value=1}
         {interpMode=RCIM_Cubic, tangentMode=RCTM_User, time=0.1, value=2.5}
         {interpMode=RCIM_Cubic, tangentMode=RCTM_User, time=0.4, value=1}
        ]}}}
     CurveIndex=&ModuloFloat(
      A=&Add_Float(A=&ReturnNormalizedExecIndex(Normalized Index Scale=-1), B=$Emitter.NormalizedLoopAge)
      B=1
     )
     Scale Curve=&Lerp_Float(
      A=0.8
      B=2
      Alpha=&Multiply_Float(
       A=&MakeFloatFromLinearColor(LinearColor=$NPC.NPC_Audio_Reactive_Parameters.Spectrum_MusicHistory2, Channel:ENiagaraLinearColor_Channels=::NewEnumerator2)
       B=1
      )
     )
    )) {enabled=false, module=ScaleRibbonWidth002}
   &ScaleRibbonWidth(Ribbon Width Scale=&FloatFromCurve(
     FloatCurve=DataInterface() {propertyValues {Curve {keys [
         {interpMode=RCIM_Cubic, tangentMode=RCTM_User, time=0, value=1}
         {interpMode=RCIM_Cubic, tangentMode=RCTM_User, time=0.1, value=2.5}
         {interpMode=RCIM_Cubic, tangentMode=RCTM_User, time=0.4, value=1}
        ]}}}
     CurveIndex=&ModuloFloat(
      A=&Add_Float(A=&ReturnNormalizedExecIndex(Normalized Index Scale=-1), B=$Emitter.NormalizedLoopAge)
      B=&Waveform("Waveform [1]"=::NewEnumerator12, "[1] Frequency"=1, "[1] Amplitude Min/Max"=<0 1>)
     )
     Scale Curve=&Lerp_Float(
      A=0.8
      B=2
      Alpha=&Multiply_Float(
       A=&MakeFloatFromLinearColor(LinearColor=$NPC.NPC_Audio_Reactive_Parameters.Spectrum_MusicHistory2, Channel:ENiagaraLinearColor_Channels=::NewEnumerator2)
       B=1
      )
     )
    )) {enabled=false, module=ScaleRibbonWidth003}
   &ScaleRibbonWidth(Ribbon Width Scale=&Lerp_Float(
     A=&FloatFromCurve(
      FloatCurve=DataInterface() {propertyValues {Curve {keys [
          {arriveTangent=9.0725, interpMode=RCIM_Cubic, leaveTangent=305.071, tangentMode=RCTM_Break, time=0, value=1}
          {arriveTangent=-2.5249, interpMode=RCIM_Cubic, leaveTangent=-2.5249, tangentMode=RCTM_User, time=0.05, value=25}
          {arriveTangent=100, interpMode=RCIM_Cubic, leaveTangent=718.0736, tangentMode=RCTM_Break, time=0.1, value=1}
          <at 0.15 35 RCIM_Cubic>
          {arriveTangent=-166.6667, interpMode=RCIM_Cubic, leaveTangent=127.2465, tangentMode=RCTM_Break, time=0.2, value=1}
          <at 0.3 10 RCIM_Cubic>
          {arriveTangent=400, interpMode=RCIM_Cubic, leaveTangent=400, time=0.35, value=1}
          <at 0.4 50 RCIM_Cubic>
          {arriveTangent=-150, interpMode=RCIM_Cubic, leaveTangent=659.551, tangentMode=RCTM_Break, time=0.45, value=1}
          <at 0.5 35 RCIM_Cubic>
          {arriveTangent=-56.0605, interpMode=RCIM_Cubic, leaveTangent=373.6216, tangentMode=RCTM_Break, time=0.55, value=1}
          {arriveTangent=-35, interpMode=RCIM_Cubic, leaveTangent=-35, time=0.6, value=17.5}
          {arriveTangent=100, interpMode=RCIM_Cubic, leaveTangent=787.5548, tangentMode=RCTM_Break, time=0.65, value=-2.5}
          {arriveTangent=23.3333, interpMode=RCIM_Cubic, leaveTangent=23.3333, time=0.7, value=27.5}
          {arriveTangent=-83.3333, interpMode=RCIM_Cubic, leaveTangent=378.8107, tangentMode=RCTM_Break, time=0.8, value=1}
          <at 0.85 15 RCIM_Cubic>
          {arriveTangent=-100, interpMode=RCIM_Cubic, leaveTangent=-100, time=0.9, value=1}
          {arriveTangent=-0.0405, interpMode=RCIM_Cubic, leaveTangent=-0.0405, tangentMode=RCTM_User, time=1, value=0}
         ]}}}
      CurveIndex=&ModuloFloat(
       A=&Add_Float(A=&ReturnNormalizedExecIndex(Normalized Index Scale=-1), B=$Emitter.NormalizedLoopAge)
       B=1
      )
      Scale Curve=&Lerp_Float(A=0.01, B=0.1, Alpha=$NPC.NPC_Audio_Reactive_Parameters.Amplitude_Music)
     )
     B=&FloatFromCurve(
      FloatCurve=DataInterface() {propertyValues {Curve {keys [
          {interpMode=RCIM_Cubic, tangentMode=RCTM_User, time=0, value=1}
          {arriveTangent=18.0459, interpMode=RCIM_Cubic, leaveTangent=18.0459, tangentMode=RCTM_User, time=0.05, value=2.25}
          {arriveTangent=-58.5938, interpMode=RCIM_Cubic, leaveTangent=-58.5938, tangentMode=RCTM_User, time=0.3, value=1}
         ]}}}
      CurveIndex=&ModuloFloat(
       A=&Add_Float(A=&ReturnNormalizedExecIndex(Normalized Index Scale=-1), B=$Emitter.NormalizedLoopAge)
       B=1
      )
      Scale Curve=2
     )
     Alpha=&Multiply_Float(
      A=$NPC.Radiant_Music_NPC.MusicBeatProgress
      B=&MakeFloatFromVector(
       Vector=&MakeVectorFromLinearColorRGB(Color=$NPC.NPC_Audio_Reactive_Parameters.Spectrum_Music)
       Channel=::NewEnumerator2
      )
     )
    )) {module=ScaleRibbonWidth004}
  ]
  rendererClasses [
   &NiagaraRibbonRendererProperties
   &NiagaraMeshRendererProperties
   &NiagaraSpriteRendererProperties
  ]
  renderers [
   {rendererClass=&NiagaraRibbonRendererProperties}
   {rendererClass=&NiagaraRibbonRendererProperties, rendererIndex=1}
   {rendererClass=&NiagaraMeshRendererProperties, rendererIndex=2}
   {rendererClass=&NiagaraMeshRendererProperties, rendererIndex=3}
   {rendererClass=&NiagaraSpriteRendererProperties, rendererIndex=4}
  ]
 }
 Plasma {
  bEnabled=false
  emitterSpawnScript [&FN_PartitionAttributes(Number of Bone Sets=1, Particles Per Partition=12)]
  emitterUpdateScript [
   &EmitterState()
   &SpawnBurst_Instantaneous(Spawn Count=$Emitter.FN_PartitionAttributes.Spawn Total, Use Spawn Probability=false)
  ]
  particleSpawnScript [
   /Niagara/Modules/Spawn/Initialization/InitializeParticle(
    Color=<rgba 0.0822 1 0 1>
    Lifetime=5
    Module.Write Color=true
    Module.Write Lifetime=true
    Module.Write Mass=true
    Module.Write Material Random=true
    Module.Write Position=true
    Module.Write Scale=false
    Module.Write SpriteRotation=false
    Module.Write SpriteSize=true
    Sprite Size=<30 30>
   ) {enabled=false}
   /Niagara/Modules/Spawn/Initialization/V2/InitializeParticle(Lifetime=5, UsePositionOffset=false, Write Material Random=true) {module=InitializeParticle001}
   &FN_Particle_Partition_v5(Enable Ribbon ID=true)
   &InitializeRibbon(
    Lifetime=0.7
    Module.Write Color=true
    Module.Write Lifetime=true
    Module.Write Mass=true
    Module.Write Material Random=true
    Module.Write Position=false
    Module.Write RibbonTwist=false
    Module.Write RibbonWidth=true
    Ribbon Width=15
   )
   /SetVariables(Particles.WobbleOverLength:NiagaraFloat=&FloatFromCurve(
     FloatCurve=DataInterface() {propertyValues {Curve {keys [
         {arriveTangent=5.2404, interpMode=RCIM_Cubic, leaveTangent=5.2404, tangentMode=RCTM_User, time=0, value=0}
         <at 0.5 1 RCIM_Cubic>
         {arriveTangent=-4.4363, interpMode=RCIM_Cubic, leaveTangent=-4.4363, tangentMode=RCTM_User, time=1, value=0}
        ]}}}
     CurveIndex=$Particles.Partition.NormalizedID
     Scale Curve=2
    )) {isSetParameters=true, module=SetVariables_#10}
   /SetVariables(Particles.WobbleOverLength2:NiagaraFloat=&FloatFromCurve(
     FloatCurve=DataInterface() {propertyValues {Curve {keys [
         {arriveTangent=5.2404, interpMode=RCIM_Cubic, leaveTangent=5.2404, tangentMode=RCTM_User, time=0, value=0}
         <at 0.5 1 RCIM_Cubic>
         {arriveTangent=-4.4363, interpMode=RCIM_Cubic, leaveTangent=-4.4363, tangentMode=RCTM_User, time=1, value=0}
        ]}}}
     CurveIndex=$Particles.Partition.NormalizedID
     Scale Curve=5
    )) {isSetParameters=true, module=SetVariables_#12}
   /SetVariables(Particles.WobbleOverLength3:NiagaraFloat=&FloatFromCurve(
     FloatCurve=DataInterface() {propertyValues {Curve {keys [
         {arriveTangent=5.2404, interpMode=RCIM_Cubic, leaveTangent=5.2404, tangentMode=RCTM_User, time=0, value=0}
         <at 0.5 1 RCIM_Cubic>
         {arriveTangent=-4.4363, interpMode=RCIM_Cubic, leaveTangent=-4.4363, tangentMode=RCTM_User, time=1, value=0}
        ]}}}
     CurveIndex=$Particles.Partition.NormalizedID
     Scale Curve=-5
    )) {isSetParameters=true, module=SetVariables_#13}
  ]
  particleUpdateScript [
   &ParticleState()
   &Color(Color=&ColorFromCurve(ColorCurve=DataInterface() {propertyValues {
       AlphaCurve {keys [<at 0 1>,<at 0.4487 1>,<at 1 1>]}
       BlueCurve {keys [<at 0 0.7379>,<at 0.134 0.5457>,<at 0.2454 0.7157>,<at 0.3458 0.073>,<at 0.4572 0.8714>,<at 0.5613 1>,<at 0.6617 1>,<at 0.7694 0.073>,<at 0.8759 0.9042>]}
       GreenCurve {keys [<at 0 0>,<at 0.134 0>,<at 0.2454 0.0176>,<at 0.3458 1>,<at 0.4572 1>,<at 0.5613 0.8797>,<at 0.6617 0>,<at 0.7694 1>,<at 0.8759 0>]}
       RedCurve {keys [<at 0 0.7682>,<at 0.134 0.1779>,<at 0.2454 0.0782>,<at 0.3458 0>,<at 0.4572 0>,<at 0.5613 0>,<at 0.6617 0.0423>,<at 0.7694 0>,<at 0.8759 1>]}
      }}))
   &SystemLocation(MaskSpawnGroup=false) {enabled=false}
   /SetVariables(Particles.RibbonUVDistance=0.05) {isSetParameters=true, module=SetVariables_#14}
   &FN_HermiteCubicSpline(
    Position Blend:NiagaraFloat=$Particles.Initial.Partition.NormalizedID
    Position End:Vector3f=&TransformPosition(Position=Linked() {linkedVariable {name=User.BeamEnd_0, type=NiagaraPosition}}, Source Space=::World, Destination Space=::Local)
    Position Start:Vector3f=$System.BeamOriginOffset
    Write Particles Position=true
   )
   /SetVariables(Particles.ParticleDistance:NiagaraFloat=&VectorLength(Vector=&Subtract_Vector(
      A=Linked() {linkedVariable {name=Particles.Position, type=NiagaraPosition}}
      B=&TransformVector(Vector=$User.BeamStart_0_0, Source Space=::World, Destination Space=::Local)
     ))) {isSetParameters=true, module=SetVariables_#15}
   /SetVariables(Particles.Position=&Add_Vector(
     A=$Output.FN_HermiteCubicSpline.Position
     B=&Multiply_VectorByFloat(
      Vector=&RotateVectorByQuaternion(
       Vector To Rotate=&MakeVector(
        X=0
        Y=&Sine(
         Normalized Angle=&Add_Float(A=$Particles.Partition.NormalizedID, B=&Multiply_Float(A=$Emitter.Age, B=5))
         Period=&Divide_Float(A=$System.BeamLength, B=1500)
         Scale=15
         Bias=0
        )
        Z=&Cosine(
         Normalized Angle=&Add_Float(A=$Particles.Partition.NormalizedID, B=&Multiply_Float(A=$Emitter.Age, B=3))
         Period=&Divide_Float(A=$System.BeamLength, B=1500)
         Scale=5
         Bias=0
        )
       )
       Quaternion=$System.BeamOrientation
      )
      Float=$Particles.WobbleOverLength
     )
    )) {enabled=false, isSetParameters=true, module=SetVariables_#11}
   /SetVariables(Particles.Position=&Add_Vector(
     A=$Output.FN_HermiteCubicSpline.Position
     B=&Multiply_VectorByFloat(
      Vector=&RotateVectorByQuaternion(
       Vector To Rotate=&MakeVector(
        X=0
        Y=&Waveform(
         Global Input Position=&Multiply_Float(A=$Particles.ParticleDistance, B=0.005)
         Global Amplitude Scale=20
         "Waveform [1]"=::NewEnumerator0
         "[1] Frequency"=1
         "[1] Amplitude Min/Max"=<0 1>
         "[1] Bias"=-0.5
         "[2] Phase"=$Emitter.Age
         "[3] Phase"=$Emitter.Age
        )
        Z=&Waveform(
         Global Input Position=&Multiply_Float(A=$Particles.ParticleDistance, B=0.005)
         Global Amplitude Scale=40
         "Waveform [1]"=::NewEnumerator1
         "[1] Frequency"=1
         "[1] Amplitude Min/Max"=<0 1>
         "[1] Bias"=-0.5
         "[2] Phase"=$Emitter.Age
         "[3] Phase"=$Emitter.Age
        )
       )
       Quaternion=$System.BeamOrientation
      )
      Float=$Particles.WobbleOverLength
     )
    )) {enabled=false, isSetParameters=true, module=SetVariables_#16}
   &ScaleRibbonWidth(Ribbon Width Scale=&RandomRangeFloat(Minimum=0.2, Maximum=0.5)) {module=ScaleRibbonWidth001}
   /SetVariables(Particles.Position=&self:ScratchDynamicInput_05(
     Default:Vector3f=<0 0 0>
     Particle Attribute Reader:NiagaraDataInterfaceParticleRead=DataInterface() {propertyValues {EmitterBinding {bindingMode=Other, emitterName=None}}}
    )) {isSetParameters=true}
  ]
  rendererClasses [&NiagaraRibbonRendererProperties]
  renderers [{rendererClass=&NiagaraRibbonRendererProperties}]
 }
 Plasma003 {
  emitterSpawnScript [&FN_PartitionAttributes(Number of Bone Sets=1, Particles Per Partition=12)]
  emitterUpdateScript [
   &EmitterState()
   &SpawnBurst_Instantaneous(Spawn Count=$Emitter.FN_PartitionAttributes.Spawn Total, Use Spawn Probability=false)
  ]
  particleSpawnScript [
   /Niagara/Modules/Spawn/Initialization/InitializeParticle(
    Color=<rgba 0.0822 1 0 1>
    Lifetime=5
    Module.Write Color=true
    Module.Write Lifetime=true
    Module.Write Mass=true
    Module.Write Material Random=true
    Module.Write Position=true
    Module.Write Scale=false
    Module.Write SpriteRotation=false
    Module.Write SpriteSize=true
    Sprite Size=<30 30>
   ) {enabled=false}
   /Niagara/Modules/Spawn/Initialization/V2/InitializeParticle(Lifetime=5, UsePositionOffset=false, Write Material Random=true) {module=InitializeParticle001}
   &FN_Particle_Partition_v5(Enable Ribbon ID=true)
   &InitializeRibbon(
    Lifetime=0.7
    Module.Write Color=true
    Module.Write Lifetime=true
    Module.Write Mass=true
    Module.Write Material Random=true
    Module.Write Position=false
    Module.Write RibbonTwist=false
    Module.Write RibbonWidth=true
    Ribbon Width=15
   )
   /SetVariables(Particles.WobbleOverLength:NiagaraFloat=&FloatFromCurve(
     FloatCurve=DataInterface() {propertyValues {Curve {keys [
         {arriveTangent=5.2404, interpMode=RCIM_Cubic, leaveTangent=5.2404, tangentMode=RCTM_User, time=0, value=0}
         <at 0.5 1 RCIM_Cubic>
         {arriveTangent=-4.4363, interpMode=RCIM_Cubic, leaveTangent=-4.4363, tangentMode=RCTM_User, time=1, value=0}
        ]}}}
     CurveIndex=$Particles.Partition.NormalizedID
     Scale Curve=2
    )) {isSetParameters=true, module=SetVariables_#10}
   /SetVariables(Particles.WobbleOverLength2:NiagaraFloat=&FloatFromCurve(
     FloatCurve=DataInterface() {propertyValues {Curve {keys [
         {arriveTangent=5.2404, interpMode=RCIM_Cubic, leaveTangent=5.2404, tangentMode=RCTM_User, time=0, value=0}
         <at 0.5 1 RCIM_Cubic>
         {arriveTangent=-4.4363, interpMode=RCIM_Cubic, leaveTangent=-4.4363, tangentMode=RCTM_User, time=1, value=0}
        ]}}}
     CurveIndex=$Particles.Partition.NormalizedID
     Scale Curve=5
    )) {isSetParameters=true, module=SetVariables_#12}
   /SetVariables(Particles.WobbleOverLength3:NiagaraFloat=&FloatFromCurve(
     FloatCurve=DataInterface() {propertyValues {Curve {keys [
         {arriveTangent=5.2404, interpMode=RCIM_Cubic, leaveTangent=5.2404, tangentMode=RCTM_User, time=0, value=0}
         <at 0.5 1 RCIM_Cubic>
         {arriveTangent=-4.4363, interpMode=RCIM_Cubic, leaveTangent=-4.4363, tangentMode=RCTM_User, time=1, value=0}
        ]}}}
     CurveIndex=$Particles.Partition.NormalizedID
     Scale Curve=-5
    )) {isSetParameters=true, module=SetVariables_#13}
  ]
  particleUpdateScript [
   &ParticleState()
   &Color(Color=&ColorFromCurve(ColorCurve=DataInterface() {propertyValues {
       AlphaCurve {keys [<at 0 1>,<at 0.4487 1>,<at 1 1>]}
       BlueCurve {keys [<at 0 0.7379>,<at 0.134 0.9427>,<at 0.2454 1>,<at 0.3458 0.073>,<at 0.4572 0.8714>,<at 0.5613 1>,<at 0.6617 1>,<at 0.7694 0.073>,<at 0.8759 0.9042>]}
       GreenCurve {keys [<at 0 0>,<at 0.134 0.4861>,<at 0.2454 0.5365>,<at 0.3458 1>,<at 0.4572 1>,<at 0.5613 0.8797>,<at 0.6617 0.5677>,<at 0.7694 1>,<at 0.8759 0>]}
       RedCurve {keys [<at 0 0.7682>,<at 0.134 0.6349>,<at 0.2454 0.5767>,<at 0.3458 0>,<at 0.4572 0>,<at 0.5613 0>,<at 0.6617 0.586>,<at 0.7694 0>,<at 0.8759 1>]}
      }}))
   &ScaleColor(
    Scale RGB=&VectorFromFloat(Value=&Lerp_Float(
      A=0.35
      B=5
      Alpha=&FloatFromCurve(
       FloatCurve=DataInterface() {propertyValues {Curve {keys [
           {arriveTangent=9.0725, interpMode=RCIM_Cubic, time=0, value=0}
           <at 0.1 1 RCIM_Cubic>
           {arriveTangent=-0.0088, interpMode=RCIM_Cubic, leaveTangent=-0.0088, tangentMode=RCTM_User, time=1, value=0}
          ]}}}
       CurveIndex=$NPC.Radiant_Music_NPC.MusicBeatProgress
      )
     ))
    ScaleA=true
    ScaleRGB=true
   ) {module=ScaleColor002}
   &SystemLocation(MaskSpawnGroup=false) {enabled=false}
   /SetVariables(Particles.RibbonUVDistance=0.05) {isSetParameters=true, module=SetVariables_#14}
   &FN_HermiteCubicSpline(
    Position Blend:NiagaraFloat=$Particles.Initial.Partition.NormalizedID
    Position End:Vector3f=&TransformPosition(Position=Linked() {linkedVariable {name=User.BeamEnd_0, type=NiagaraPosition}}, Source Space=::World, Destination Space=::Local)
    Position Start:Vector3f=$System.BeamOriginOffset
    Write Particles Position=true
   )
   /SetVariables(Particles.ParticleDistance:NiagaraFloat=&VectorLength(Vector=&Subtract_Vector(
      A=Linked() {linkedVariable {name=Particles.Position, type=NiagaraPosition}}
      B=&TransformVector(Vector=$User.BeamStart_0_0, Source Space=::World, Destination Space=::Local)
     ))) {isSetParameters=true, module=SetVariables_#15}
   /SetVariables(Particles.Position=&Add_Vector(
     A=$Output.FN_HermiteCubicSpline.Position
     B=&Multiply_VectorByFloat(
      Vector=&RotateVectorByQuaternion(
       Vector To Rotate=&MakeVector(
        X=0
        Y=&Sine(
         Normalized Angle=&Add_Float(A=$Particles.Partition.NormalizedID, B=&Multiply_Float(A=$Emitter.Age, B=5))
         Period=&Divide_Float(A=$System.BeamLength, B=1500)
         Scale=15
         Bias=0
        )
        Z=&Cosine(
         Normalized Angle=&Add_Float(A=$Particles.Partition.NormalizedID, B=&Multiply_Float(A=$Emitter.Age, B=3))
         Period=&Divide_Float(A=$System.BeamLength, B=1500)
         Scale=5
         Bias=0
        )
       )
       Quaternion=$System.BeamOrientation
      )
      Float=$Particles.WobbleOverLength
     )
    )) {isSetParameters=true, module=SetVariables_#11}
   /SetVariables(Particles.Position=&Add_Vector(
     A=$Output.FN_HermiteCubicSpline.Position
     B=&Multiply_VectorByFloat(
      Vector=&RotateVectorByQuaternion(
       Vector To Rotate=&MakeVector(
        X=0
        Y=&Waveform(
         Global Input Position=&Multiply_Float(A=$Particles.ParticleDistance, B=0.005)
         Global Amplitude Scale=20
         "Waveform [1]"=::NewEnumerator0
         "[1] Frequency"=1
         "[1] Amplitude Min/Max"=<0 1>
         "[1] Bias"=-0.5
         "[2] Phase"=$Emitter.Age
         "[3] Phase"=$Emitter.Age
        )
        Z=&Waveform(
         Global Input Position=&Multiply_Float(A=$Particles.ParticleDistance, B=0.005)
         Global Amplitude Scale=40
         "Waveform [1]"=::NewEnumerator1
         "[1] Frequency"=1
         "[1] Amplitude Min/Max"=<0 1>
         "[1] Bias"=-0.5
         "[2] Phase"=$Emitter.Age
         "[3] Phase"=$Emitter.Age
        )
       )
       Quaternion=$System.BeamOrientation
      )
      Float=$Particles.WobbleOverLength
     )
    )) {isSetParameters=true, module=SetVariables_#16}
   &ScaleRibbonWidth(Ribbon Width Scale=&Sine(Normalized Angle=&Add_Float(A=$Particles.Partition.NormalizedID, B=$Emitter.Age), Period=3, Scale=0.15, Bias=1)) {enabled=false}
   &ScaleRibbonWidth(Ribbon Width Scale=&RandomRangeFloat(Minimum=0.2, Maximum=0.5)) {module=ScaleRibbonWidth001}
   &FN_RibbonMaterial(
    Color=<rgba 1.2671 0.5205 5 1>
    Macro Distortion Amount=0.1
    Macro Pan Speed=1
    Macro Scale U=10
    Opacity Pan U Speed=3
    Opacity Scale U=250
    UV Basis:NiagaraFloat=$Particles.Partition.NormalizedID
   ) {enabled=false}
   /SetVariables(Particles.RibbonUVDistance=10) {enabled=false, isSetParameters=true}
   /SetVariables(Particles.Pos2:NiagaraPosition=&Add_Vector(
     A=$Output.FN_HermiteCubicSpline.Position
     B=&Multiply_VectorByFloat(
      Vector=&RotateVectorByQuaternion(
       Vector To Rotate=&MakeVector(
        X=0
        Y=&Cosine(
         Normalized Angle=&Add_Float(A=$Particles.Partition.NormalizedID, B=&Multiply_Float(A=$Emitter.Age, B=6))
         Period=&Divide_Float(A=$System.BeamLength, B=1500)
         Scale=5
         Bias=0
        )
        Z=&Sine(
         Normalized Angle=&Add_Float(A=$Particles.Partition.NormalizedID, B=&Multiply_Float(A=$Emitter.Age, B=5))
         Period=&Divide_Float(A=$System.BeamLength, B=1500)
         Scale=15
         Bias=0
        )
       )
       Quaternion=$System.BeamOrientation
      )
      Float=$Particles.WobbleOverLength2
     )
    )) {isSetParameters=true}
   /SetVariables(Particles.Pos2:NiagaraPosition=&Add_Vector(
     A=$Output.FN_HermiteCubicSpline.Position
     B=&Multiply_VectorByFloat(
      Vector=&RotateVectorByQuaternion(
       Vector To Rotate=&MakeVector(
        X=0
        Y=&Waveform(
         Global Input Position=&Multiply_Float(A=$Particles.ParticleDistance, B=0.005)
         Global Amplitude Scale=40
         "Waveform [1]"=::NewEnumerator1
         "[1] Frequency"=1
         "[1] Amplitude Min/Max"=<0 1>
         "[1] Bias"=-0.5
         "[2] Phase"=$Emitter.Age
         "[3] Phase"=$Emitter.Age
        )
        Z=&Waveform(
         Global Input Position=&Multiply_Float(A=$Particles.ParticleDistance, B=0.005)
         Global Amplitude Scale=20
         "Waveform [1]"=::NewEnumerator0
         "[1] Frequency"=1
         "[1] Amplitude Min/Max"=<0 1>
         "[1] Bias"=-0.5
         "[2] Phase"=$Emitter.Age
         "[3] Phase"=$Emitter.Age
        )
       )
       Quaternion=$System.BeamOrientation
      )
      Float=$Particles.WobbleOverLength2
     )
    )) {isSetParameters=true}
   /SetVariables(Particles.Pos3:NiagaraPosition=&Add_Vector(
     A=$Output.FN_HermiteCubicSpline.Position
     B=&Multiply_VectorByFloat(
      Vector=&RotateVectorByQuaternion(
       Vector To Rotate=&MakeVector(
        X=0
        Y=&Sine(
         Normalized Angle=&Add_Float(A=$Particles.Partition.NormalizedID, B=&Multiply_Float(A=$Emitter.Age, B=2))
         Period=&Divide_Float(A=$System.BeamLength, B=1650)
         Scale=11
         Bias=2
        )
        Z=&Cosine(
         Normalized Angle=&Add_Float(A=$Particles.Partition.NormalizedID, B=&Multiply_Float(A=$Emitter.Age, B=5))
         Period=&Divide_Float(A=$System.BeamLength, B=700)
         Scale=9
         Bias=0.25
        )
       )
       Quaternion=$System.BeamOrientation
      )
      Float=$Particles.WobbleOverLength3
     )
    )) {isSetParameters=true}
   /SetVariables(Particles.Pos3:NiagaraPosition=&Add_Vector(
     A=$Output.FN_HermiteCubicSpline.Position
     B=&Multiply_VectorByFloat(
      Vector=&RotateVectorByQuaternion(
       Vector To Rotate=&MakeVector(
        X=0
        Y=&Waveform(
         Global Input Position=&Multiply_Float(A=$Particles.ParticleDistance, B=0.005)
         Global Amplitude Scale=20
         "Waveform [1]"=::NewEnumerator0
         "[1] Frequency"=1
         "[1] Amplitude Min/Max"=<0 1>
         "[1] Bias"=-0.5
         "[2] Phase"=$Emitter.Age
         "[3] Phase"=$Emitter.Age
        )
        Z=&Waveform(
         Global Input Position=&Multiply_Float(A=$Particles.ParticleDistance, B=0.005)
         Global Amplitude Scale=40
         "Waveform [1]"=::NewEnumerator1
         "[1] Frequency"=1
         "[1] Amplitude Min/Max"=<0 1>
         "[1] Bias"=-0.5
         "[2] Phase"=$Emitter.Age
         "[3] Phase"=$Emitter.Age
        )
       )
       Quaternion=$System.BeamOrientation
      )
      Float=$Particles.WobbleOverLength3
     )
    )) {isSetParameters=true}
   &ScaleRibbonWidth(Ribbon Width Scale=&Lerp_Float(
     A=1
     B=6
     Alpha=&FloatFromCurve(
      FloatCurve=DataInterface() {propertyValues {Curve {keys [
          {arriveTangent=6.3132, interpMode=RCIM_Cubic, leaveTangent=6.3132, tangentMode=RCTM_User, time=0, value=0}
          {arriveTangent=-2.5249, interpMode=RCIM_Cubic, leaveTangent=-2.5249, tangentMode=RCTM_User, time=0.1408, value=1}
          {arriveTangent=-0.0405, interpMode=RCIM_Cubic, leaveTangent=-0.0405, tangentMode=RCTM_User, time=0.2, value=0}
         ]}}}
      CurveIndex=$NPC.Radiant_Music_NPC.MusicBeatProgress
      Scale Curve=&MakeFloatFromVector(
       Vector=&MakeVectorFromLinearColorRGB(Color=&MakeLinearColorFromFloat(Value:NiagaraFloat=&MakeFloatFromLinearColor(LinearColor=$NPC.NPC_Audio_Reactive_Parameters.Spectrum_Music, Channel:ENiagaraLinearColor_Channels=::NewEnumerator0)))
       Channel=::NewEnumerator0
      )
     )
    )) {module=ScaleRibbonWidth002}
  ]
  rendererClasses [&NiagaraRibbonRendererProperties, &NiagaraMeshRendererProperties]
  renderers [
   {rendererClass=&NiagaraRibbonRendererProperties}
   {rendererClass=&NiagaraMeshRendererProperties, rendererIndex=1}
   {rendererClass=&NiagaraRibbonRendererProperties, rendererIndex=2}
   {rendererClass=&NiagaraRibbonRendererProperties, rendererIndex=3}
  ]
 }
 Plot_Spring {
  bEnabled=false
  emitterUpdateScript [&EmitterState()]
  particleSpawnScript [
   &InitializeParticle(
    Color=<rgba 0 1 0.5251 1>
    Lifetime=12
    Sprite Size Mode=::NewEnumerator3
    Uniform Sprite Size=5
    UsePositionOffset=false
    Write Material Random=true
   ) {enabled=false}
   &InitializeRibbon(
    Color=<rgba 0 1 0.0734 1>
    Lifetime=6
    Position Offset:Vector3f=$User.BeamStart_0_0
    Ribbon Width=2.5
    UsePositionOffset=true
    Write Material Random=true
   )
   &AddVelocity(Velocity=&Multiply_Vector(
     A=&SubtractVectorFromPosition(
      Position=Linked() {linkedVariable {name=User.BeamStart_0_0, type=Vector3f}}
      Vector=Linked() {linkedVariable {name=User.BeamEnd_0, type=NiagaraPosition}}
     )
     B=<1 250 1>
    ))
  ]
  particleUpdateScript [
   &ParticleState()
   &Color(Color=&ColorFromCurve(ColorCurve=DataInterface() {propertyValues {
       AlphaCurve {keys [<at 0 1>,<at 0.4487 1>,<at 1 1>]}
       BlueCurve {keys [<at 0 0.7379>,<at 0.134 0.9427>,<at 0.2454 1>,<at 0.3458 0.073>,<at 0.4572 0.8714>,<at 0.5613 1>,<at 0.6617 1>,<at 0.7694 0.073>,<at 0.8759 0.9042>]}
       GreenCurve {keys [<at 0 0>,<at 0.134 0.4861>,<at 0.2454 0.5365>,<at 0.3458 1>,<at 0.4572 1>,<at 0.5613 0.8797>,<at 0.6617 0.5677>,<at 0.7694 1>,<at 0.8759 0>]}
       RedCurve {keys [<at 0 0.7682>,<at 0.134 0.6349>,<at 0.2454 0.5767>,<at 0.3458 0>,<at 0.4572 0>,<at 0.5613 0>,<at 0.6617 0.586>,<at 0.7694 0>,<at 0.8759 1>]}
      }}))
   &SolveForcesAndVelocity(Clamp Velocity=false, Limit Acceleration=false)
   &ScaleColor(Scale RGB=&VectorFromFloat(Value=10), ScaleA=false, ScaleRGB=true)
  ]
  rendererClasses [&NiagaraRibbonRendererProperties]
  renderers [
   {rendererClass=&NiagaraRibbonRendererProperties}
   {rendererClass=&NiagaraRibbonRendererProperties, rendererIndex=1}
  ]
 }
 RingsMeshesSpacedOut {
  emitterSpawnScript [&FN_PartitionAttributes(Number of Bone Sets=1, Particles Per Partition=12)]
  emitterUpdateScript [
   &EmitterState(Life Cycle Mode=::NewEnumerator1, Loop Duration=0.5, UseLoopDelay=false)
   &SpawnPerUnit(
    Module.Use Max Movement Threshold=true
    Module.Use Movement Tolerance=true
    Movement Tolerance=0
    Spawn Probability=1
    Spawn Spacing=30
    Use Spawn Probability=true
   ) {enabled=false}
   &FN_SpawnByEngineDistance(Multiplier=5, Step=20) {enabled=false}
   &SpawnBurst_Instantaneous(Spawn Count=10, Spawn Probability=0.5, Use Spawn Probability=true, UseLoopCountLimit=true) {enabled=false, module=SpawnBurst_Instantaneous001}
   &SpawnBurst_Instantaneous(Spawn Count=$Emitter.FN_PartitionAttributes.Spawn Total, Use Spawn Probability=false) {enabled=false}
   &SpawnRate(Spawn Probability=0.5, SpawnRate=5, Use Spawn Probability=false)
  ]
  particleSpawnScript [
   &PartitionParticles()
   &InitializeParticle(
    Lifetime Max=4
    Lifetime Mode=::NewEnumerator1
    Mesh Scale Max=<1.65 1.65 2.25>
    Mesh Scale Min=<1.65 1.65 1.5>
    Mesh Scale Mode=::NewEnumerator2
    Sprite Size Mode=::NewEnumerator3
    UsePositionOffset=true
    Write Material Random=true
   ) {module=InitializeParticle001}
   &FN_Particle_Partition_v5(Enable Ribbon ID=true)
   &InitializeRibbon(
    Color=<rgba 1 0 0.6104 0.15>
    Lifetime=0.7
    Module.Write Color=true
    Module.Write Lifetime=true
    Module.Write Mass=true
    Module.Write Material Random=true
    Module.Write Position=false
    Module.Write RibbonTwist=false
    Module.Write RibbonWidth=true
    Ribbon Width=14
   ) {enabled=false}
   &InitialMeshOrientation(
    Orientation Axis=<0 0 1>
    Orientation Coordinate Space=::World
    Orientation Vector=&RotateVectorByQuaternion(Vector To Rotate=<1 0 0>, Quaternion=$System.BeamOrientation)
    Use Orientation Vector=true
    Use Rotation Vector=false
   )
   &InitialMeshOrientation(Mesh Orientation Mode=::NewEnumerator6, Orientation Quaternion:Quat4f=$System.BeamOrientation, Use Rotation Vector=false) {enabled=false, module=InitialMeshOrientation001}
   /SetVariables(Particles.WobbleOverLength:NiagaraFloat=&FloatFromCurve(
     FloatCurve=DataInterface() {propertyValues {Curve {keys [
         {arriveTangent=5.2404, interpMode=RCIM_Cubic, leaveTangent=5.2404, tangentMode=RCTM_User, time=0, value=0}
         <at 0.5 1 RCIM_Cubic>
         {arriveTangent=-4.4363, interpMode=RCIM_Cubic, leaveTangent=-4.4363, tangentMode=RCTM_User, time=1, value=0}
        ]}}}
     CurveIndex=$Particles.Partition.NormalizedID
    )) {enabled=false, isSetParameters=true, module=SetVariables_#5}
   &AddVelocity(Velocity=<0.1 0 0>) {enabled=false}
  ]
  particleUpdateScript [
   &ParticleState()
   &ScaleMeshSize(Scale Factor=&VectorFromFloat(Value=&Lerp_Float(
      A=&FloatFromCurve(
       FloatCurve=DataInterface() {propertyValues {Curve {keys [
           {arriveTangent=9.0725, interpMode=RCIM_Cubic, leaveTangent=9.0725, tangentMode=RCTM_User, time=0, value=0}
           {arriveTangent=-2.5249, interpMode=RCIM_Cubic, leaveTangent=-2.5249, tangentMode=RCTM_User, time=0.1408, value=0.9847}
           {arriveTangent=-0.0405, interpMode=RCIM_Cubic, leaveTangent=-0.0405, tangentMode=RCTM_User, time=1, value=0}
          ]}}}
       Scale Curve=&Lerp_Float(
        A=0.5
        B=1
        Alpha=&MakeFloatFromVector(
         Vector=&MakeVectorFromLinearColorRGB(Color=$NPC.NPC_Audio_Reactive_Parameters.Spectrum_Music)
         Channel=::NewEnumerator2
        )
       )
      )
      B=&FloatFromCurve(
       FloatCurve=DataInterface() {propertyValues {Curve {keys [
           {arriveTangent=9.0725, interpMode=RCIM_Cubic, leaveTangent=9.0725, tangentMode=RCTM_User, time=0, value=0}
           {arriveTangent=-2.5249, interpMode=RCIM_Cubic, leaveTangent=-2.5249, tangentMode=RCTM_User, time=0.1408, value=0.9847}
           {arriveTangent=-0.0405, interpMode=RCIM_Cubic, leaveTangent=-0.0405, tangentMode=RCTM_User, time=1, value=0}
          ]}}}
       Scale Curve=&Lerp_Float(
        A=0.5
        B=1
        Alpha=&MakeFloatFromVector(
         Vector=&MakeVectorFromLinearColorRGB(Color=$NPC.NPC_Audio_Reactive_Parameters.Spectrum_Music)
         Channel=::NewEnumerator2
        )
       )
      )
      Alpha=$Particles.PartitionParticles.NormalizedPartition
     ))) {enabled=false, module=ScaleMeshSize004}
   &ScaleMeshSize(Scale Factor=&VectorFromFloat(Value=&FloatFromCurve(
      FloatCurve=DataInterface() {propertyValues {Curve {keys [
          {interpMode=RCIM_Cubic, tangentMode=RCTM_User, time=0, value=0}
          {arriveTangent=1.25, leaveTangent=1.25, time=0.2, value=1}
          {arriveTangent=-1.25, interpMode=RCIM_Cubic, leaveTangent=-1.25, time=0.8, value=1}
          {tangentMode=RCTM_User, time=1, value=0}
         ]}}}
      Scale Curve=1.5
     ))) {module=ScaleMeshSize003}
   &ScaleMeshSize(Scale Factor=&VectorFromFloat(Value=&FloatFromCurve(
      FloatCurve=DataInterface() {propertyValues {Curve {keys [
          {interpMode=RCIM_Cubic, tangentMode=RCTM_User, time=0, value=0.55}
          {arriveTangent=1.0101, time=0.1, value=1.05}
          {interpMode=RCIM_Cubic, tangentMode=RCTM_User, time=0.9, value=1}
          {tangentMode=RCTM_User, time=1, value=0}
         ]}}}
      CurveIndex=&ReturnNormalizedExecIndex()
     ))) {enabled=false, module=ScaleMeshSize002}
   &ScaleMeshSize(Scale Factor=&VectorFromFloat(Value=&RandomRangeFloat(
      Minimum=&Lerp_Float(
       A=0.4
       B=4
       Alpha=&FloatFromCurve(
        FloatCurve=DataInterface() {propertyValues {Curve {keys [
            {arriveTangent=9.0725, interpMode=RCIM_Cubic, time=0, value=0}
            <at 0.1 1 RCIM_Cubic>
            {arriveTangent=-0.0088, interpMode=RCIM_Cubic, leaveTangent=-0.0088, tangentMode=RCTM_User, time=1, value=0}
           ]}}}
        CurveIndex=$NPC.Radiant_Music_NPC.MusicBeatProgress
       )
      )
      Maximum=$NPC.NPC_Audio_Reactive_Parameters.Amplitude_Music
     ))) {module=ScaleMeshSize005}
   &ScaleMeshSize(Scale Factor=&VectorFromFloat(Value=&Add_Float(
      A=&Sine(Normalized Angle=&Subtract_Float(A=$Particles.NormalizedAge, B=$Emitter.Age), Period=0.5, Scale=0.25, Bias=0)
      B=1
     ))) {enabled=false}
   &ScaleMeshSize(Scale Factor=&MakeVector(
     X=&Add_Float(
      A=&Sine(Normalized Angle=&Subtract_Float(A=$Particles.NormalizedAge, B=$Emitter.Age), Period=0.5, Scale=0.25, Bias=0)
      B=1
     )
     Y=&Add_Float(
      A=&Sine(Normalized Angle=&Subtract_Float(A=$Particles.NormalizedAge, B=$Emitter.Age), Period=0.5, Scale=0.25, Bias=0)
      B=1
     )
     Z=0.75
    )) {enabled=false, module=ScaleMeshSize001}
   /SetVariables(Particles.Position=&Lerp_Vector(
     A=&TransformVector(Vector=$User.BeamStart_0_0, Source Space=::World, Destination Space=::World)
     B=&TransformVector(Vector=Linked() {linkedVariable {name=User.BeamEnd_0, type=NiagaraPosition}}, Source Space=::World, Destination Space=::World)
     Alpha=&FloatFromCurve(FloatCurve=DataInterface() {propertyValues {Curve {keys [{tangentMode=RCTM_User, time=0, value=0}, {tangentMode=RCTM_User, time=1, value=1}]}}})
    )) {enabled=false, isSetParameters=true, module=SetVariables_#7}
   /SetVariables(Particles.Position=&Lerp_Vector(
     A=$User.BeamStart_0_0
     B=Linked() {linkedVariable {name=User.BeamEnd_0, type=NiagaraPosition}}
     Alpha=&FloatFromCurve(FloatCurve=DataInterface() {propertyValues {Curve {keys [{tangentMode=RCTM_User, time=0, value=0}, {tangentMode=RCTM_User, time=1, value=1}]}}})
    )) {isSetParameters=true, module=SetVariables_#6}
   &KillParticles(Kill Particles=&self:ScratchDynamicInput(BeamEnd:Vector3f=Linked() {linkedVariable {name=User.BeamEnd_0, type=NiagaraPosition}})) {enabled=false}
   &ScaleColor(
    Curve Index=&ReturnNormalizedExecIndex(Normalized Index Scale=2)
    Linear Color Curve=DataInterface() {propertyValues {
      AlphaCurve {keys [<at 0 1>,<at 0.4487 1>,<at 1 1>]}
      BlueCurve {keys [<at 0 0.7379>,<at 0.134 0.5457>,<at 0.2454 0.7157>,<at 0.3458 0.073>,<at 0.4572 0.8714>,<at 0.5613 1>,<at 0.6617 1>,<at 0.7694 0.073>,<at 0.8759 0.9042>]}
      GreenCurve {keys [<at 0 0>,<at 0.134 0>,<at 0.2454 0.0176>,<at 0.3458 1>,<at 0.4572 1>,<at 0.5613 0.8797>,<at 0.6617 0>,<at 0.7694 1>,<at 0.8759 0>]}
      RedCurve {keys [<at 0 0.7682>,<at 0.134 0.1779>,<at 0.2454 0.0782>,<at 0.3458 0>,<at 0.4572 0>,<at 0.5613 0>,<at 0.6617 0.0423>,<at 0.7694 0>,<at 0.8759 1>]}
     }}
    Scale Mode=::NewEnumerator2
   ) {module=ScaleColor001}
   &ScaleColor(
    Scale RGB=&VectorFromFloat(Value=&Lerp_Float(
      A=0.35
      B=3.75
      Alpha=&FloatFromCurve(
       FloatCurve=DataInterface() {propertyValues {Curve {keys [
           {arriveTangent=9.0725, interpMode=RCIM_Cubic, time=0, value=0}
           <at 0.1 1 RCIM_Cubic>
           {arriveTangent=-0.0088, interpMode=RCIM_Cubic, leaveTangent=-0.0088, tangentMode=RCTM_User, time=1, value=0}
          ]}}}
       CurveIndex=$NPC.Radiant_Music_NPC.MusicBeatProgress
      )
     ))
    ScaleA=true
    ScaleRGB=true
   ) {module=ScaleColor002}
   &DynamicMaterialParameters(
    Index 0 Param 1=&MakeFloatFromVector(Vector=$Particles.Scale, Channel=::NewEnumerator0)
    Index 0 Param 2=&MakeCustomFloatFromBool(Bool=true, True Float=35, False Float=20)
    Index 0 Param 3=&MakeCustomFloatFromBool(
     Bool=true
     True Float=$System.AudioReactValue
     False Float=&FloatFromCurve(
      FloatCurve=DataInterface() {propertyValues {Curve {keys [
          {interpMode=RCIM_Cubic, leaveTangent=4.8905, tangentMode=RCTM_Break, time=0, value=0.15}
          <at 0.2 0.9>
          {interpMode=RCIM_Cubic, leaveTangent=-1.0585, tangentMode=RCTM_Break, time=0.3, value=0.9}
          {arriveTangent=-0.6609, tangentMode=RCTM_User, time=1, value=0.15}
         ]}}}
      CurveIndex=$System.NormalizedLoopAge
     )
    )
    Index 0 Param 4=&MakeCustomFloatFromBool(Bool=true, True Float=1, False Float=0)
    Module.Index0Param0Write=true
    Module.Index0Param1Write=true
    Module.Index0Param2Write=true
    Module.Index0Param3Write=true
   ) {enabled=false}
   &FN_Light_v3(
    Color=&SelectLinearColorFromArray(
     Array Sampling Mode=::NewEnumerator0
     Direct Array Index=?
     Color Selection Array=DataInterface() {propertyValues {ColorData [<rgba 10 0 9.6388 1>, <rgba 0 8.332 10 1>, <rgba 0.9935 10 0 1>]}}
     Random Seed=0
    )
    Create Color Attribute=true
    Create Exponent Attribute=true
    Create Position Attribute=true
    Create Radius Attribute=true
    Enable Particle Light Enable=true
    Exponent=8
    Light Enabled=true
    Position:Vector3f=&AddVectorToPosition(Position=$Particles.Position, Vector=<0 0 0>)
    Radius=1000
    Scale Color=0.1
   ) {enabled=false}
  ]
  rendererClasses [&NiagaraMeshRendererProperties]
  renderers [
   {rendererClass=&NiagaraMeshRendererProperties, rendererIndex=0}
   {rendererClass=&NiagaraMeshRendererProperties, rendererIndex=1}
  ]
 }
 SparkLengthCollisionTest {
  bEnabled=false
  emitterUpdateScript [
   &EmitterState(Life Cycle Mode=::NewEnumerator1, Loop Duration=0.1, UseLoopDelay=false)
   &SpawnBurst_Instantaneous(Spawn Count=25, Use Spawn Probability=false, UseLoopCountLimit=false) {enabled=false}
   &SpawnRate(Spawn Probability=0.85, SpawnRate=50, Use Spawn Probability=true)
  ]
  particleSpawnScript [
   /SetVariables(Particles.ScaleOVerLength:NiagaraFloat=&FloatFromCurve(
     FloatCurve=DataInterface() {propertyValues {Curve {keys [
         {arriveTangent=0.1131, interpMode=RCIM_Cubic, leaveTangent=0.1131, tangentMode=RCTM_User, time=0, value=0}
         <at 0.5 1 RCIM_Cubic>
         {arriveTangent=-0.0182, interpMode=RCIM_Cubic, leaveTangent=-0.0182, tangentMode=RCTM_User, time=1, value=0}
        ]}}}
     CurveIndex=&ReturnNormalizedExecIndex()
    )) {enabled=false, isSetParameters=true, module=SetVariables_#8}
   &InitializeParticle(
    Color=<rgba 15 15 15 1>
    Lifetime Max=5
    Lifetime Min=3
    Lifetime Mode=::NewEnumerator1
    Mesh Scale Max=<0.05 0.05 0.2>
    Mesh Scale Min=<0.05 0.05 0.05>
    Mesh Scale Mode=::NewEnumerator2
    Position Offset=&RandomRangeVector(Minimum=<0 0 0>, Maximum=<0 3 0>)
    Sprite Size Mode=::NewEnumerator1
    Uniform Sprite Size Max=7
    UsePositionOffset=true
    Write Material Random=true
   )
   &SphereLocation(
    Hemisphere X=true
    MaskSpawnGroup=false
    Override Seed=false
    Sphere Radius=10
    Surface Only=false
    UseNonUniformScale=false
    UseOffset=false
    UseSphereOrientation=false
   ) {module=SphereLocation001}
   /SetVariables(Particles.Position=&AddVectorToPosition(
     Position=<0 0 0>
     Vector=&self:ScratchDynamicInput_01(
      Default:Vector3f=&TransformVector(Vector=$User.BeamStart_0_0, Source Space=::World, Destination Space=::Local)
      Vector:Vector3f=&TransformVector(Vector=Linked() {linkedVariable {name=User.BeamEnd_0, type=NiagaraPosition}}, Source Space=::World, Destination Space=::Local)
      Spacing:NiagaraFloat=35
     )
    )) {isSetParameters=true, module=SetVariables_#9}
   &AddVelocity(Velocity=&Multiply_VectorByFloat(
     Vector=&NormalizeVector(
      Vector To Normalize=&Add_Vector(
       A=<0.6 0 0>
       B=&TransformVector(
        Vector=&NormalizeVector(
         Vector To Normalize=/Niagara/DynamicInputs/Vectors/NegateVector(Vector:Vector3f=$Engine.Owner.Velocity)
         Safe Normalize=true
         Safe Fallback Threshold=0
         Safe Fallback Vector=<0 0 0.001>
        )
        Source Space=::World
        Destination Space=::Local
       )
      )
      Safe Normalize=true
      Safe Fallback Threshold=0
      Safe Fallback Vector=<0 0 0.001>
     )
     Float=&RandomRangeFloat(Minimum=50, Maximum=300)
    )) {enabled=false}
   &AddVelocity(
    Coordinate Space=::Simulation
    Velocity=&Multiply_VectorByFloat(
     Vector=$Output.SphereLocation001.SphereNormal
     Float=&RandomRangeFloat(Minimum=50, Maximum=150, Evaluation Type=::NewEnumerator1)
    )
   ) {enabled=false, module=AddVelocity001}
  ]
  particleUpdateScript [
   &ParticleState()
   &ScaleMeshSize(Scale Factor=&MakeVector(
     X=0.5
     Y=0.5
     Z=&FloatFromCurve(
      FloatCurve=DataInterface() {propertyValues {Curve {keys [<at 0 0 RCIM_Cubic>, {arriveTangent=3.4753, tangentMode=RCTM_User, time=0.25, value=1}]}}}
      Scale Curve=3
     )
    ))
   /SetVariables(Particles.VisibilityTag=&SpawnGroup()) {isSetParameters=true}
   &AccelerationForce(Acceleration=<0 0 -900>) {enabled=false}
   &Drag(Drag=2, Use Linear Drag=true, Use Rotational Drag=false) {enabled=false}
   &ScaleSpriteSize(Scale Factor=&Vector2DFromFloat(Value=&FloatFromCurve(
      FloatCurve=DataInterface() {propertyValues {Curve {keys [
          {tangentMode=RCTM_User, time=0, value=1}
          {arriveTangent=-0.0405, interpMode=RCIM_Cubic, leaveTangent=-0.0405, tangentMode=RCTM_User, time=0.7009, value=1}
          {arriveTangent=-8.0089, interpMode=RCIM_Cubic, leaveTangent=-8.0089, tangentMode=RCTM_User, time=1, value=0}
         ]}}}
      Scale Curve=0.5
     ))) {enabled=false, module=ScaleSpriteSize001}
   &Collision(
    Age Colliding Particles=false
    Module.Control Rest State Via Surface Normal=false
    Module.EnableMaxCollisionCount=false
    Module.Randomize Collision Normal Vector=false
    Restitution=0.1
   ) {enabled=false}
   &GravityForce(Gravity=<0 0 -1000>)
   &SolveForcesAndVelocity(Clamp Velocity=false, Limit Acceleration=false)
   &ScaleMeshSizeBySpeed(Max Scale Factor=<1 1 2>)
   &ScaleSpriteSizeBySpeed(
    Max Scale Factor=<0.66 5>
    Source Velocity:Vector3f=&Multiply_Vector(A=<0.01 0.01 1>, B=$Particles.Velocity)
    Velocity Threshold=600
   )
   &FN_Light_Attributes(
    Color=$Particles.Color
    Create Color Attribute=true
    Create Exponent Attribute=true
    Create Position Attribute=true
    Create Radius Attribute=true
    Enable Particle Light Enable=true
    Exponent=2
    Light Enabled=true
    Position:NiagaraPosition=$Particles.Position
    Radius=25
   ) {enabled=false}
   /SetVariables(
    Particles.Linear Color=&MultiplyLinearColor(Linear Color=$Particles.Color, Scale Factor=<0.02 0.02 0.02 0.0001>)
    Particles.Vector:Vector3f=&Multiply_Vector(A=$Particles.Velocity, B=<1 1 1>)
    Particles.Vector 2D:Vector2f=&Multiply_Vector2D(A=<2 5>, B=$Particles.SpriteSize)
   ) {isSetParameters=true}
   &ScaleColor(
    Scale RGB=&VectorFromFloat(Value=&Lerp_Float(
      A=0.35
      B=5
      Alpha=&FloatFromCurve(
       FloatCurve=DataInterface() {propertyValues {Curve {keys [
           {arriveTangent=9.0725, interpMode=RCIM_Cubic, time=0, value=0}
           <at 0.1 1 RCIM_Cubic>
           {arriveTangent=-0.0088, interpMode=RCIM_Cubic, leaveTangent=-0.0088, tangentMode=RCTM_User, time=1, value=0}
          ]}}}
       CurveIndex=$NPC.Radiant_Music_NPC.MusicBeatProgress
      )
     ))
    ScaleA=true
    ScaleRGB=true
   ) {module=ScaleColor002}
  ]
  rendererClasses [
   &NiagaraSpriteRendererProperties
   &NiagaraLightRendererProperties
   &NiagaraMeshRendererProperties
  ]
  renderers [
   {rendererClass=&NiagaraSpriteRendererProperties}
   {rendererClass=&NiagaraLightRendererProperties, rendererIndex=1}
   {rendererClass=&NiagaraSpriteRendererProperties, rendererIndex=2}
   {rendererClass=&NiagaraSpriteRendererProperties, rendererIndex=3}
   {rendererClass=&NiagaraMeshRendererProperties, rendererIndex=4}
  ]
 }
 Spring {
  bEnabled=false
  emitterUpdateScript [
   &EmitterState()
   &SpawnBurst_Instantaneous(Use Spawn Probability=false, UseLoopCountLimit=false)
  ]
  particleSpawnScript [&InitializeParticle(
    Color=<rgba 0 1 0.3 1>
    Lifetime=1000
    Sprite Size Mode=::NewEnumerator3
    Uniform Sprite Size=15
    UsePositionOffset=false
    Write Material Random=true
   ) {module=InitializeParticle001}]
  particleUpdateScript [
   &ParticleState()
   &self:TargetPos_from_Parent_02(PAR_DI=DataInterface() {propertyValues {EmitterBinding {bindingMode=Other, emitterName=Empty}}}) {module=TargetPos_from_Parent_02}
   /Niagara/Modules/Update/Forces/SpringForce(
    " Particle Equilibrium Position":NiagaraPosition=&ConvertPositionToVector(Input Position=$Particles.TargetPosition, Convert To Absolute Large World Vector=false)
    Dampening Coefficient=35
    Desired Minimum Separation Distance=0
    Spring Tightness=500
   ) {enabled=false}
   &Drag(Drag=3, Use Linear Drag=true, Use Rotational Drag=false) {enabled=false}
   &SpringForce(" Particle Equilibrium Position":NiagaraPosition=$Particles.TargetPosition, Dampening Coefficient=35, Spring Tightness=1) {module=SpringForce001}
   &SolveForcesAndVelocity(Clamp Velocity=false, Limit Acceleration=false)
   &GenerateLocationEvent(Event Send Rate=60, Use Event Probability=false, UseEventDelay=false)
  ]
  rendererClasses [&NiagaraSpriteRendererProperties]
  renderers [{rendererClass=&NiagaraSpriteRendererProperties}]
 }
}
summary {
 systemName=NS_Assault_Radient_ARG_Beam
 emitters [
  {emitterName=RingsMeshesSpacedOut, rendererClasses [&NiagaraMeshRendererProperties]}
  {
   emitterName=MainRibbon001
   rendererClasses [
    &NiagaraRibbonRendererProperties
    &NiagaraMeshRendererProperties
    &NiagaraSpriteRendererProperties
   ]
  }
  {emitterName=Plasma003, rendererClasses [&NiagaraRibbonRendererProperties, &NiagaraMeshRendererProperties]}
  {emitterName=EQ_Boxes, rendererClasses [&NiagaraMeshRendererProperties]}
  {emitterName=Cube_Glitch, rendererClasses [&NiagaraMeshRendererProperties]}
  {emitterName=Cube_Glitch001, rendererClasses [&NiagaraSpriteRendererProperties]}
  {
   bEnabled=false
   emitterName=SparkLengthCollisionTest
   rendererClasses [
    &NiagaraSpriteRendererProperties
    &NiagaraLightRendererProperties
    &NiagaraMeshRendererProperties
   ]
  }
  {emitterName=AirflowConeHot, rendererClasses [&NiagaraMeshRendererProperties]}
  {bEnabled=false, emitterName=Debug_Audio_Reactivity_Sprite, rendererClasses [&NiagaraSpriteRendererProperties]}
  {bEnabled=false, emitterName=Plasma, rendererClasses [&NiagaraRibbonRendererProperties]}
  {bEnabled=false, emitterName=Empty001, rendererClasses [&NiagaraSpriteRendererProperties]}
  {bEnabled=false, emitterName=Spring, rendererClasses [&NiagaraSpriteRendererProperties]}
  {bEnabled=false, emitterName=Plot_Spring, rendererClasses [&NiagaraRibbonRendererProperties]}
  {bEnabled=false, emitterName=Empty, rendererClasses [&NiagaraSpriteRendererProperties]}
  {emitterName=Light, rendererClasses [&NiagaraLightRendererProperties, &NiagaraSpriteRendererProperties]}
  {
   emitterName=Beam_Lights
   rendererClasses [
    &NiagaraMeshRendererProperties
    &NiagaraLightRendererProperties
    &NiagaraSpriteRendererProperties
   ]
  }
  {
   bEnabled=false
   emitterName=EQBarHit
   rendererClasses [&NiagaraSpriteRendererProperties, &NiagaraMeshRendererProperties]
  }
 ]
 userVariables [
  User.BeamEnd_0 {type=NiagaraPosition, description="", defaultValue {}}
  User.BeamStart_0_0 {type=Vector3f, description="", defaultValue {struct=Vector3f, value=<1000 0 1000>}}
  User.bIsFrontEnd {type=NiagaraBool}
  User.bIsFrontEndPreview {type=NiagaraBool}
  User.EndPoint {type=Vector3f, description="", defaultValue {struct=Vector3f, value=<3000 0 2000>}}
  User.LocalPlayer {type=NiagaraBool}
  User.MinWarmupTime {type=NiagaraFloat, defaultValue=1.25, description=""}
  User.ParaGlideLeanAlpha {type=NiagaraFloat, defaultValue=0, description=""}
  User.PhysicalSurface {type=NiagaraInt32}
  User.WarmupTime {type=NiagaraFloat, defaultValue=2.5, description=""}
 ]
}
