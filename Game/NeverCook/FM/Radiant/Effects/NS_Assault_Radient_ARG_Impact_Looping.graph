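# Looping Niagara impact effect for the Radiant assault weapon, with glow, sparks and digital bits that react to music.
# Corona sprite and light, GPU bit sprites offset using beam endpoints, a mesh flash, ribbon and spark emitters form the visible layers.
# Music parameter collections drive corona size by beat, mesh scale by amplitude and spark brightness by spectrum.
# Physical surface gates ribbon and spark bursts; surface-specific debris and mud emitters are disabled.
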NiagaraSystem /Game/NeverCook/FM/Radiant/Effects/NS_Assault_Radient_ARG_Impact_Looping
dependencies {
 usedDataInterfaces [
  &NiagaraDataInterfaceCurve
  &NiagaraDataInterfaceColorCurve
  &NiagaraDataInterfaceVectorCurve
  &NiagaraDataInterfaceArrayFloat
  &NiagaraDataInterfaceSpriteRendererInfo
  &NiagaraDataInterfaceVector2DCurve
 ]
 usedDynamicInputs [&RotateVectorByQuaternion]
 usedModules [&SystemState]
}
emitters {
 Corona {
  emitterUpdateScript [
   &EmitterState()
   &SpawnBurst_Instantaneous(Spawn Count=1, Use Spawn Probability=false)
   &SpawnRate(SpawnRate=6, Use Spawn Probability=false) {enabled=false}
  ]
  particleSpawnScript [
   &InitializeParticle(
    Color=<rgba 4 0 3.5041 6>
    Lifetime=0.1
    Sprite Rotation Mode=::NewEnumerator2
    Sprite Size Mode=::NewEnumerator3
    Uniform Sprite Size=150
    UsePositionOffset=true
    Write Material Random=true
   )
   &DynamicMaterialParameters(
    Index 0 Param 1=0.35
    Index 0 Param 2=1
    Module.Param0WriteEnabled=true
    Module.Param1WriteEnabled=true
    Module.Param2WriteEnabled=false
    Module.Param3WriteEnabled=false
   ) {enabled=false}
  ]
  particleUpdateScript [
   &ParticleState()
   &CameraOffset(Camera Offset Amount=10)
   &ScaleColor(
    Scale Alpha=&FloatFromCurve(FloatCurve=DataInterface() {propertyValues {Curve {keys [
         {arriveTangent=2.3365, interpMode=RCIM_Cubic, leaveTangent=2.3365, tangentMode=RCTM_User, time=0, value=0}
         {interpMode=RCIM_Cubic, tangentMode=RCTM_User, time=0.5, value=1}
        ]}}})
    ScaleA=true
    ScaleRGB=false
   ) {enabled=false}
   &ScaleSpriteSize(Scale Factor=&Vector2DFromFloat(Value=&Lerp_Float(
      A=0.9
      B=2
      Alpha=&FloatFromCurve(
       FloatCurve=DataInterface() {propertyValues {Curve {keys [
           {interpMode=RCIM_Cubic, tangentMode=RCTM_User, time=0, value=0}
           {interpMode=RCIM_Cubic, tangentMode=RCTM_User, time=0.5, value=1}
           {tangentMode=RCTM_User, time=1, value=0}
          ]}}}
       CurveIndex=$NPC.Radiant_Music_NPC.MusicBeatProgress
      )
     ))) {module=ScaleSpriteSize001}
   &FN_Light_Attributes(
    Color=<rgba 0.6667 0 4 1>
    Create Color Attribute=true
    Create Exponent Attribute=true
    Create Radius Attribute=true
    Exponent=2
    Radius=125
    Scale Color=&Waveform(Global Amplitude Scale=4, "Waveform [1]"=::NewEnumerator12, "[1] Frequency"=13, "[1] Amplitude Min/Max"=<0.33 1>)
   )
  ]
  rendererClasses [&NiagaraSpriteRendererProperties, &NiagaraLightRendererProperties]
  renderers [
   {rendererClass=&NiagaraSpriteRendererProperties}
   {rendererClass=&NiagaraLightRendererProperties, rendererIndex=1}
  ]
 }
 Digital_Bits002 {
  simTarget=GPUComputeSim
  emitterUpdateScript [
   &EmitterState(Inactive Response=::NewEnumerator1, Life Cycle Mode=::NewEnumerator1, Loop Delay=0.5, Loop Duration=0.1, UseLoopDelay=false)
   &SpawnBurst_Instantaneous(Spawn Count=35, Use Spawn Probability=false, UseLoopCountLimit=false)
   &SpawnRate(SpawnRate=250, Use Spawn Probability=false) {enabled=false, module=SpawnRate001}
   &SpawnRate(
    SpawnRate=&FloatFromCurve(
     FloatCurve=DataInterface() {propertyValues {Curve {keys [<at 0 1 RCIM_Cubic>, {arriveTangent=-2.7657, interpMode=RCIM_Cubic, leaveTangent=-2.7657, tangentMode=RCTM_User, time=1, value=0}]}}}
     CurveIndex=$Emitter.NormalizedLoopAge
     Scale Curve=&RandomRangeFloat(Minimum=100, Maximum=750)
    )
    Use Spawn Probability=false
   ) {enabled=false}
  ]
  particleSpawnScript [
   &PartitionParticles(Number Of Partitions=7, Partition Mode:ENiagara_PartitionMode=::NewEnumerator5) {enabled=false}
   /SetVariables(Particles.ChromaOffset=50) {isSetParameters=true}
   &InitializeParticle(
    Lifetime Max=0.2
    Lifetime Min=0.05
    Lifetime Mode=::NewEnumerator1
    Sprite Size Max=<1 1>
    Sprite Size Min=<0.1 0.1>
    Sprite Size Mode=::NewEnumerator2
    UsePositionOffset=false
    Write Material Random=true
   ) {module=InitializeParticle002}
   &SkeletalMeshLocation(
    Inherit Velocity=false
    MaskSpawnGroup=false
    Mesh Sampling Type=::NewEnumerator7
    Module.Override Randomness=false
    OffsetAlongNormal=false
    Skeletal Mesh=$System.Skeletal Mesh
   ) {enabled=false}
   /SetVariables(
    Particles.NormalVelocity:Vector3f=$Output.SkeletalMeshLocation.SampledNormal
    Particles.TextureSampleA:LinearColor=$Output.SampleTexture.SampledColor
    Particles.TextureSampleB:LinearColor=$Output.SampleTexture001.SampledColor
   ) {isSetParameters=true}
   &CameraOffset(Camera Offset Amount=-10) {enabled=false}
   &AddVelocity(Rotation Coordinate Space=::World, Velocity=<0 0 -50>) {module=AddVelocity001}
   &ShapeLocation(MaskSpawnGroup=false, Override Seed=false, Sphere Radius=80, UseOwnerScale=true)
  ]
  particleUpdateScript [
   &ParticleState(Kill Particles When Lifetime Has Elapsed=false, Let Infinitely Lived Particles Die When Emitter Deactivates=true)
   &self:SphereMaskEdge(
    Cull Particles outside of max distance=true
    Max Distance:NiagaraFloat=&FloatFromCurve(
     FloatCurve=DataInterface() {propertyValues {Curve {keys [
         {arriveTangent=1.1063, interpMode=RCIM_Cubic, leaveTangent=1.1063, tangentMode=RCTM_User, time=0.001, value=0}
         {arriveTangent=1.1925, interpMode=RCIM_Cubic, leaveTangent=1.1925, tangentMode=RCTM_User, time=0.9, value=1}
        ]}}}
     CurveIndex=$Emitter.NormalizedLoopAge
     Scale Curve=128
    )
    Min Distance:NiagaraFloat=&FloatFromCurve(
     FloatCurve=DataInterface() {propertyValues {Curve {keys [
         {arriveTangent=0.7543, interpMode=RCIM_Cubic, leaveTangent=0.7543, tangentMode=RCTM_User, time=0.004, value=0}
         {arriveTangent=1.004, interpMode=RCIM_Cubic, leaveTangent=1.004, time=0.6, value=0.451}
         {arriveTangent=1.4135, interpMode=RCIM_Cubic, leaveTangent=1.4135, tangentMode=RCTM_User, time=1, value=1}
        ]}}}
     CurveIndex=$Emitter.NormalizedLoopAge
     Scale Curve=128
    )
    SphereCenter:Vector3f=<0 8.5591 128>
   ) {enabled=false, module=SphereMaskEdge}
   &AddVelocity(Velocity=&Multiply_Vector(A=$Particles.NormalVelocity, B=&VectorFromFloat(Value=-1)))
   &SolveForcesAndVelocity(Clamp Velocity=false, Limit Acceleration=false)
   /SetVariables(Particles.ChromaOffset:NiagaraFloat=&FloatFromCurve(
     FloatCurve=DataInterface() {propertyValues {Curve {keys [{interpMode=RCIM_Cubic, tangentMode=RCTM_User, time=0, value=1}, {interpMode=RCIM_Cubic, tangentMode=RCTM_User, time=1, value=0}]}}}
     Scale Curve=0.5
    )) {isSetParameters=true}
   /SetVariables(Particles.Color001:LinearColor=<rgba 0.005 0 0.1 0.5>) {isSetParameters=true}
   /SetVariables(Particles.Color002=<rgba 0.0206 1 0.6109 0.5>) {isSetParameters=true}
   /SetVariables(Particles.Color003:LinearColor=<rgba 0.8542 0 1 0.5>) {isSetParameters=true}
   &ScaleSpriteSize(Scale Factor=&Vector2DFromCurve(
     Vector2Curve=DataInterface() {propertyValues {
       XCurve {keys [
         {interpMode=RCIM_Cubic, tangentMode=RCTM_User, time=0, value=1}
         {arriveTangent=-6.968, interpMode=RCIM_Cubic, leaveTangent=-6.968, tangentMode=RCTM_User, time=0.5, value=2}
         {tangentMode=RCTM_User, time=1, value=0}
        ]}
       YCurve {keys [
         {interpMode=RCIM_Cubic, tangentMode=RCTM_User, time=0, value=0}
         {arriveTangent=13.8249, interpMode=RCIM_Cubic, leaveTangent=13.8249, tangentMode=RCTM_User, time=0.6, value=2}
         {arriveTangent=8.3745, interpMode=RCIM_Cubic, leaveTangent=8.3745, tangentMode=RCTM_User, time=1, value=8}
        ]}
      }}
     ScaleCurve=<2 8>
    ))
   /SetVariables(Particles.SpriteSize001:Vector2f=&Multiply_Vector2D_ByFloat(Float=1, Vector2D=$Particles.SpriteSize)) {isSetParameters=true}
   /SetVariables(Particles.Position=&AddVectorToPosition(
     Position=$Particles.ShapeLocation.ShapePosition
     Vector=&self:ScratchDynamicInput_02(
      Default:Vector3f=&TransformVector(Vector=$User.BeamStart_0_0, Source Space=::World, Destination Space=::Local)
      Vector:Vector3f=&TransformVector(Vector=Linked() {linkedVariable {name=User.BeamEnd_0, type=NiagaraPosition}}, Source Space=::World, Destination Space=::Local)
      Spacing:NiagaraFloat=50
     )
    )) {isSetParameters=true}
  ]
  rendererClasses [&NiagaraSpriteRendererProperties]
  renderers [
   {rendererClass=&NiagaraSpriteRendererProperties}
   {rendererClass=&NiagaraSpriteRendererProperties, rendererIndex=1}
   {rendererClass=&NiagaraSpriteRendererProperties, rendererIndex=2}
  ]
 }
 FN_Weapons_Debris_Sprite_SubUV_BETA003 {
  bEnabled=false
  emitterUpdateScript [
   &EmitterState()
   &SpawnBurst_Instantaneous(
    Spawn Count=1
    Spawn Probability=&MakeFloatFromInt(Int=/Game/Effects/Niagara/DataChannels/SelectIntBySurface(
      Default:NiagaraInt32=0
      Surface:EPhysicalSurface=&self:ScratchDynamicInput_01(PhysicalSurface:NiagaraInt32=$User.PhysicalSurface)
      Num Surfaces:NiagaraInt32=1
      Surface 0:EPhysicalSurface=SurfaceType8
      Value For Surface 0:NiagaraInt32=1
      Surface 1:EPhysicalSurface=?
      Value For Surface 1:NiagaraInt32=?
      Surface 2:EPhysicalSurface=?
      Value For Surface 2:NiagaraInt32=?
      Surface 3:EPhysicalSurface=?
      Value For Surface 3:NiagaraInt32=?
     ))
    Use Spawn Probability=true
   )
  ]
  particleSpawnScript [
   /Niagara/Modules/Spawn/Initialization/InitializeParticle(
    Lifetime=0.3
    Module.Write Color=true
    Module.Write Lifetime=true
    Module.Write Mass=true
    Module.Write Material Random=true
    Module.Write Position=true
    Module.Write Scale=false
    Module.Write SpriteRotation=false
    Module.Write SpriteSize=true
    Position=&Add_Vector(A=/Niagara/DynamicInputs/Helpers/SimulationPosition(), B=<0 0 2>)
    Sprite Size=&RandomRangeVector2D(Minimum=<6 10>, Maximum=<12 16>)
   )
   &AddVelocityInCone(
    Cone Angle=30
    Module.Use Velocity Falloff On Cone Axis=false
    Velocity Strength=&UniformRangedFloat(Minimum=200, Maximum=300)
   )
   &FN_PerParticleLOD(Enable SpriteSize Distance Scaling=true, Sprite Size Scale Far=<1.5 1.5>)
   /NiagaraFN/Sprite/FN_RandomizeSpriteUV()
  ]
  particleUpdateScript [
   &ParticleState()
   &AccelerationForce(Acceleration=<0 0 -800>)
   &Drag(Drag=0.1, Use Linear Drag=true, Use Rotational Drag=false) {enabled=false}
   &SolveForcesAndVelocity(Clamp Velocity=false, Limit Acceleration=false)
   &ScaleSpriteSize(Scale Factor=&Vector2DFromFloat(Value=&FloatFromCurve(FloatCurve=DataInterface() {propertyValues {Curve {keys [<at -0.0046 1.0113>,<at 0.77 7.78>,<at 1.0011 0.002>]}}})))
   &DynamicMaterialParameters(
    Index 0 Param 2=&FloatFromCurve(FloatCurve=DataInterface() {propertyValues {Curve {keys [
         {arriveTangent=46.3377, interpMode=RCIM_Cubic, leaveTangent=46.3379, tangentMode=RCTM_User, time=-0.0046, value=0}
         {arriveTangent=0.0019, interpMode=RCIM_Cubic, leaveTangent=0.0019, time=0.05, value=0.9763}
         {arriveTangent=-2.5858, interpMode=RCIM_Cubic, leaveTangent=-2.5858, tangentMode=RCTM_User, time=1.0011, value=0.002}
        ]}}})
    Module.Param0WriteEnabled=false
    Module.Param1WriteEnabled=true
    Module.Param2WriteEnabled=false
    Module.Param3WriteEnabled=false
   )
   /Niagara/Modules/Update/SubUV/SubUVAnimation(Random SubUV Lookup Index:NiagaraFloat=$Particles.Initial.MaterialRandom, SubUV Animation Mode=::NewEnumerator1) {enabled=false}
   &Color(Color=&MakeLinearColorFromVectorAndFloat(
     "Vector (RGB)"=&MakeVectorFromLinearColorRGB(Color=<rgba 1 0 0.8926 1>)
     "Float (Alpha)"=&FloatFromCurve(FloatCurve=DataInterface() {propertyValues {Curve {keys [<at 0.4 1.0113 RCIM_Cubic>,<at 1.0011 0.002>]}}})
    ))
  ]
  rendererClasses [&NiagaraSpriteRendererProperties]
  renderers [{rendererClass=&NiagaraSpriteRendererProperties}]
 }
 Mud_DropsCollision {
  bEnabled=false
  emitterUpdateScript [
   &EmitterState(Enable Distance Culling=true, MaxDistance=4000, Module.bUseMaxDistance=true, Module.bUseMinDistance=false, Scalability Mode=::NewEnumerator1)
   &SpawnRate(SpawnRate=15, Use Spawn Probability=false) {enabled=false}
   &SpawnBurst_Instantaneous(
    Spawn Count=6
    Spawn Probability=&MakeFloatFromInt(Int=&SelectIntBySurface(
      Default:NiagaraInt32=0
      Surface:EPhysicalSurface=&self:ScratchDynamicInput_01(PhysicalSurface:NiagaraInt32=$User.PhysicalSurface)
      Num Surfaces:NiagaraInt32=1
      Surface 0:EPhysicalSurface=SurfaceType30
      Value For Surface 0:NiagaraInt32=1
      Surface 1:EPhysicalSurface=?
      Value For Surface 1:NiagaraInt32=?
      Surface 2:EPhysicalSurface=?
      Value For Surface 2:NiagaraInt32=?
      Surface 3:EPhysicalSurface=?
      Value For Surface 3:NiagaraInt32=?
     ))
    Use Spawn Probability=true
    UseLoopCountLimit=false
   )
  ]
  particleSpawnScript [
   /Niagara/Modules/Spawn/Initialization/V2/InitializeParticle(
    Lifetime Max=0.6
    Lifetime Min=0.4
    Lifetime Mode=::NewEnumerator1
    Sprite Size Mode=::NewEnumerator1
    Uniform Sprite Size Max=8
    Uniform Sprite Size Min=3
    UsePositionOffset=false
    Write Material Random=true
   )
   &ShapeLocation(
    Hemisphere Distribution=<1 0.66>
    MaskSpawnGroup=false
    Override Seed=false
    Rotation Mode=::NewEnumerator1
    Sphere Radius=10
    Sphere Surface Distribution=1
    UseOwnerScale=true
   ) {module=ShapeLocation001}
   &SkeletalMeshLocation(
    Bone / Socket Index=8
    Bone Sampling Mode=::NewEnumerator4
    Interpolate Mesh Sampling=false
    MaskSpawnGroup=false
    Sampled Position Offset=$Particles.ShapeLocation001.ShapePosition
    Skeletal Mesh=$User.SkeletalMesh
    Use Sampled Mesh Attribute as Ribbon ID=false
   ) {enabled=false}
   &AddVelocity(
    Additional Quaternion Rotation=false
    Rotation Mode=::NewEnumerator3
    Velocity=$Particles.ShapeLocation001.ShapeNormal
    Velocity Speed Scale=&RandomRangeFloat(Minimum=100, Maximum=200)
   ) {module=AddVelocity001}
   &AddVelocity(Velocity=<1 0 0>, Velocity Speed Scale=&RandomRangeFloat(Minimum=50, Maximum=300))
   &ApplyInitialForces(Apply Rotational Force to Rotational Velocity=false)
   /SetVariables() {enabled=false, isSetParameters=true}
  ]
  particleUpdateScript [
   &ParticleState()
   &GravityForce(Gravity=<1 0 -900>)
   &Collision(
    Age Colliding Particles=false
    Friction=1
    Friction During a Bounce=1
    Max CPU Trace Length=100
    Module.Control Rest State Via Surface Normal=false
    Module.EnableMaxCollisionCount=false
    Module.Randomize Collision Normal Vector=false
    Particle Radius Scale=0.333
    Restitution=0
   ) {enabled=false, module=Collision001}
   &InheritSourceMovement(
    Applied Position Scale=&VectorFromFloat(Value=&MakeCustomFloatFromBool(Bool=$Particles.HasCollided, True Float=0, False Float=0.9))
    ApplyForce=false
    ApplyPosition=true
    ApplyPositionFalloff=false
   ) {enabled=false}
   &SolveForcesAndVelocity(Clamp Velocity=false, Limit Acceleration=false)
   &Timeline(
    Additional Output:ENiagara_TimelineOutput=::NewEnumerator1
    Additional Output Mode:ENiagara_TimelineOutputMode=::NewEnumerator1
    Float Curve=DataInterface() {propertyValues {Curve {keys [
        {arriveTangent=9.0725, interpMode=RCIM_Cubic, leaveTangent=9.0725, tangentMode=RCTM_User, time=0, value=0}
        {arriveTangent=-2.5249, interpMode=RCIM_Cubic, leaveTangent=-2.5249, tangentMode=RCTM_User, time=0.1408, value=0.9847}
        {arriveTangent=-0.0405, interpMode=RCIM_Cubic, leaveTangent=-0.0405, tangentMode=RCTM_User, time=1, value=0}
       ]}}}
    Instantaneous Rewind=true
    Play=$Particles.HasCollided
   ) {enabled=false}
   &DynamicMaterialParameters(
    Index 0 Param 1=$Particles.Timeline.Float
    Index 0 Param 2=5
    Index 0 Param 3=10
    Index 0 Param 4=5
    Index 1 Param 1=&FloatFromCurve(
     FloatCurve=DataInterface() {propertyValues {Curve {keys [
         {arriveTangent=-3.7492, interpMode=RCIM_Cubic, leaveTangent=-3.7492, tangentMode=RCTM_User, time=0, value=1}
         {arriveTangent=-0.7477, interpMode=RCIM_Cubic, leaveTangent=-0.7477, tangentMode=RCTM_User, time=0.3, value=0.3}
         {interpMode=RCIM_Cubic, leaveTangent=-1.9286, time=1, value=0}
        ]}}}
     CurveIndex=$Particles.Timeline.NormalizedTime
     Scale Curve=&RandomRangeFloat(Minimum=2, Maximum=2.25)
    )
    Module.Param0WriteEnabled=true
    Module.Param0WriteEnabled1=true
    Module.Param1WriteEnabled=true
    Module.Param1WriteEnabled1=false
    Module.Param2WriteEnabled=true
    Module.Param2WriteEnabled1=false
    Module.Param3WriteEnabled=true
    Module.Param3WriteEnabled1=false
    Write Parameter Index 1=true
   )
   &ScaleColor(Scale RGB=<0 0 0>, ScaleA=false, ScaleRGB=true)
   &ScaleSpriteSize(
    Uniform Curve Scale=&MakeCustomFloatFromBool(
     Bool=$System.SpawnShotgun
     True Float=0.4
     False Float=&MakeCustomFloatFromBool(Bool=$System.SpawnPistol, True Float=0.6, False Float=1)
    )
    Uniform Curve Sprite Scale=DataInterface() {propertyValues {Curve {keys [
        {interpMode=RCIM_Cubic, tangentMode=RCTM_User, time=0.7, value=1}
        {arriveTangent=-6.3247, interpMode=RCIM_Cubic, leaveTangent=-6.3247, tangentMode=RCTM_User, time=1, value=0}
       ]}}}
   )
  ]
  rendererClasses [&NiagaraSpriteRendererProperties]
  renderers [{rendererClass=&NiagaraSpriteRendererProperties}]
 }
 NE_Parent_Mud_Splash {
  bEnabled=false
  emitterUpdateScript [
   &EmitterState(Enable Distance Culling=true, MaxDistance=9000, Module.bUseMaxDistance=true, Module.bUseMinDistance=false, Scalability Mode=::NewEnumerator1)
   &SpawnPerUnit(Module.Use Max Movement Threshold=true, Module.Use Movement Tolerance=false, Use Spawn Probability=false) {enabled=false}
   &SpawnBurst_Instantaneous(
    Spawn Count=3
    Spawn Probability=&MakeFloatFromInt(Int=&SelectIntBySurface(
      Default:NiagaraInt32=0
      Surface:EPhysicalSurface=&self:ScratchDynamicInput_01(PhysicalSurface:NiagaraInt32=$User.PhysicalSurface)
      Num Surfaces:NiagaraInt32=1
      Surface 0:EPhysicalSurface=SurfaceType30
      Value For Surface 0:NiagaraInt32=1
      Surface 1:EPhysicalSurface=?
      Value For Surface 1:NiagaraInt32=?
      Surface 2:EPhysicalSurface=?
      Value For Surface 2:NiagaraInt32=?
      Surface 3:EPhysicalSurface=?
      Value For Surface 3:NiagaraInt32=?
     ))
    Use Spawn Probability=true
    UseLoopCountLimit=false
   )
   &SpawnBurst_Instantaneous(
    Spawn Count=&MakeIntFromFloat(Float=&MakeFloatFromInt(Int=&SelectIntBySurface(
       Default:NiagaraInt32=0
       Surface:EPhysicalSurface=&self:ScratchDynamicInput_01(PhysicalSurface:NiagaraInt32=$User.PhysicalSurface)
       Num Surfaces:NiagaraInt32=1
       Surface 0:EPhysicalSurface=SurfaceType30
       Value For Surface 0:NiagaraInt32=1
       Surface 1:EPhysicalSurface=?
       Value For Surface 1:NiagaraInt32=?
       Surface 2:EPhysicalSurface=?
       Value For Surface 2:NiagaraInt32=?
       Surface 3:EPhysicalSurface=?
       Value For Surface 3:NiagaraInt32=?
      )))
    Spawn Group=1
    Spawn Probability=0.5
    Use Spawn Probability=true
    UseLoopCountLimit=false
   ) {module=SpawnBurst_Instantaneous001}
   &SpawnRate(SpawnRate=200, Use Spawn Probability=false) {enabled=false}
   &SpawnBurst_Instantaneous(
    Spawn Probability=&MakeFloatFromInt(Int=&SelectIntBySurface(
      Default:NiagaraInt32=0
      Surface:EPhysicalSurface=&self:ScratchDynamicInput_01(PhysicalSurface:NiagaraInt32=$User.PhysicalSurface)
      Num Surfaces:NiagaraInt32=1
      Surface 0:EPhysicalSurface=SurfaceType30
      Value For Surface 0:NiagaraInt32=1
      Surface 1:EPhysicalSurface=?
      Value For Surface 1:NiagaraInt32=?
      Surface 2:EPhysicalSurface=?
      Value For Surface 2:NiagaraInt32=?
      Surface 3:EPhysicalSurface=?
      Value For Surface 3:NiagaraInt32=?
     ))
    Spawn Time=0.474
    Use Spawn Probability=true
    UseLoopCountLimit=false
   ) {enabled=false, module=SpawnBurst_Instantaneous002}
  ]
  particleSpawnScript [
   /SetVariables(Particles.VisibilityTag=$Emitter.SpawnGroup) {isSetParameters=true}
   &InitializeParticle(
    Color=<rgba 1 1 1 1.2>
    Lifetime=&MaskFloatBySpawnGroup(Masked Float=0.45, Passthrough Float=&RandomRangeFloat(Minimum=0.5, Maximum=0.6), Spawn Group Mask=1)
    Mesh Scale Mode=::NewEnumerator3
    Mesh Uniform Scale=&Multiply_Float(A=&RandomRangeFloat(Minimum=2, Maximum=2.5), B=1)
    Position Offset=<0 0 0.5>
    Sprite Rotation Mode=::NewEnumerator2
    Sprite Size Mode=::NewEnumerator1
    Sprite UV Mode=::NewEnumerator3
    Uniform Sprite Size Max=50
    Uniform Sprite Size Min=40
    UsePositionOffset=true
    Write Material Random=true
   )
   /Niagara/Modules/Update/SubUV/V2/SubUVAnimation(
    "Sprite Renderer "=DataInterface() {propertyValues {SpriteRenderer=&self:NE_Parent_Mud_Splash.NiagaraSpriteRendererProperties_0}}
    SubUV Animation Mode:ENiagara_SubUVLookupModeV2=::NewEnumerator2
    UseEndFrame=false
    UseStartFrame=false
   )
   &ShapeLocation(
    Apply To Particle Position=::NewEnumerator1
    MaskSpawnGroup=false
    Offset Coordinate Space=::Local
    Offset Mode=::NewEnumerator0
    Override Seed=false
    Rotation Coordinate Space=::World
    Rotation Mode=::NewEnumerator1
    Sphere Radius=&MaskFloatBySpawnGroup(Masked Float=0.1, Passthrough Float=1, Spawn Group Mask=1)
    UseOwnerScale=true
   ) {module=ShapeLocation001}
   &InitialMeshOrientation(
    Mesh Orientation Mode=::NewEnumerator7
    Rotation=&RandomRangeVector(Minimum=<0 0 0>, Maximum=<0 0 360>)
    Use Rotation Vector=true
   )
   &AddVelocity(
    Velocity=$Particles.ShapeLocation001.ShapeNormal
    Velocity Speed Scale=&MaskFloatBySpawnGroup(Masked Float=0, Passthrough Float=100, Spawn Group Mask=1)
   )
   &AddVelocity(
    Velocity=<1 0 0>
    Velocity Speed Scale=&MaskFloatBySpawnGroup(Masked Float=10, Passthrough Float=200, Spawn Group Mask=1)
   ) {module=AddVelocity001}
   &CameraOffset(Camera Offset Amount=15)
  ]
  particleUpdateScript [
   &ParticleState()
   &ScaleMeshSize(Scale Factor=&Multiply_VectorByFloat(
     Vector=&MakeVector(
      X=2
      Y=2
      Z=&FloatFromCurve(
       FloatCurve=DataInterface() {propertyValues {Curve {keys [
           {tangentMode=RCTM_User, time=0, value=1.875}
           {arriveTangent=-0.0405, interpMode=RCIM_Cubic, leaveTangent=-0.0405, tangentMode=RCTM_User, time=0.5, value=0.875}
           {arriveTangent=-3.7129, interpMode=RCIM_Cubic, leaveTangent=-3.7129, tangentMode=RCTM_User, time=1, value=0}
          ]}}}
       Scale Curve=1.25
      )
     )
     Float=&FloatFromCurve(
      FloatCurve=DataInterface() {propertyValues {Curve {keys [
          {arriveTangent=3.4838, interpMode=RCIM_Cubic, leaveTangent=3.4838, tangentMode=RCTM_User, time=0, value=0.2}
          {arriveTangent=1.1179, interpMode=RCIM_Cubic, leaveTangent=1.1179, tangentMode=RCTM_User, time=0.2, value=0.6}
          {arriveTangent=0.0224, interpMode=RCIM_Cubic, leaveTangent=0.0224, tangentMode=RCTM_User, time=1, value=1}
         ]}}}
      Scale Curve=1.05
     )
    ))
   &GravityForce(Gravity=&MaskVectorBySpawnGroup(Masked Vector=<0 0 -300>, Passthrough Vector=<0 0 -1200>, Spawn Group Mask=1))
   &InheritSourceMovement(
    Applied Position Scale=&VectorFromFloat(Value=0.95)
    ApplyForce=true
    ApplyForceFalloff=false
    ApplyPosition=true
    ApplyPositionFalloff=false
    Module.LimitForceSpeed=false
   ) {enabled=false}
   &SolveForcesAndVelocity(Clamp Velocity=false, Limit Acceleration=false)
   &ScaleSpriteSize(
    Uniform Curve Scale=&MakeCustomFloatFromBool(
     Bool=$System.SpawnPistol
     True Float=0.65
     False Float=&MakeCustomFloatFromBool(Bool=$System.SpawnShotgun, True Float=1.29, False Float=1)
    )
    Uniform Curve Sprite Scale=DataInterface() {propertyValues {Curve {keys [
        {arriveTangent=4.1785, interpMode=RCIM_Cubic, leaveTangent=4.1785, tangentMode=RCTM_User, time=0, value=0.15}
        {arriveTangent=1.1179, interpMode=RCIM_Cubic, leaveTangent=1.1179, tangentMode=RCTM_User, time=0.2042, value=0.6228}
        {arriveTangent=0.0224, interpMode=RCIM_Cubic, leaveTangent=0.0224, tangentMode=RCTM_User, time=1, value=1}
       ]}}}
   )
   &ScaleColor(
    Scale Alpha=&MaskFloatBySpawnGroup(
     Masked Float=&FloatFromCurve(FloatCurve=DataInterface() {propertyValues {Curve {keys [
          {arriveTangent=-3.0638, interpMode=RCIM_Cubic, leaveTangent=-3.0638, tangentMode=RCTM_User, time=0, value=1}
          {tangentMode=RCTM_User, time=1, value=-1}
         ]}}})
     Passthrough Float=&FloatFromCurve(FloatCurve=DataInterface() {propertyValues {Curve {keys [<at 0.5 1 RCIM_Cubic>, {arriveTangent=-2.0383, interpMode=RCIM_Cubic, leaveTangent=-2.0383, tangentMode=RCTM_User, time=1, value=0}]}}})
     Spawn Group Mask=1
    )
    Scale RGB=&MaskVectorBySpawnGroup(Masked Vector=<1.2 1.2 1.2>, Passthrough Vector=<0 0 0>, Spawn Group Mask=1)
    ScaleA=true
    ScaleRGB=true
   )
  ]
  rendererClasses [&NiagaraSpriteRendererProperties, &NiagaraMeshRendererProperties]
  renderers [
   {rendererClass=&NiagaraSpriteRendererProperties}
   {rendererClass=&NiagaraMeshRendererProperties, rendererIndex=1}
  ]
 }
 Ribbon_Trail {
  emitterUpdateScript [
   &EmitterState()
   &SpawnPerUnit(
    Module.Use Max Movement Threshold=true
    Module.Use Movement Tolerance=true
    Movement Tolerance=10
    Spawn Spacing=25
    Use Spawn Probability=false
   ) {enabled=false}
   &SpawnBurst_Instantaneous(
    Spawn Probability=&MakeFloatFromInt(Int=&SelectIntBySurface(
      Default:NiagaraInt32=1
      Surface:EPhysicalSurface=&self:ScratchDynamicInput_01(PhysicalSurface:NiagaraInt32=$User.PhysicalSurface)
      Num Surfaces:NiagaraInt32=2
      Surface 0:EPhysicalSurface=SurfaceType8
      Value For Surface 0:NiagaraInt32=0
      Surface 1:EPhysicalSurface=SurfaceType30
      Value For Surface 1:NiagaraInt32=0
      Surface 2:EPhysicalSurface=?
      Value For Surface 2:NiagaraInt32=?
      Surface 3:EPhysicalSurface=?
      Value For Surface 3:NiagaraInt32=?
     ))
    Use Spawn Probability=true
    UseLoopCountLimit=false
   )
  ]
  particleSpawnScript [
   &InitializeParticle(
    Color Mode=::NewEnumerator0
    Position Offset=<-0.05 0 0>
    Ribbon Facing Mode=::NewEnumerator1
    Ribbon Facing Vector:Vector3f=$Engine.Owner.SystemXAxis
    Ribbon Width=50
    Ribbon Width Mode=::NewEnumerator1
    Sprite Size Mode=::NewEnumerator3
    Uniform Sprite Size=20
    UsePositionOffset=false
    Write Material Random=true
   )
   /SetVariables(Particles.RibbonID=/Niagara/DynamicInputs/NiagaraID/MakeNiagaraIDFromInt(Index:NiagaraInt32=$User.RibbonID, Acquire Tag:NiagaraInt32=0)) {isSetParameters=true}
  ]
  particleUpdateScript [
   &ParticleState()
   &DynamicMaterialParameters(
    Index 0 Param 2=&FloatFromCurve(FloatCurve=DataInterface() {propertyValues {Curve {keys [{tangentMode=RCTM_User, time=0, value=1}, {tangentMode=RCTM_User, time=1, value=0}]}}})
    Module.Param0WriteEnabled=false
    Module.Param1WriteEnabled=true
    Module.Param2WriteEnabled=false
    Module.Param3WriteEnabled=false
   ) {enabled=false}
   &ScaleColor(
    Scale Alpha=&FloatFromCurve(FloatCurve=DataInterface() {propertyValues {Curve {keys [{interpMode=RCIM_Cubic, tangentMode=RCTM_User, time=0.66, value=1}, {interpMode=RCIM_Cubic, leaveTangent=-7.011, time=1, value=0}]}}})
    Scale RGB=&VectorFromCurve(
     VectorCurve=DataInterface() {propertyValues {
       XCurve {keys [
         {arriveTangent=-4.5664, interpMode=RCIM_Cubic, leaveTangent=-4.5664, tangentMode=RCTM_User, time=0, value=0.5}
         {arriveTangent=-1.25, interpMode=RCIM_Cubic, leaveTangent=-1.25, time=0.2, value=0.1}
         <at 0.4 0 RCIM_Cubic>
        ]}
       YCurve {keys [<at 0 0.1>, {interpMode=RCIM_Cubic, tangentMode=RCTM_User, time=0.2, value=0}, <at 0.4 0 RCIM_Cubic>]}
       ZCurve {keys [<at 0 1>, {interpMode=RCIM_Cubic, tangentMode=RCTM_User, time=0.2, value=0}, <at 0.4 0 RCIM_Cubic>]}
      }}
     Scale Curve=<7 7 7>
    )
    ScaleA=true
    ScaleRGB=true
   )
  ]
  rendererClasses [&NiagaraRibbonRendererProperties]
  renderers [{rendererClass=&NiagaraRibbonRendererProperties}]
 }
 SkidSharp {
  emitterUpdateScript [
   &EmitterState()
   &SpawnBurst_Instantaneous(Use Spawn Probability=false, UseLoopCountLimit=false)
  ]
  particleSpawnScript [
   &InitializeParticle(
    Color=<rgba 5 5 5 1>
    Lifetime=0.133
    Mesh Scale=/Niagara/DynamicInputs/CascadeConversion/SplitVector(
     Double Component Vector Channels:ECascadeNiagaraTwoVectorChannels=::NewEnumerator1
     Double Component:NiagaraFloat=&RandomRangeFloat(Minimum=4, Maximum=5, Evaluation Type=::NewEnumerator1)
     Single Component:NiagaraFloat=&RandomRangeFloat(Minimum=1, Maximum=1.5, Evaluation Type=::NewEnumerator1)
    )
    Mesh Scale Mode=::NewEnumerator4
    UsePositionOffset=false
    Write Material Random=true
   ) {module=InitializeParticle001}
   &DynamicMaterialParameters(
    Index 0 Param 3=&RandomRangeFloat(Minimum=1, Maximum=1.5, Evaluation Type=::NewEnumerator1)
    Module.Param0WriteEnabled=false
    Module.Param1WriteEnabled=false
    Module.Param2WriteEnabled=true
    Module.Param3WriteEnabled=false
   ) {module=DynamicMaterialParameters001}
   &AddVelocityInCone(
    Cone Angle=20
    Module.Use Velocity Falloff On Cone Axis=false
    Velocity Strength=&RandomRangeFloat(Minimum=5, Maximum=15)
   )
   &InitialMeshOrientation(
    Rotation=&RandomRangeVector(Minimum=<-1 0 0>, Maximum=<1 0 0>)
    Use Orientation Vector=true
    Use Rotation Vector=true
   )
  ]
  particleUpdateScript [
   &ParticleState()
   &ScaleMeshSize(Scale Factor=&VectorFromFloat(Value=&Lerp_Float(A=0.5, B=0.8, Alpha=$NPC.NPC_Audio_Reactive_Parameters.Amplitude_Music))) {module=ScaleMeshSize002}
   &Drag(Drag=4, Use Linear Drag=true, Use Rotational Drag=false)
   &SolveForcesAndVelocity(Clamp Velocity=false, Limit Acceleration=false)
   &ScaleColor(
    Scale Alpha=&FloatFromCurve(FloatCurve=DataInterface() {propertyValues {Curve {keys [
         {arriveTangent=-1.0981, interpMode=RCIM_Cubic, leaveTangent=-1.0981, tangentMode=RCTM_User, time=0.33, value=1}
         {arriveTangent=-1.6087, interpMode=RCIM_Cubic, leaveTangent=-1.6087, tangentMode=RCTM_User, time=1, value=0}
        ]}}})
    Scale RGB=&VectorFromFloat(Value=&FloatFromCurve(FloatCurve=DataInterface() {propertyValues {Curve {keys [<at 0 5>,<at 1 0>]}}}))
    ScaleA=true
    ScaleRGB=true
   )
   &DynamicMaterialParameters(
    Index 0 Param 1=&FloatFromCurve(FloatCurve=DataInterface() {propertyValues {Curve {keys [<at 0.1 0.1 RCIM_Cubic>,<at 0.8 1>]}}})
    Index 0 Param 2=2
    Index 0 Param 4=&FloatFromCurve(FloatCurve=DataInterface() {propertyValues {Curve {keys [<at 0 1>,<at 1 3>]}}})
    Module.Param0WriteEnabled=true
    Module.Param1WriteEnabled=true
    Module.Param2WriteEnabled=false
    Module.Param3WriteEnabled=true
   )
   &CameraOffset(Camera Offset Amount=16)
   &ScaleMeshSize(Scale Factor=&VectorFromCurve(VectorCurve=DataInterface() {propertyValues {
       XCurve {keys [
         {arriveTangent=2.0829, interpMode=RCIM_Cubic, leaveTangent=2.0829, tangentMode=RCTM_User, time=0, value=0}
         {arriveTangent=1, interpMode=RCIM_Cubic, leaveTangent=1, time=0.5, value=0.74}
         <at 1 1 RCIM_Cubic>
        ]}
       YCurve {keys [
         <at 0 0.2 RCIM_Cubic>
         <at 0.5 1 RCIM_Cubic>
         {arriveTangent=-2.5169, interpMode=RCIM_Cubic, leaveTangent=-2.5169, tangentMode=RCTM_User, time=1, value=0.2}
        ]}
       ZCurve {keys [
         <at 0 0.2 RCIM_Cubic>
         <at 0.5 1 RCIM_Cubic>
         {arriveTangent=-3.2433, interpMode=RCIM_Cubic, leaveTangent=-3.2433, tangentMode=RCTM_User, time=1, value=0.2}
        ]}
      }})) {enabled=false}
   &ScaleMeshSize(Scale Factor=&VectorFromFloat(Value=&FloatFromCurve(FloatCurve=DataInterface() {propertyValues {Curve {keys [
          {arriveTangent=2.4702, interpMode=RCIM_Cubic, leaveTangent=2.4702, tangentMode=RCTM_User, time=0, value=0}
          {arriveTangent=0.9995, interpMode=RCIM_Cubic, leaveTangent=0.9995, tangentMode=RCTM_User, time=0.4, value=0.75}
          {arriveTangent=-0.0405, interpMode=RCIM_Cubic, leaveTangent=-0.0405, tangentMode=RCTM_User, time=1, value=1}
         ]}}}))) {module=ScaleMeshSize001}
  ]
  rendererClasses [&NiagaraMeshRendererProperties]
  renderers [{rendererClass=&NiagaraMeshRendererProperties, rendererIndex=0}]
 }
 SparksLazy001 {
  emitterUpdateScript [
   &EmitterState()
   &SpawnBurst_Instantaneous(
    Spawn Count=&RandomRangeInt(Minimum=2, Maximum=3)
    Spawn Probability=&MakeFloatFromInt(Int=&SelectIntBySurface(
      Default:NiagaraInt32=1
      Surface:EPhysicalSurface=&self:ScratchDynamicInput_01(PhysicalSurface:NiagaraInt32=$User.PhysicalSurface)
      Num Surfaces:NiagaraInt32=2
      Surface 0:EPhysicalSurface=SurfaceType8
      Value For Surface 0:NiagaraInt32=0
      Surface 1:EPhysicalSurface=SurfaceType30
      Value For Surface 1:NiagaraInt32=0
      Surface 2:EPhysicalSurface=?
      Value For Surface 2:NiagaraInt32=?
      Surface 3:EPhysicalSurface=?
      Value For Surface 3:NiagaraInt32=?
     ))
    Use Spawn Probability=true
    UseLoopCountLimit=false
   )
  ]
  particleSpawnScript [
   &FN_SetSpawnGroup(Number of Spawn Groups=2)
   &InitializeParticle(
    AdjustAlpha=false
    AdjustHue=false
    AdjustSaturation=false
    AdjustValue=true
    Color=<rgba 25 25 25 1>
    Color Mode=::NewEnumerator3
    Lifetime Max=0.2
    Lifetime Min=0.1
    Lifetime Mode=::NewEnumerator1
    Sprite Size Max=<6 17>
    Sprite Size Min=<5 7>
    Sprite Size Mode=::NewEnumerator2
    UsePositionOffset=false
    Value Range:Vector2f=<0.5 5>
    Write Material Random=true
   ) {module=InitializeParticle001}
   /SetVariables(Particles.VisibilityTag=$Particles.SpawnGroup) {enabled=false, isSetParameters=true}
   &SphereLocation(
    Hemisphere X=true
    MaskSpawnGroup=false
    Override Seed=false
    Sphere Radius=5
    Surface Only=true
    Surface Only Band Thickness=0.1
    UseNonUniformScale=false
    UseOffset=false
    UseSphereOrientation=false
   ) {module=SphereLocation001}
   &AddVelocity(Velocity=&MaskVectorBySpawnGroup(
     Masked Vector=&RandomRangeVector(Minimum=<250 0 0>, Maximum=<1500 1 1>)
     Passthrough Vector=&RandomRangeVector(Minimum=<50 0 0>, Maximum=<1000 1 1>)
    )) {enabled=false}
   &AddVelocity(
    Coordinate Space=::World
    Velocity=&Multiply_VectorByFloat(
     Vector=$Output.SphereLocation001.SphereNormal
     Float=&RandomRangeFloat(Minimum=400, Maximum=800, Evaluation Type=::NewEnumerator1)
    )
   ) {module=AddVelocity001}
   /SetVariables(Particles.Renderer2_Color:LinearColor=&RandomRangeLinearColor(Minimum=<rgba 0.0191 0.0156 0.0147 1>, Maximum=<rgba 0.1267 0.1035 0.0975 1>)) {isSetParameters=true}
  ]
  particleUpdateScript [
   &ParticleState()
   /SetVariables(Particles.VisibilityTag=&SpawnGroup()) {isSetParameters=true}
   &AccelerationForce(Acceleration=&RandomRangeVector(Minimum=<0 0 -1000>, Maximum=<0 0 1000>))
   &Drag(Drag=&RandomRangeFloat(Minimum=1, Maximum=4), Use Linear Drag=true, Use Rotational Drag=false)
   &SolveForcesAndVelocity(Clamp Velocity=false, Limit Acceleration=false)
   &ScaleSpriteSizeBySpeed(
    Max Scale Factor=<0.75 6>
    Min Scale Factor=<1 1.5>
    Source Velocity:Vector3f=&Multiply_Vector(A=<0.01 0.01 1>, B=$Particles.Velocity)
   )
   &ScaleColor(
    Scale Alpha=&FloatFromCurve(FloatCurve=DataInterface() {propertyValues {Curve {keys [<at 0 3>,<at 0.2 1>]}}})
    Scale RGB=&VectorFromFloat(Value=&Lerp_Float(
      A=0.25
      B=1
      Alpha=&MakeFloatFromLinearColor(LinearColor=$NPC.NPC_Audio_Reactive_Parameters.Spectrum_MusicHistory1, Channel:ENiagaraLinearColor_Channels=::NewEnumerator1)
     ))
    ScaleA=true
    ScaleRGB=true
   )
   &ScaleSpriteSize(Scale Factor=&Vector2DFromFloat(Value=&FloatFromCurve(FloatCurve=DataInterface() {propertyValues {Curve {keys [
          {tangentMode=RCTM_User, time=0, value=1}
          {arriveTangent=-0.0405, interpMode=RCIM_Cubic, leaveTangent=-0.0405, tangentMode=RCTM_User, time=0.7009, value=1}
          {arriveTangent=-8.0089, interpMode=RCIM_Cubic, leaveTangent=-8.0089, tangentMode=RCTM_User, time=1, value=0}
         ]}}})))
   &DynamicMaterialParameters(
    Index 0 Param 1=0.8
    Index 1 Param 1=150
    Index 1 Param 4=1
    Module.Param0WriteEnabled=true
    Module.Param0WriteEnabled1=true
    Module.Param1WriteEnabled=true
    Module.Param1WriteEnabled1=true
    Module.Param2WriteEnabled=true
    Module.Param2WriteEnabled1=true
    Module.Param3WriteEnabled=false
    Module.Param3WriteEnabled1=true
    Write Parameter Index 1=true
   )
   /SetVariables(
    Particles.Linear Color=&MultiplyLinearColor(Linear Color=$Particles.Color, Scale Factor=<0.02 0.02 0.02 0.0001>)
    Particles.Vector:Vector3f=&Multiply_Vector(A=$Particles.Velocity, B=<1 1 1>)
    Particles.Vector 2D:Vector2f=&Multiply_Vector2D(A=<2 5>, B=$Particles.SpriteSize)
   ) {isSetParameters=true}
  ]
  rendererClasses [&NiagaraSpriteRendererProperties]
  renderers [
   {rendererClass=&NiagaraSpriteRendererProperties}
   {rendererClass=&NiagaraSpriteRendererProperties, rendererIndex=1}
   {rendererClass=&NiagaraSpriteRendererProperties, rendererIndex=2}
  ]
 }
 SparksLazy002 {
  simTarget=GPUComputeSim
  emitterUpdateScript [
   &EmitterState()
   &SpawnRate(Spawn Probability=0.65, SpawnRate=50, Use Spawn Probability=true)
   &SpawnBurst_Instantaneous(
    Spawn Count=&RandomRangeInt(Minimum=6, Maximum=15)
    Spawn Probability=&MakeFloatFromInt(Int=&SelectIntBySurface(
      Default:NiagaraInt32=1
      Surface:EPhysicalSurface=&self:ScratchDynamicInput_01(PhysicalSurface:NiagaraInt32=$User.PhysicalSurface)
      Num Surfaces:NiagaraInt32=2
      Surface 0:EPhysicalSurface=SurfaceType8
      Value For Surface 0:NiagaraInt32=0
      Surface 1:EPhysicalSurface=SurfaceType30
      Value For Surface 1:NiagaraInt32=0
      Surface 2:EPhysicalSurface=?
      Value For Surface 2:NiagaraInt32=?
      Surface 3:EPhysicalSurface=?
      Value For Surface 3:NiagaraInt32=?
     ))
    Use Spawn Probability=true
    UseLoopCountLimit=false
   )
  ]
  particleSpawnScript [
   &InitializeParticle(
    AdjustAlpha=false
    AdjustHue=true
    AdjustSaturation=false
    AdjustValue=true
    Color=<rgba 25 25 25 1>
    Color Mode=::NewEnumerator3
    Hue Shift Range:Vector2f=<-0.5 0.5>
    Lifetime Max=14
    Lifetime Min=2
    Lifetime Mode=::NewEnumerator1
    Sprite Size Mode=::NewEnumerator1
    UsePositionOffset=false
    Value Range:Vector2f=<0.5 5>
    Write Material Random=true
   ) {module=InitializeParticle001}
   &SphereLocation(
    Hemisphere X=true
    MaskSpawnGroup=false
    Override Seed=false
    Sphere Radius=5
    Surface Only=false
    UseNonUniformScale=false
    UseOffset=false
    UseSphereOrientation=false
   ) {module=SphereLocation001}
   &AddVelocity(Velocity=&Multiply_VectorByFloat(
     Vector=&NormalizeVector(
      Vector To Normalize=&Add_Vector(
       A=<0.6 0 0>
       B=&TransformVector(
        Vector=&NormalizeVector(
         Vector To Normalize=/Niagara/DynamicInputs/Vectors/NegateVector(Vector:Vector3f=$Engine.Owner.Velocity)
         Safe Normalize=true
         Safe Fallback Threshold=0
         Safe Fallback Vector=<0 0 0.001>
        )
        Source Space=::World
        Destination Space=::Local
       )
      )
      Safe Normalize=true
      Safe Fallback Threshold=0
      Safe Fallback Vector=<0 0 0.001>
     )
     Float=&RandomRangeFloat(Minimum=50, Maximum=300)
    ))
   &AddVelocity(
    Coordinate Space=::Simulation
    Velocity=&Multiply_VectorByFloat(
     Vector=$Output.SphereLocation001.SphereNormal
     Float=&RandomRangeFloat(Minimum=50, Maximum=500, Evaluation Type=::NewEnumerator1)
    )
   ) {module=AddVelocity001}
  ]
  particleUpdateScript [
   &ParticleState()
   /SetVariables(Particles.VisibilityTag=&SpawnGroup()) {isSetParameters=true}
   &AccelerationForce(Acceleration=<0 0 -900>)
   &GravityForce()
   &Drag(Use Linear Drag=true, Use Rotational Drag=false)
   &Collision(
    Age Colliding Particles=false
    Kill Offscreen Particles=true
    Module.Control Rest State Via Surface Normal=false
    Module.EnableMaxCollisionCount=false
    Module.Randomize Collision Normal Vector=false
    Restitution=0.1
   )
   &SolveForcesAndVelocity(Clamp Velocity=false, Limit Acceleration=false)
   &ScaleSpriteSizeBySpeed(
    Max Scale Factor=<0.66 8>
    Source Velocity:Vector3f=&Multiply_Vector(A=<0.01 0.01 1>, B=$Particles.Velocity)
    Velocity Threshold=600
   )
   &ScaleColor(
    Scale Alpha=&FloatFromCurve(FloatCurve=DataInterface() {propertyValues {Curve {keys [
         {tangentMode=RCTM_User, time=0, value=1}
         {arriveTangent=-0.0405, interpMode=RCIM_Cubic, leaveTangent=-0.0405, tangentMode=RCTM_User, time=0.7009, value=1}
         {arriveTangent=-8.0089, interpMode=RCIM_Cubic, leaveTangent=-8.0089, tangentMode=RCTM_User, time=1, value=0}
        ]}}})
    Scale RGB=&MakeVectorFromLinearColorRGB(Color=&ColorFromCurve(ColorCurve=DataInterface() {propertyValues {
        AlphaCurve {keys [<at 0 1 RCIM_Cubic>,<at 1 1 RCIM_Cubic>]}
        BlueCurve {keys [{arriveTangent=-1.3944, interpMode=RCIM_Cubic, time=0, value=3.2093}, <at 0.5237 0.66>, <at 0.7297 0.66>, <at 0.9971 0.0156>]}
        GreenCurve {keys [{arriveTangent=-11.4062, interpMode=RCIM_Cubic, time=0, value=0}, <at 0.5237 0>, <at 0.7297 0>, <at 0.9971 1.5>]}
        RedCurve {keys [{arriveTangent=-20.6532, interpMode=RCIM_Cubic, time=0, value=4}, <at 0.5237 0.0521>, <at 0.7297 0.0521>, <at 0.9971 0>]}
       }}))
    ScaleA=true
    ScaleRGB=true
   ) {enabled=false}
   &ScaleColor(
    Scale Alpha=&FloatFromCurve(FloatCurve=DataInterface() {propertyValues {Curve {keys [{arriveTangent=-12.1355, interpMode=RCIM_Cubic, leaveTangent=-12.1355, tangentMode=RCTM_User, time=0, value=3}, <at 0.2 1>]}}})
    Scale RGB=&VectorFromFloat(Value=&Lerp_Float(
      A=0.1
      B=1
      Alpha=&FloatFromCurve(
       FloatCurve=DataInterface() {propertyValues {Curve {keys [
           {arriveTangent=-0.1138, interpMode=RCIM_Cubic, leaveTangent=-0.1138, tangentMode=RCTM_User, time=0, value=0}
           {arriveTangent=-3.8042, interpMode=RCIM_Cubic, leaveTangent=-3.8042, tangentMode=RCTM_User, time=0.9, value=0.825}
           {arriveTangent=-0.0405, interpMode=RCIM_Cubic, leaveTangent=-0.0405, tangentMode=RCTM_User, time=1, value=0}
          ]}}}
       CurveIndex=$NPC.Radiant_Music_NPC.MusicBeatProgress
      )
     ))
    ScaleA=true
    ScaleRGB=true
   ) {module=ScaleColor001}
   &ScaleSpriteSize(Scale Factor=&Vector2DFromFloat(Value=&FloatFromCurve(
      FloatCurve=DataInterface() {propertyValues {Curve {keys [
          {tangentMode=RCTM_User, time=0, value=1}
          {arriveTangent=-0.0405, interpMode=RCIM_Cubic, leaveTangent=-0.0405, tangentMode=RCTM_User, time=0.7009, value=1}
          {arriveTangent=-8.0089, interpMode=RCIM_Cubic, leaveTangent=-8.0089, tangentMode=RCTM_User, time=1, value=0}
         ]}}}
      Scale Curve=0.75
     )))
   &ScaleSpriteSize(
    Uniform Curve Scale=&Multiply_Float(
     A=&Multiply_Float(
      A=1
      B=&Waveform(
       "Waveform [1]"=::NewEnumerator0
       "[1] Phase Scale"=&RandomRangeFloat(Minimum=3, Maximum=10)
       "[1] Frequency"=&RandomRangeFloat(Minimum=0.2, Maximum=0.5)
       "[1] Amplitude Min/Max"=<0 1>
      )
     )
     B=&FN_LOD_Distance_Remap(Start Distance=500, End Distance=5000, Minimum Output=0.45, Maximum Output=1, Clamp Output=true)
    )
    Uniform Curve Sprite Scale=DataInterface() {propertyValues {Curve {keys [
        {arriveTangent=2.9868, interpMode=RCIM_Cubic, leaveTangent=2.9868, tangentMode=RCTM_User, time=0, value=1.65}
        <at 0.4 2.3>
        {arriveTangent=-0.0405, interpMode=RCIM_Cubic, leaveTangent=-0.0405, tangentMode=RCTM_User, time=1, value=1.4}
       ]}}}
   ) {enabled=false, module=ScaleSpriteSize001}
   &DynamicMaterialParameters(
    Index 0 Param 1=0.8
    Index 0 Param 2=1
    Index 0 Param 3=1
    Index 1 Param 1=150
    Index 1 Param 4=1
    Module.Param0WriteEnabled=true
    Module.Param0WriteEnabled1=true
    Module.Param1WriteEnabled=true
    Module.Param1WriteEnabled1=true
    Module.Param2WriteEnabled=true
    Module.Param2WriteEnabled1=true
    Module.Param3WriteEnabled=false
    Module.Param3WriteEnabled1=true
    Write Parameter Index 1=true
   )
   /SetVariables(
    Particles.Linear Color=&MultiplyLinearColor(Linear Color=$Particles.Color, Scale Factor=<0.02 0.02 0.02 0.0001>)
    Particles.Vector:Vector3f=&Multiply_Vector(A=$Particles.Velocity, B=<1 1 1>)
    Particles.Vector 2D:Vector2f=&Multiply_Vector2D(A=<2 5>, B=$Particles.SpriteSize)
   ) {isSetParameters=true}
  ]
  rendererClasses [&NiagaraSpriteRendererProperties]
  renderers [
   {rendererClass=&NiagaraSpriteRendererProperties}
   {rendererClass=&NiagaraSpriteRendererProperties, rendererIndex=1}
   {rendererClass=&NiagaraSpriteRendererProperties, rendererIndex=2}
  ]
 }
 Water_Drops {
  bEnabled=false
  simTarget=GPUComputeSim
  emitterUpdateScript [
   &EmitterState(
    Enable Distance Culling=true
    Max Distance Response=::KillAfterParticlesFinish
    Module.bUseMaxDistance=true
    Module.bUseMinDistance=false
    Scalability Mode=::NewEnumerator1
   )
   &SpawnBurst_Instantaneous(
    Spawn Count=&UniformRangedInt(Minimum=15, Maximum=30)
    Spawn Probability=&MakeFloatFromInt(Int=&SelectIntBySurface(
      Default:NiagaraInt32=0
      Surface:EPhysicalSurface=&self:ScratchDynamicInput_01(PhysicalSurface:NiagaraInt32=$User.PhysicalSurface)
      Num Surfaces:NiagaraInt32=1
      Surface 0:EPhysicalSurface=SurfaceType8
      Value For Surface 0:NiagaraInt32=1
      Surface 1:EPhysicalSurface=?
      Value For Surface 1:NiagaraInt32=?
      Surface 2:EPhysicalSurface=?
      Value For Surface 2:NiagaraInt32=?
      Surface 3:EPhysicalSurface=?
      Value For Surface 3:NiagaraInt32=?
     ))
    Use Spawn Probability=true
   )
  ]
  particleSpawnScript [
   /Niagara/Modules/Spawn/Initialization/InitializeParticle(
    Lifetime=&UniformRangedFloat(Minimum=0.3, Maximum=1)
    Module.Write Color=true
    Module.Write Lifetime=true
    Module.Write Mass=true
    Module.Write Material Random=true
    Module.Write Position=false
    Module.Write Scale=false
    Module.Write SpriteRotation=false
    Module.Write SpriteSize=true
    Position=/Niagara/Functions/Localspace/SimulationPosition()
    Sprite Size=&UniformRangedVector2D(Minimum=<3 4>, Maximum=<7 8>)
   )
   &AddVelocityInCone(
    Cone Angle=90
    Module.Use Velocity Falloff On Cone Axis=false
    Velocity Strength=&UniformRangedFloat(Minimum=100, Maximum=400)
   )
   &FN_PerParticleLOD(Enable SpriteSize Distance Scaling=true, Sprite Size Scale Far=<1.5 1.5>)
  ]
  particleUpdateScript [
   &ParticleState()
   &AccelerationForce(Acceleration=<0 0 -980>)
   &Drag(Drag=0.1, Use Linear Drag=true, Use Rotational Drag=false) {enabled=false}
   &SolveForcesAndVelocity(Clamp Velocity=false, Limit Acceleration=false)
   &ScaleSpriteSize(Scale Factor=&Vector2DFromFloat(Value=&OneMinusFloat(Float=$Particles.NormalizedAge)))
   &ScaleSpriteSizeBySpeed(Max Scale Factor=<1 2>, Velocity Threshold=500)
   /Niagara/Modules/Update/SubUV/SubUVAnimation(Random SubUV Lookup Index:NiagaraFloat=$Particles.Initial.MaterialRandom, SubUV Animation Mode=::NewEnumerator1) {enabled=false}
   &Color(Color=&MakeLinearColorFromVectorAndFloat(
     "Vector (RGB)"=&MakeVectorFromLinearColorRGB(Color=<rgba 1 0 0.8944 1>)
     "Float (Alpha)"=&FloatFromCurve(FloatCurve=DataInterface() {propertyValues {Curve {keys [<at 0.4 1.0113 RCIM_Cubic>,<at 1.0011 0.002>]}}})
    ))
  ]
  rendererClasses [&NiagaraSpriteRendererProperties]
  renderers [{rendererClass=&NiagaraSpriteRendererProperties}]
 }
}
summary {
 systemName=NS_Assault_Radient_ARG_Impact_Looping
 emitters [
  {emitterName=SparksLazy001, rendererClasses [&NiagaraSpriteRendererProperties]}
  {emitterName=SparksLazy002, simTarget=GPUComputeSim, rendererClasses [&NiagaraSpriteRendererProperties]}
  {emitterName=SkidSharp, rendererClasses [&NiagaraMeshRendererProperties]}
  {emitterName=Ribbon_Trail, rendererClasses [&NiagaraRibbonRendererProperties]}
  {emitterName=Corona, rendererClasses [&NiagaraSpriteRendererProperties, &NiagaraLightRendererProperties]}
  {bEnabled=false, emitterName=Water_Drops, simTarget=GPUComputeSim, rendererClasses [&NiagaraSpriteRendererProperties]}
  {bEnabled=false, emitterName=Mud_DropsCollision, rendererClasses [&NiagaraSpriteRendererProperties]}
  {bEnabled=false, emitterName=FN_Weapons_Debris_Sprite_SubUV_BETA003, rendererClasses [&NiagaraSpriteRendererProperties]}
  {
   bEnabled=false
   emitterName=NE_Parent_Mud_Splash
   rendererClasses [&NiagaraSpriteRendererProperties, &NiagaraMeshRendererProperties]
  }
  {emitterName=Digital_Bits002, simTarget=GPUComputeSim, rendererClasses [&NiagaraSpriteRendererProperties]}
 ]
 userVariables [
  User.BeamEnd_0 {type=NiagaraPosition, description="", defaultValue {}}
  User.BeamStart_0_0 {type=Vector3f, description="", defaultValue {struct=Vector3f, value=<1000 0 1000>}}
  User.bIsMudImpact {type=NiagaraBool}
  User.bIsWaterImpact {type=NiagaraBool}
  User.PhysicalSurface {type=NiagaraInt32}
  User.RibbonID {type=NiagaraInt32}
  User.WaterBodyIndex {type=NiagaraFloat, defaultValue=0, description=""}
  User.WaterDepth {type=NiagaraFloat, defaultValue=0, description=""}
  User.WaveOffset {type=Vector3f, description="", defaultValue {struct=Vector3f, value=<0 0 0>}}
 ]
}
