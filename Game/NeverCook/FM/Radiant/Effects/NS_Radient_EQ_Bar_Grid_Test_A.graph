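# Niagara test system for music-reactive EQ bars, spawning a 5x5 jittered grid of looping mesh particles with randomized scale.
# Two stacked scale modules drive bar height from parameter collections: one uses a spectrum color channel, the other a pulse curve sampled randomly between music amplitude and beat progress.
# Color-curve scaling is disabled; dynamic material parameters and a user ScaleAlpha control feed appearance across two mesh renderers.
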NiagaraSystem /Game/NeverCook/FM/Radiant/Effects/NS_Radient_EQ_Bar_Grid_Test_A
dependencies {
 usedDataInterfaces [&NiagaraDataInterfaceArrayFloat3, &NiagaraDataInterfaceColorCurve, &NiagaraDataInterfaceCurve]
 usedDynamicInputs []
 usedModules [&SystemState]
}
emitters {EQ_Boxes_GridSpawn {
  emitterUpdateScript [
   &EmitterState(
    Inactive Response=::NewEnumerator1
    Life Cycle Mode=::NewEnumerator1
    Loop Behavior=::NewEnumerator1
    Loop Duration Mode=::NewEnumerator1
    UseLoopDelay=false
   )
   /Niagara/Modules/Spawn/Location/SpawnParticlesInGrid(X Count=5, Y Count=5, Z Count=1)
  ]
  particleSpawnScript [
   &PartitionParticles()
   &InitializeParticle(
    Color=<rgba 25 25 25 1>
    Lifetime Max=3
    Lifetime Mode=::NewEnumerator1
    Mesh Scale Max=<0.06 0.05 0.45>
    Mesh Scale Min=<0.03 0.03 0.3>
    Mesh Scale Mode=::NewEnumerator2
    UsePositionOffset=true
    Write Material Random=true
   ) {module=InitializeParticle001}
   &InitialMeshOrientation(
    Mesh Orientation Mode=::NewEnumerator2
    Rotation=&SelectVectorFromArray(
     Array Sampling Mode=::NewEnumerator0
     Direct Array Index=?
     Vector Selection Array=DataInterface() {propertyValues {FloatData [<0 0 0>, <0.5 0 0>]}}
     Random Seed=0
    )
    Use Rotation Vector=true
   ) {enabled=false}
   /Niagara/Modules/Spawn/Location/V2/GridLocation(
    Align To Grid Cell=true
    Dimensions Definition:ENiagaraGridPlacementType=::NewEnumerator1
    Module.Jitter Inside Grid Cells=true
    Module.Override Random Seed=false
    Offset:Vector3f=<0 1 0>
    XYZ Dimensions:Vector3f=&MakeVector(X=250, Y=250, Z=0.1)
   )
  ]
  particleUpdateScript [
   &SolveForcesAndVelocity(Clamp Velocity=false, Limit Acceleration=false) {enabled=false}
   &ScaleColor(
    Linear Color Curve=DataInterface() {propertyValues {
      AlphaCurve {keys [<at 0 1>,<at 0.4487 1>,<at 1 1>]}
      BlueCurve {keys [<at 0 0.7379>,<at 0.134 0.5457>,<at 0.2454 0.7157>,<at 0.3458 0.073>,<at 0.4572 0.8714>,<at 0.5613 1>,<at 0.6617 1>,<at 0.7694 0.073>,<at 0.8759 0.9042>]}
      GreenCurve {keys [<at 0 0>,<at 0.134 0>,<at 0.2454 0.0176>,<at 0.3458 1>,<at 0.4572 1>,<at 0.5613 0.8797>,<at 0.6617 0>,<at 0.7694 1>,<at 0.8759 0>]}
      RedCurve {keys [<at 0 0.7682>,<at 0.134 0.1779>,<at 0.2454 0.0782>,<at 0.3458 0>,<at 0.4572 0>,<at 0.5613 0>,<at 0.6617 0.0423>,<at 0.7694 0>,<at 0.8759 1>]}
     }}
    Scale Mode=::NewEnumerator2
   ) {enabled=false}
   &ParticleState(
    Kill Particles When Lifetime Has Elapsed=false
    Let Infinitely Lived Particles Die When Emitter Deactivates=true
    Loop Particles Lifetime=true
   )
   &ScaleMeshSize(Scale Factor=&MakeVector(
     X=1
     Y=1
     Z=&Lerp_Float(
      A=0
      B=&FloatFromCurve(FloatCurve=DataInterface() {propertyValues {Curve {keys [
           {interpMode=RCIM_Cubic, tangentMode=RCTM_User, time=0, value=1}
           {interpMode=RCIM_Cubic, tangentMode=RCTM_User, time=0.5, value=1.5}
           {tangentMode=RCTM_User, time=1, value=1}
          ]}}})
      Alpha=&MakeFloatFromVector(
       Vector=&MakeVectorFromLinearColorRGB(Color=$NPC.NPC_Audio_Reactive_Parameters.Spectrum_Music)
       Channel=::NewEnumerator1
      )
     )
    ))
   &ScaleMeshSize(Scale Factor=&MakeVector(
     X=1
     Y=1
     Z=&FloatFromCurve(
      FloatCurve=DataInterface() {propertyValues {Curve {keys [
          {arriveTangent=9.0725, interpMode=RCIM_Cubic, time=0, value=0}
          <at 0.1 1 RCIM_Cubic>
          {arriveTangent=-0.0088, interpMode=RCIM_Cubic, leaveTangent=-0.0088, tangentMode=RCTM_User, time=1, value=0}
         ]}}}
      CurveIndex=&RandomRangeFloat(Minimum=$NPC.NPC_Audio_Reactive_Parameters.Amplitude_Music, Maximum=$NPC.Radiant_Music_NPC.MusicBeatProgress)
     )
    )) {module=ScaleMeshSize005}
   &ScaleColor(
    Curve Index=&ReturnNormalizedExecIndex(Normalized Index Scale=2)
    Linear Color Curve=DataInterface() {propertyValues {
      AlphaCurve {keys [<at 0 1>,<at 0.4487 1>,<at 1 1>]}
      BlueCurve {keys [<at 0 0.7682>,<at 0.134 0.5457>,<at 0.2454 0.7157>,<at 0.3458 0.073>,<at 0.4572 0.8714>,<at 0.5613 1>,<at 0.6617 1>,<at 0.7694 0.073>,<at 0.8759 1>]}
      GreenCurve {keys [<at 0 0.0453>,<at 0.134 0>,<at 0.2454 0.0176>,<at 0.3458 1>,<at 0.4572 1>,<at 0.5613 0.8797>,<at 0.6617 0>,<at 0.7694 1>,<at 0.8759 0>]}
      RedCurve {keys [<at 0 0>,<at 0.134 0.1779>,<at 0.2454 0.0782>,<at 0.3458 0>,<at 0.4572 0>,<at 0.5613 0>,<at 0.6617 0.0423>,<at 0.7694 0>,<at 0.8759 0.0663>]}
     }}
    Scale Mode=::NewEnumerator2
   ) {enabled=false, module=ScaleColor001}
   &DynamicMaterialParameters(
    Index 0 Param 1=0.95
    Index 0 Param 4=1
    Module.Index0Param0Write=true
    Module.Index0Param1Write=true
    Module.Index0Param2Write=true
    Module.Index0Param3Write=true
   )
   &ScaleColor(Scale Alpha=$User.ScaleAlpha, ScaleA=true, ScaleRGB=false) {module=ScaleColor003}
  ]
  rendererClasses [&NiagaraMeshRendererProperties]
  renderers [
   {rendererClass=&NiagaraMeshRendererProperties, rendererIndex=0}
   {rendererClass=&NiagaraMeshRendererProperties, rendererIndex=1}
  ]
 }}
summary {
 systemName=NS_Radient_EQ_Bar_Grid_Test_A
 emitters [{emitterName=EQ_Boxes_GridSpawn, rendererClasses [&NiagaraMeshRendererProperties]}]
 userVariables [User.ScaleAlpha {type=NiagaraFloat, defaultValue=1, description=""}]
}
