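# Niagara test system for a Radiant tube detail effect, pairing a thin tube mesh with an equalizer-like cluster of box meshes.
# BlueTubeThin burst-spawns a scaled mesh with constant material parameters, its placement and color modules disabled.
# EQ_Boxes_CylinderShape001 bursts 50 randomly sized meshes onto a rotated ring-style shape, cycling color through a multi-hue curve.
# Random ranges bounded by curves jitter their Z scale, while a curve pulses a dynamic material parameter.
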NiagaraSystem /Game/NeverCook/FM/Radiant/Effects/NS_Radient_TubeDetail_Test_A
dependencies {
 usedDataInterfaces [&NiagaraDataInterfaceArrayFloat3, &NiagaraDataInterfaceColorCurve, &NiagaraDataInterfaceCurve]
 usedDynamicInputs []
 usedModules [&SystemState]
}
emitters {
 BlueTubeThin {
  emitterUpdateScript [
   &EmitterState(
    Inactive Response=::NewEnumerator1
    Life Cycle Mode=::NewEnumerator1
    Loop Behavior=::NewEnumerator1
    Loop Duration Mode=::NewEnumerator1
    UseLoopDelay=false
   )
   &SpawnBurst_Instantaneous(Use Spawn Probability=false, UseLoopCountLimit=true)
   /Niagara/Modules/Spawn/Location/SpawnParticlesInGrid(Z Count=1) {enabled=false}
  ]
  particleSpawnScript [
   &PartitionParticles()
   &InitializeParticle(Mesh Scale Mode=::NewEnumerator4, UsePositionOffset=true, Write Material Random=true) {module=InitializeParticle001}
   &InitialMeshOrientation(
    Mesh Orientation Mode=::NewEnumerator2
    Rotation=&SelectVectorFromArray(
     Array Sampling Mode=::NewEnumerator0
     Direct Array Index=?
     Vector Selection Array=DataInterface() {propertyValues {FloatData [<0 0 0>, <0.5 0 0>]}}
     Random Seed=0
    )
    Use Rotation Vector=true
   )
   /Niagara/Modules/Spawn/Location/V2/GridLocation(
    Align To Grid Cell=true
    Dimensions Definition:ENiagaraGridPlacementType=::NewEnumerator1
    Module.Jitter Inside Grid Cells=true
    Module.Override Random Seed=false
    Offset:Vector3f=<0 1 0>
    XYZ Dimensions:Vector3f=&MakeVector(X=250, Y=250, Z=0.1)
   ) {enabled=false}
   &ShapeLocation(
    MaskSpawnGroup=false
    Non Uniform Scale=<1 1 0>
    Ring / Disc Distribution Mode:ENiagaraSphereDistributionMode=::NewEnumerator3
    Ring Radius=50
    Shape Primitive=::NewEnumerator4
    Uniform Spiral Falloff=0
    UseOwnerScale=true
   ) {enabled=false}
   &AddVelocity() {enabled=false}
  ]
  particleUpdateScript [
   &SolveForcesAndVelocity(Clamp Velocity=false, Limit Acceleration=false)
   &ScaleColor(
    Linear Color Curve=DataInterface() {propertyValues {
      AlphaCurve {keys [<at 0 1>,<at 0.4487 1>,<at 1 1>]}
      BlueCurve {keys [<at 0 0.7379>,<at 0.134 0.5457>,<at 0.2454 0.7157>,<at 0.3458 0.073>,<at 0.4572 0.8714>,<at 0.5613 1>,<at 0.6617 1>,<at 0.7694 0.073>,<at 0.8759 0.9042>]}
      GreenCurve {keys [<at 0 0>,<at 0.134 0>,<at 0.2454 0.0176>,<at 0.3458 1>,<at 0.4572 1>,<at 0.5613 0.8797>,<at 0.6617 0>,<at 0.7694 1>,<at 0.8759 0>]}
      RedCurve {keys [<at 0 0.7682>,<at 0.134 0.1779>,<at 0.2454 0.0782>,<at 0.3458 0>,<at 0.4572 0>,<at 0.5613 0>,<at 0.6617 0.0423>,<at 0.7694 0>,<at 0.8759 1>]}
     }}
    Scale Mode=::NewEnumerator2
   ) {enabled=false}
   &ParticleState(Kill Particles When Lifetime Has Elapsed=false, Let Infinitely Lived Particles Die When Emitter Deactivates=true)
   &ScaleMeshSize(Scale Factor=<0.45 0.45 1.5>)
   &ScaleColor(
    Curve Index=&ReturnNormalizedExecIndex(Normalized Index Scale=2)
    Linear Color Curve=DataInterface() {propertyValues {
      AlphaCurve {keys [<at 0 1>,<at 0.4487 1>,<at 1 1>]}
      BlueCurve {keys [<at 0 0.7682>,<at 0.134 0.5457>,<at 0.2454 0.7157>,<at 0.3458 0.073>,<at 0.4572 0.8714>,<at 0.5613 1>,<at 0.6617 1>,<at 0.7694 0.073>,<at 0.8759 1>]}
      GreenCurve {keys [<at 0 0.0453>,<at 0.134 0>,<at 0.2454 0.0176>,<at 0.3458 1>,<at 0.4572 1>,<at 0.5613 0.8797>,<at 0.6617 0>,<at 0.7694 1>,<at 0.8759 0>]}
      RedCurve {keys [<at 0 0>,<at 0.134 0.1779>,<at 0.2454 0.0782>,<at 0.3458 0>,<at 0.4572 0>,<at 0.5613 0>,<at 0.6617 0.0423>,<at 0.7694 0>,<at 0.8759 0.0663>]}
     }}
    Scale Mode=::NewEnumerator2
   ) {enabled=false, module=ScaleColor001}
   &DynamicMaterialParameters(
    Index 0 Param 1=0.5
    Index 0 Param 4=1
    Module.Index0Param0Write=true
    Module.Index0Param1Write=true
    Module.Index0Param2Write=true
    Module.Index0Param3Write=true
   )
  ]
  rendererClasses [&NiagaraMeshRendererProperties]
  renderers [
   {rendererClass=&NiagaraMeshRendererProperties, rendererIndex=0}
   {rendererClass=&NiagaraMeshRendererProperties, rendererIndex=1}
  ]
 }
 EQ_Boxes_CylinderShape001 {
  emitterUpdateScript [
   &EmitterState(Inactive Response=::NewEnumerator1, Life Cycle Mode=::NewEnumerator1, Loop Duration=1, UseLoopDelay=false)
   &SpawnBurst_Instantaneous(Spawn Count=50, Use Spawn Probability=false, UseLoopCountLimit=false)
   &SpawnParticlesInGrid(Z Count=1) {enabled=false}
  ]
  particleSpawnScript [
   &PartitionParticles()
   &InitializeParticle(
    Lifetime Max=3
    Lifetime Mode=::NewEnumerator1
    Mesh Scale Max=<0.06 0.05 0.5>
    Mesh Scale Min=<0.02 0.02 0.3>
    Mesh Scale Mode=::NewEnumerator2
    UsePositionOffset=true
    Write Material Random=true
   ) {module=InitializeParticle001}
   &InitialMeshOrientation(
    Mesh Orientation Mode=::NewEnumerator2
    Rotation=&SelectVectorFromArray(
     Array Sampling Mode=::NewEnumerator0
     Direct Array Index=?
     Vector Selection Array=DataInterface() {propertyValues {FloatData [<0 0 0>, <0.5 0 0>]}}
     Random Seed=0
    )
    Use Rotation Vector=true
   )
   &GridLocation(
    Align To Grid Cell=true
    Dimensions Definition:ENiagaraGridPlacementType=::NewEnumerator1
    Module.Jitter Inside Grid Cells=true
    Module.Override Random Seed=false
    Offset:Vector3f=<0 1 0>
    XYZ Dimensions:Vector3f=&MakeVector(X=250, Y=250, Z=0.1)
   ) {enabled=false}
   &ShapeLocation(
    MaskSpawnGroup=false
    Non Uniform Scale=<1 1 0>
    Ring / Disc Distribution Mode:ENiagaraSphereDistributionMode=::NewEnumerator3
    Ring Radius=50
    Rotation Axis=<0 1 0>
    Rotation Mode=::NewEnumerator1
    Shape Primitive=::NewEnumerator4
    Uniform Spiral Falloff=0
    UseOwnerScale=true
   )
  ]
  particleUpdateScript [
   &SolveForcesAndVelocity(Clamp Velocity=false, Limit Acceleration=false) {enabled=false}
   &ScaleColor(
    Linear Color Curve=DataInterface() {propertyValues {
      AlphaCurve {keys [<at 0 1>,<at 0.4487 1>,<at 1 1>]}
      BlueCurve {keys [<at 0 0.7379>,<at 0.134 0.5457>,<at 0.2454 0.7157>,<at 0.3458 0.073>,<at 0.4572 0.8714>,<at 0.5613 1>,<at 0.6617 1>,<at 0.7694 0.073>,<at 0.8759 0.9042>]}
      GreenCurve {keys [<at 0 0>,<at 0.134 0>,<at 0.2454 0.0176>,<at 0.3458 1>,<at 0.4572 1>,<at 0.5613 0.8797>,<at 0.6617 0>,<at 0.7694 1>,<at 0.8759 0>]}
      RedCurve {keys [<at 0 0.7682>,<at 0.134 0.1779>,<at 0.2454 0.0782>,<at 0.3458 0>,<at 0.4572 0>,<at 0.5613 0>,<at 0.6617 0.0423>,<at 0.7694 0>,<at 0.8759 1>]}
     }}
    Scale Mode=::NewEnumerator2
   )
   &ParticleState()
   &ScaleMeshSize(Scale Factor=&MakeVector(
     X=2
     Y=2
     Z=&RandomRangeFloat(
      Minimum=&FloatFromCurve(
       FloatCurve=DataInterface() {propertyValues {Curve {keys [
           {arriveTangent=0.5032, interpMode=RCIM_Cubic, leaveTangent=0.5032, tangentMode=RCTM_User, time=0, value=1}
           {arriveTangent=0.125, interpMode=RCIM_Cubic, leaveTangent=0.125, time=0.1, value=0.375}
           {arriveTangent=-2.5249, interpMode=RCIM_Cubic, leaveTangent=-2.5249, tangentMode=RCTM_User, time=0.2, value=1.025}
           {arriveTangent=1.5, interpMode=RCIM_Cubic, leaveTangent=1.5, time=0.3, value=0.175}
           <at 0.5 1.475 RCIM_Cubic>
           {arriveTangent=-2, interpMode=RCIM_Cubic, leaveTangent=-2, time=0.7, value=0.175}
           {arriveTangent=0.625, interpMode=RCIM_Cubic, leaveTangent=0.625, time=0.8, value=0.875}
           {arriveTangent=0.625, interpMode=RCIM_Cubic, leaveTangent=0.625, time=0.9, value=0.3}
           {arriveTangent=-0.0405, interpMode=RCIM_Cubic, leaveTangent=-0.0405, tangentMode=RCTM_User, time=1, value=1}
          ]}}}
       Scale Curve=0.5
      )
      Maximum=&FloatFromCurve(FloatCurve=DataInterface() {propertyValues {Curve {keys [
           {arriveTangent=0.5032, interpMode=RCIM_Cubic, leaveTangent=0.5032, tangentMode=RCTM_User, time=0, value=0}
           {arriveTangent=5.125, interpMode=RCIM_Cubic, leaveTangent=5.125, time=0.1, value=1.4}
           {arriveTangent=-2.5249, interpMode=RCIM_Cubic, leaveTangent=-2.5249, tangentMode=RCTM_User, time=0.2, value=1.025}
           {arriveTangent=-0.25, interpMode=RCIM_Cubic, leaveTangent=-0.25, time=0.4, value=1.3}
           {arriveTangent=-0.1, interpMode=RCIM_Cubic, leaveTangent=-0.1, time=0.8, value=0.875}
           {arriveTangent=-4.375, interpMode=RCIM_Cubic, leaveTangent=-4.375, time=0.9, value=1.25}
           {arriveTangent=-0.0405, interpMode=RCIM_Cubic, leaveTangent=-0.0405, tangentMode=RCTM_User, time=1, value=0}
          ]}}})
     )
    ))
   &ScaleColor(
    Curve Index=&ReturnNormalizedExecIndex(Normalized Index Scale=2)
    Linear Color Curve=DataInterface() {propertyValues {
      AlphaCurve {keys [<at 0 1>,<at 0.4487 1>,<at 1 1>]}
      BlueCurve {keys [<at 0 0.7682>,<at 0.134 0.5457>,<at 0.2454 0.7157>,<at 0.3458 0.073>,<at 0.4572 0.8714>,<at 0.5613 1>,<at 0.6617 1>,<at 0.7694 0.073>,<at 0.8759 1>]}
      GreenCurve {keys [<at 0 0.0453>,<at 0.134 0>,<at 0.2454 0.0176>,<at 0.3458 1>,<at 0.4572 1>,<at 0.5613 0.8797>,<at 0.6617 0>,<at 0.7694 1>,<at 0.8759 0>]}
      RedCurve {keys [<at 0 0>,<at 0.134 0.1779>,<at 0.2454 0.0782>,<at 0.3458 0>,<at 0.4572 0>,<at 0.5613 0>,<at 0.6617 0.0423>,<at 0.7694 0>,<at 0.8759 0.0663>]}
     }}
    Scale Mode=::NewEnumerator2
   ) {enabled=false, module=ScaleColor001}
   &DynamicMaterialParameters(
    Index 0 Param 1=0.5
    Index 0 Param 4=&FloatFromCurve(FloatCurve=DataInterface() {propertyValues {Curve {keys [
         {interpMode=RCIM_Cubic, tangentMode=RCTM_User, time=0, value=0}
         {interpMode=RCIM_Cubic, tangentMode=RCTM_User, time=0.5, value=1}
         {tangentMode=RCTM_User, time=1, value=0}
        ]}}})
    Module.Index0Param0Write=true
    Module.Index0Param1Write=true
    Module.Index0Param2Write=true
    Module.Index0Param3Write=true
   )
  ]
  rendererClasses [&NiagaraMeshRendererProperties]
  renderers [
   {rendererClass=&NiagaraMeshRendererProperties, rendererIndex=0}
   {rendererClass=&NiagaraMeshRendererProperties, rendererIndex=1}
  ]
 }
}
summary {
 systemName=NS_Radient_TubeDetail_Test_A
 emitters [
  {emitterName=BlueTubeThin, rendererClasses [&NiagaraMeshRendererProperties]}
  {emitterName=EQ_Boxes_CylinderShape001, rendererClasses [&NiagaraMeshRendererProperties]}
 ]
 userVariables []
}
