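# LS_Radiant is a Level Sequence under NeverCook that flies a single possessed CineCameraActor through a level; the file does not show its in-game use.
# One transform track keys a roughly one-minute path at 30 fps, descending, moving back and forth, and panning widely as the initial roll levels out.
# The child camera component holds constant wide focal length, distant manual focus and aperture, with one camera cut section on the master track.
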LevelSequence /Game/NeverCook/FM/Radiant/LS_Radiant
bindings [
 CameraComponent {
  id=#1
  parent_id=#2
  possessed_class=CineCameraComponent
  tracks [
   {
    type=MovieSceneFloatTrack
    display_name=Current Focal Length
    sections [{type=MovieSceneFloatSection, channels [None {type=MovieSceneScriptingFloatChannel, default=16.4668}]}]
   }
   {
    type=MovieSceneFloatTrack
    display_name="Manual Focus Distance (Focus Settings)"
    sections [{type=MovieSceneFloatSection, channels [None {type=MovieSceneScriptingFloatChannel, default=100000}]}]
   }
   {
    type=MovieSceneFloatTrack
    display_name=Current Aperture
    sections [{type=MovieSceneFloatSection, channels [None {type=MovieSceneScriptingFloatChannel, default=2.8}]}]
   }
  ]
 }
 CineCameraActor {
  id=#2
  possessed_class=CineCameraActor
  tracks [{
    type=MovieScene3DTransformTrack
    sections [{
      type=MovieScene3DTransformSection
      channels [
       Location.X {
        type=MovieSceneScriptingDoubleChannel
        default=-393.5192
        keys [[23,533.9718],[451,456.9954],[592,408.3566],[765,-803.4682],[887,-603.0177],[933,313.7155],[1348,376.8884],[1469,-763.3852],[1691,-260.7906],[1981,294.1877]]
       }
       Location.Y {
        type=MovieSceneScriptingDoubleChannel
        default=2723.0016
        keys [[23,2333.4463],[451,1137.2233],[592,1073.5536],[765,1141.666],[887,1986.5514],[933,2107.8152],[1348,1308.5782],[1469,1134.45],[1691,2458.5969],[1981,1409.9257]]
       }
       Location.Z {
        type=MovieSceneScriptingDoubleChannel
        default=2294.2017
        thinned_from=10
        keys [[23,2883.9546],[451,2078.1375],[592,2094.853],[765,2157.8745],[887,2031.0756],[933,2048.926],[1348,2059.4873],[1691,2092.9734],[1981,2086.7505]]
       }
       Rotation.X {
        type=MovieSceneScriptingDoubleChannel
        default=0
        thinned_from=10
        keys [[23,-30.4722],[451,-6.588],[592,-2],[765,-1.6974],[887,-0.0527],[933,-0.0387],[1348,-0.014],[1981,-0.014]]
       }
       Rotation.Y {
        type=MovieSceneScriptingDoubleChannel
        default=-26.3403
        keys [[23,-69.2001],[451,-75.7714],[592,-0.1529],[765,-14.1036],[887,-1.6239],[933,-1.7649],[1348,0.254],[1469,0.654],[1691,-8.946],[1981,-1.746]]
       }
       Rotation.Z {
        type=MovieSceneScriptingDoubleChannel
        default=-96.3278
        keys [[23,-118.8],[451,-122.024],[592,-155.6985],[765,-46.693],[887,51.7073],[933,178.5088],[1348,140.3158],[1469,49.7158],[1691,-92.4842],[1981,-222.8842]]
       }
       Scale.X {type=MovieSceneScriptingDoubleChannel, default=1, const [1,10]}
       Scale.Y {type=MovieSceneScriptingDoubleChannel, default=1, const [1,10]}
       Scale.Z {type=MovieSceneScriptingDoubleChannel, default=1, const [1,10]}
      ]
     }]
   }]
 }
]
display_rate {denominator=1, numerator=30}
tick_resolution {denominator=1, numerator=24000}
tracks [{type=MovieSceneCameraCutTrack, sections [{type=MovieSceneCameraCutSection, end_frame=16412, start_frame=0}]}]
