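# M_Radiant_Base is a Default Lit base material for a dark, metallic surface with scratches, exposing Base color, Metallic, Specular and NormalScale parameters.
# A scratch normal map borrowed from GrindRail is sampled with UVs divided by NormalScale, so the small default tiles it densely.
# An atmospheric cloud noise red channel blends between two low constants, likely for roughness variation, though output connections are not shown.
# The Specular default is far above the usual 0–1 range.
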Material /Game/NeverCook/FM/Radiant/Materials/M_Radiant_Base
graph {
 properties {bAutomaticallySetUsageInEditor=false, usedShadingModels=Default Lit}
 nodes [
  LinearInterpolate(Alpha=TextureSample() {texture=&AtmosphericCloudNoise01}.R) {constA=0.15, constB=0.25}
  ScalarParameter() {defaultValue=1000, parameterName=Specular}
  ScalarParameter() {defaultValue=0.85, parameterName=Metallic}
  TextureSample(UVs=Divide(A=TextureCoordinate(), B=ScalarParameter() {defaultValue=0.0456, parameterName=NormalScale})) {samplerType=SAMPLERTYPE_Normal, texture=/GrindRail/Textures/T_Scratches_Normal}
  VectorParameter() {parameterName=Base, defaultValue=<rgba 0.1 0.1 0.1 0>}
 ]
}
